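# Control Rig for the F_MED_ToothGreek character body that runs shared deform-rig correctives, then adjusts skirt and chest ribbon bones.
# Skirt bases follow thigh rotation through parent constraints, plus offsets weighted by remapped thigh angles, with products of both legs' weights for combined poses.
# Interpolated crouch, knee-slide, grind-rail and emote flags blend in further skirt offsets.
# Spine_04 rotation drives chest ribbons; optional debug controls pose thighs and spine.
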ControlRigBlueprint /BRCosmetics/Characters/Player/Female/Medium/Bodies/F_MED_ToothGreek/Meshes/F_MED_ToothGreek_CtrlRig
graphs [
 New Function_ContainedGraph {nodes [
   n1 = RigVMFunctionEntryNode()
   RigVMFunction_ControlFlowBranch = &RigVMFunction_ControlFlowBranch(Condition=@n1.Debug Controls, ExecuteContext=@n1.ExecuteContext)
   RigVMFunctionReturnNode(ExecuteContext=@RigVMFunction_ControlFlowBranch.Completed)
   RigUnit_GetTransform = &RigUnit_GetTransform(Item=(Type=Control,Name="thigh_r_ctrl"), bInitial=False)
   &RigUnit_SetTransform(
    Item=(Type=Bone,Name="thigh_r")
    Weight=1
    bInitial=False
    Value=@RigUnit_GetTransform.Transform
    ExecutePin=@RigVMFunction_ControlFlowBranch.True
   )
   > &RigUnit_SetTransform(
    Item=(Type=Bone,Name="thigh_l")
    Weight=1
    bInitial=False
    Value=&RigUnit_GetTransform(Item=(Type=Control,Name="thigh_l_ctrl"), bInitial=False).Transform
   )
   > &RigUnit_SetTransform(
    Item=(Type=Bone,Name="spine_04")
    Weight=1
    bInitial=False
    Value=&RigUnit_GetTransform(Item=(Type=Control,Name="spine_04_ctrl"), bInitial=False).Transform
   )
  ]}
 New Function_ContainedGraph {nodes [
   RigVMFunctionEntryNode()
   > RigVMFunction_Sequence = RigVMAggregateNode()
   Delta Rotation Y = RigVMFunctionReferenceNode(Child=thigh_l, Parent=thigh_l_null, Source Maximum=-75, Source Minimum=0, ExecuteContext=@RigVMFunction_Sequence.A)
   > RigVMVariableNode(Value=@"Delta Rotation Y".Delta Rotation) {variable=Thigh_L_SideRotation}
   > Delta Rotation Y_1 = RigVMFunctionReferenceNode(Child=thigh_r, Parent=thigh_r_null, Source Maximum=-75, Source Minimum=0)
   > RigVMVariableNode(Value=@"Delta Rotation Y_1".Delta Rotation) {variable=Thigh_R_SideRotation}
   > Delta Rotation Z = RigVMFunctionReferenceNode(Child=thigh_l, Parent=thigh_l_null, Source Maximum=70, Source Minimum=5)
   > RigVMVariableNode(Value=@"Delta Rotation Z".Delta Rotation) {variable=Thigh_L_BackRotation}
   > Delta Rotation Z_1 = RigVMFunctionReferenceNode(Child=thigh_r, Parent=thigh_r_null, Source Maximum=70, Source Minimum=5)
   > RigVMVariableNode(Value=@"Delta Rotation Z_1".Delta Rotation) {variable=Thigh_R_BackRotation}
   > Delta Rotation Z_2 = RigVMFunctionReferenceNode(Child=thigh_l, Parent=thigh_l_null, Source Maximum=-100, Source Minimum=-3)
   > RigVMVariableNode(Value=@"Delta Rotation Z_2".Delta Rotation) {variable=Thigh_L_FrontRotation}
   > Delta Rotation Z_3 = RigVMFunctionReferenceNode(Child=thigh_r, Parent=thigh_r_null, Source Maximum=-100, Source Minimum=-3)
   > RigVMVariableNode(Value=@"Delta Rotation Z_3".Delta Rotation) {variable=Thigh_R_FrontRotation}
   > Delta Rotation Z_4 = RigVMFunctionReferenceNode(Child=thigh_l, Parent=thigh_l_null, Source Maximum=-135, Source Minimum=-100)
   > RigVMVariableNode(Value=@"Delta Rotation Z_4".Delta Rotation) {variable=extremeThighRotation}
   &RigUnit_ParentConstraint(
    AdvancedSettings=(InterpolationType=Average,RotationOrderForFilter=XZY)
    Child=(Type=Bone,Name="base_skirt_fr_c")
    Filter=(TranslationFilter=(bX=false,bY=false,bZ=false),RotationFilter=(bX=True,bY=True,bZ=True),ScaleFilter=(bX=True,bY=True,bZ=True))
    Parents=((Item=(Type=Bone,Name="thigh_l"),Weight=0.4),(Item=(Type=Bone,Name="thigh_r"),Weight=0.4),(Item=(Type=Bone,Name="pelvis"),Weight=0.2))
    Weight=1
    ExecutePin=@RigVMFunction_Sequence.B
   )
   > &RigUnit_ParentConstraint(
    AdvancedSettings=(InterpolationType=Average,RotationOrderForFilter=XZY)
    Child=(Type=Bone,Name="base_skirt_bk_c")
    Filter=(TranslationFilter=(bX=false,bY=false,bZ=false),RotationFilter=(bX=True,bY=True,bZ=True),ScaleFilter=(bX=True,bY=True,bZ=True))
    Parents=((Item=(Type=Bone,Name="thigh_l"),Weight=0.5),(Item=(Type=Bone,Name="thigh_r"),Weight=0.5))
    Weight=1
   )
   > &RigUnit_ParentConstraint(
    AdvancedSettings=(InterpolationType=Average,RotationOrderForFilter=XZY)
    Child=(Type=Bone,Name="base_skirt_fr_l")
    Filter=(TranslationFilter=(bX=false,bY=false,bZ=false),RotationFilter=(bX=True,bY=True,bZ=True),ScaleFilter=(bX=True,bY=True,bZ=True))
    Parents=((Item=(Type=Bone,Name="thigh_l"),Weight=1))
    Weight=1
   )
   > &RigUnit_ParentConstraint(
    AdvancedSettings=(InterpolationType=Average,RotationOrderForFilter=XZY)
    Child=(Type=Bone,Name="base_skirt_sd_l")
    Filter=(TranslationFilter=(bX=false,bY=false,bZ=false),RotationFilter=(bX=True,bY=True,bZ=True),ScaleFilter=(bX=True,bY=True,bZ=True))
    Parents=((Item=(Type=Bone,Name="thigh_l"),Weight=1))
    Weight=1
   )
   > &RigUnit_ParentConstraint(
    AdvancedSettings=(InterpolationType=Average,RotationOrderForFilter=XZY)
    Child=(Type=Bone,Name="base_skirt_bk_l")
    Filter=(TranslationFilter=(bX=false,bY=false,bZ=false),RotationFilter=(bX=True,bY=True,bZ=True),ScaleFilter=(bX=True,bY=True,bZ=True))
    Parents=((Item=(Type=Bone,Name="thigh_l"),Weight=1))
    Weight=1
   )
   > &RigUnit_ParentConstraint(
    AdvancedSettings=(InterpolationType=Average,RotationOrderForFilter=XZY)
    Child=(Type=Bone,Name="base_skirt_fr_r")
    Filter=(TranslationFilter=(bX=false,bY=false,bZ=false),RotationFilter=(bX=True,bY=True,bZ=True),ScaleFilter=(bX=True,bY=True,bZ=True))
    Parents=((Item=(Type=Bone,Name="thigh_r"),Weight=1))
    Weight=1
   )
   > &RigUnit_ParentConstraint(
    AdvancedSettings=(InterpolationType=Average,RotationOrderForFilter=XZY)
    Child=(Type=Bone,Name="base_skirt_sd_r")
    Filter=(TranslationFilter=(bX=false,bY=false,bZ=false),RotationFilter=(bX=True,bY=True,bZ=True),ScaleFilter=(bX=True,bY=True,bZ=True))
    Parents=((Item=(Type=Bone,Name="thigh_r"),Weight=1))
    Weight=1
   )
   > &RigUnit_ParentConstraint(
    AdvancedSettings=(InterpolationType=Average,RotationOrderForFilter=XZY)
    Child=(Type=Bone,Name="base_skirt_bk_r")
    Filter=(TranslationFilter=(bX=false,bY=false,bZ=false),RotationFilter=(bX=True,bY=True,bZ=True),ScaleFilter=(bX=True,bY=True,bZ=True))
    Parents=((Item=(Type=Bone,Name="thigh_r"),Weight=1))
    Weight=1
   )
   &RigUnit_ModifyTransforms(
    ItemToModify="((Item=(Type=Bone,Name=\"base_skirt_fr_c\"),Transform=(Rotation=(X=0.3524,Y=0.0061,Z=0.4245,W=-0.834),Translation=(X=-0.8041,Y=-2.2092,Z=0),Scale3D=(X=1,Y=1,Z=1))),(Item=(Type=Bone,Name=\"base_skirt_fr_l\"),Transform=(Rotation=(X=-0.002,Y=0.0947,Z=0.003,W=-0.9 ...(truncated)"
    WeightMinimum=0
    ExecutePin=@RigVMFunction_Sequence.C
    Weight=RigVMVariableNode() {variable=Thigh_L_FrontRotation}.Value
   )
   > &RigUnit_ModifyTransforms(
    ItemToModify="((Item=(Type=Bone,Name=\"base_skirt_fr_c\"),Transform=(Rotation=(X=0,Y=0.1179,Z=0,W=-0.993),Translation=(X=0,Y=0,Z=0),Scale3D=(X=1,Y=1,Z=1))),(Item=(Type=Bone,Name=\"base_skirt_fr_l\"),Transform=(Rotation=(X=0.2406,Y=0.0893,Z=-0.1145,W=0.9597),Translation=(X=0 ...(truncated)"
    WeightMinimum=0
    Weight=RigVMVariableNode() {variable=Thigh_L_SideRotation}.Value
   )
   > &RigUnit_ModifyTransforms(
    ItemToModify="((Item=(Type=Bone,Name=\"base_skirt_fr_c\"),Transform=(Rotation=(X=0,Y=0,Z=0.2179,W=-0.976),Translation=(X=0,Y=0,Z=0),Scale3D=(X=1,Y=1,Z=1))),(Item=(Type=Bone,Name=\"base_skirt_fr_l\"),Transform=(Rotation=(X=0,Y=0,Z=-0.4982,W=0.8671),Translation=(X=0,Y=0,Z=0), ...(truncated)"
    WeightMinimum=0
    Weight=RigVMVariableNode() {variable=Thigh_L_BackRotation}.Value
   )
   &RigUnit_ModifyTransforms(
    ItemToModify="((Item=(Type=Bone,Name=\"base_skirt_fr_c\"),Transform=(Rotation=(X=-0.3697,Y=0.0415,Z=0.5241,W=-0.7661),Translation=(X=-1.5215,Y=-1.2073,Z=-1.8224),Scale3D=(X=1,Y=1,Z=1))),(Item=(Type=Bone,Name=\"base_skirt_fr_r\"),Transform=(Rotation=(X=-0.0173,Y=-0.0811,Z=0. ...(truncated)"
    WeightMinimum=0
    ExecutePin=@RigVMFunction_Sequence.D
    Weight=RigVMVariableNode() {variable=Thigh_R_FrontRotation}.Value
   )
   > &RigUnit_ModifyTransforms(
    ItemToModify="((Item=(Type=Bone,Name=\"base_skirt_fr_c\"),Transform=(Rotation=(X=0,Y=0.1736,Z=0,W=0.9848),Translation=(X=0,Y=0,Z=0),Scale3D=(X=1,Y=1,Z=1))),(Item=(Type=Bone,Name=\"base_skirt_fr_r\"),Transform=(Rotation=(X=0.0759,Y=-0.0452,Z=0.0897,W=-0.992),Translation=(X=0 ...(truncated)"
    WeightMinimum=0
    Weight=RigVMVariableNode() {variable=Thigh_R_SideRotation}.Value
   )
   > &RigUnit_ModifyTransforms(
    ItemToModify="((Item=(Type=Bone,Name=\"base_skirt_fr_c\"),Transform=(Rotation=(X=0,Y=0,Z=0.2179,W=-0.976),Translation=(X=0,Y=0,Z=0),Scale3D=(X=1,Y=1,Z=1))),(Item=(Type=Bone,Name=\"base_skirt_fr_r\"),Transform=(Rotation=(X=0,Y=0,Z=-0.4982,W=0.8671),Translation=(X=0,Y=0,Z=0), ...(truncated)"
    WeightMinimum=0
    Weight=RigVMVariableNode() {variable=Thigh_R_BackRotation}.Value
   )
   Multiply = &RigVMFunction_MathDoubleMul(A=RigVMVariableNode() {variable=Thigh_L_FrontRotation}.Value, B=RigVMVariableNode() {variable=Thigh_R_FrontRotation}.Value)
   &RigUnit_ModifyTransforms(
    ItemToModify="((Item=(Type=Bone,Name=\"base_skirt_fr_c\"),Transform=(Rotation=(X=-0.0182,Y=0.3945,Z=0.7173,W=0.574),Translation=(X=5.1186,Y=1.3663,Z=-3.5661),Scale3D=(X=1,Y=1,Z=1))),(Item=(Type=Bone,Name=\"None\"),Transform=(Rotation=(X=0,Y=0,Z=0,W=1),Translation=(X=0,Y=0,Z ...(truncated)"
    WeightMinimum=0
    Weight=@Multiply.Result
    ExecutePin=@RigVMFunction_Sequence.E
   )
   > &RigUnit_ModifyTransforms(
    ItemToModify="((Item=(Type=Bone,Name=\"base_skirt_bk_c\"),Transform=(Rotation=(X=-0.0204,Y=-0.1284,Z=0.5625,W=0.8165),Translation=(X=2.9623,Y=-1.7716,Z=0.5092),Scale3D=(X=1,Y=1,Z=1))),(Item=(Type=Bone,Name=\"None\"),Transform=(Rotation=(X=0,Y=0,Z=0,W=1),Translation=(X=0,Y=0 ...(truncated)"
    WeightMinimum=0
    Weight=&RigVMFunction_MathDoubleMul(A=RigVMVariableNode() {variable=Thigh_R_BackRotation}.Value, B=RigVMVariableNode() {variable=Thigh_L_BackRotation}.Value).Result
   )
   RigVMFunction_MathBoolToFloat = &RigVMFunction_MathBoolToFloat(Value=RigVMVariableNode() {variable=isCrouching_CR}.Value)
   n1 = &RigVMFunction_AlphaInterp(InterpSpeedDecreasing=10, InterpSpeedIncreasing=2, Scale=1, bClampResult=true, bMapRange=true, Value=@RigVMFunction_MathBoolToFloat.Result)
   > &RigUnit_ModifyTransforms(
    ItemToModify="((Item=(Type=Bone,Name=\"base_skirt_fr_l\"),Transform=(Rotation=(X=0.3423,Y=0.0528,Z=-0.0977,W=-0.933),Translation=(X=-0.2307,Y=-0.2132,Z=2.6346),Scale3D=(X=1,Y=1,Z=1))),(Item=(Type=Bone,Name=\"base_skirt_bk_l\"),Transform=(Rotation=(X=-0.4188,Y=-0.145,Z=-0.03 ...(truncated)"
    WeightMinimum=0
    Weight=@n1.Result
   )
   > &RigUnit_ModifyTransforms(
    ItemToModify="((Item=(Type=Bone,Name=\"base_skirt_fr_l\"),Transform=(Rotation=(X=0,Y=0,Z=-0.0523,W=0.9986),Translation=(X=0,Y=-2,Z=0),Scale3D=(X=1,Y=1,Z=1))),(Item=(Type=Bone,Name=\"base_skirt_sd_l\"),Transform=(Rotation=(X=0,Y=0,Z=0,W=1),Translation=(X=0,Y=0,Z=0),Scale3D=( ...(truncated)"
    WeightMinimum=0
    Weight=&RigVMFunction_MathDoubleMul(B=@n1.Result, A=RigVMVariableNode() {variable=extremeThighRotation}.Value).Result
   )
   n2 = &RigVMFunction_AlphaInterp(
    InterpSpeedDecreasing=10
    InterpSpeedIncreasing=2
    Scale=1
    bClampResult=true
    bMapRange=true
    Value=&RigVMFunction_MathBoolToFloat(Value=RigVMVariableNode() {variable=isKneeSliding_CR}.Value).Result
   )
   > &RigUnit_ModifyTransforms(
    ItemToModify="((Item=(Type=Bone,Name=\"base_skirt_bk_c\"),Transform=(Rotation=(X=0,Y=0,Z=0.0872,W=0.9962),Translation=(X=0,Y=0,Z=0),Scale3D=(X=1,Y=1,Z=1))),(Item=(Type=Bone,Name=\"base_skirt_bk_l\"),Transform=(Rotation=(X=0,Y=0,Z=0,W=1),Translation=(X=0,Y=-5,Z=0),Scale3D=(X ...(truncated)"
    WeightMinimum=0
    Weight=@n2.Result
   )
   n3 = &RigVMFunction_AlphaInterp(
    InterpSpeedDecreasing=10
    InterpSpeedIncreasing=2
    Scale=1
    bClampResult=true
    bMapRange=true
    Value=&RigVMFunction_MathBoolToFloat(Value=RigVMVariableNode() {variable=isOnGrindRail_CR}.Value).Result
   )
   > &RigUnit_ModifyTransforms(
    ItemToModify="((Item=(Type=Bone,Name=\"base_skirt_fr_l\"),Transform=(Rotation=(X=0.0023,Y=0.0871,Z=-0.0261,W=0.9959),Translation=(X=0,Y=-1,Z=-5),Scale3D=(X=1,Y=1,Z=1))),(Item=(Type=Bone,Name=\"None\"),Transform=(Rotation=(X=0,Y=0,Z=0,W=1),Translation=(X=0,Y=0,Z=0),Scale3D=( ...(truncated)"
    WeightMinimum=0
    Weight=@n3.Result
   )
   n4 = &RigVMFunction_AlphaInterp(
    InterpSpeedDecreasing=10
    InterpSpeedIncreasing=2
    Scale=1
    bClampResult=true
    bMapRange=true
    Value=&RigVMFunction_MathBoolToFloat(Value=RigVMVariableNode() {variable=isEmoting_CR}.Value).Result
   )
   > &RigUnit_ModifyTransforms(
    ItemToModify="((Item=(Type=Bone,Name=\"base_skirt_fr_l\"),Transform=(Rotation=(X=0,Y=0,Z=0,W=1),Translation=(X=0,Y=-2,Z=5),Scale3D=(X=1,Y=1,Z=1))),(Item=(Type=Bone,Name=\"base_skirt_fr_c\"),Transform=(Rotation=(X=0,Y=0,Z=0,W=1),Translation=(X=0,Y=-2,Z=3),Scale3D=(X=1,Y=1,Z= ...(truncated)"
    WeightMinimum=0
    Weight=&RigVMFunction_MathDoubleMul(B=@n4.Result, A=RigVMVariableNode() {variable=Thigh_L_FrontRotation}.Value).Result
   )
   RigVMFunctionReturnNode(ExecuteContext=@RigVMFunction_Sequence.G)
  ]}
 New Function_ContainedGraph {nodes [
   n1 = RigVMFunctionEntryNode()
   RigVMFunction_MathQuaternionToEuler = &RigVMFunction_MathQuaternionToEuler(
    RotationOrder=XYZ
    Value=&RigUnit_GetRelativeTransformForItem(
     Child=(Type=Bone,Name="hand_l")
     Parent=(Type=Null,Name="lowerarm_l")
     bChildInitial=False
     bParentInitial=False
     Child.Name=@n1.Child
     Parent.Name=@n1.Parent
    ).RelativeTransform.Rotation
   )
   Remap = &RigVMFunction_MathDoubleRemap(
    TargetMaximum=1
    TargetMinimum=0
    Value=@RigVMFunction_MathQuaternionToEuler.Result.Y
    SourceMaximum=@n1.Source Maximum
    SourceMinimum=@n1.Source Minimum
   )
   > RigVMFunctionReturnNode(Delta Rotation=@Remap.Result)
  ]}
 New Function_ContainedGraph {nodes [
   n1 = RigVMFunctionEntryNode()
   RigVMFunction_MathQuaternionToEuler = &RigVMFunction_MathQuaternionToEuler(
    RotationOrder=XYZ
    Value=&RigUnit_GetRelativeTransformForItem(
     Child=(Type=Bone,Name="hand_l")
     Parent=(Type=Null,Name="lowerarm_l")
     bChildInitial=False
     bParentInitial=False
     Child.Name=@n1.Child
     Parent.Name=@n1.Parent
    ).RelativeTransform.Rotation
   )
   Remap = &RigVMFunction_MathDoubleRemap(
    TargetMaximum=1
    TargetMinimum=0
    Value=@RigVMFunction_MathQuaternionToEuler.Result.Z
    SourceMaximum=@n1.Source Maximum
    SourceMinimum=@n1.Source Minimum
   )
   > RigVMFunctionReturnNode(Delta Rotation=@Remap.Result)
  ]}
 New Function_ContainedGraph {nodes [
   RigVMFunctionEntryNode()
   > RigVMFunction_Sequence = RigVMAggregateNode()
   Delta Rotation Y = RigVMFunctionReferenceNode(Child=spine_04, Parent=spine_04_null, Source Maximum=18, Source Minimum=2, ExecuteContext=@RigVMFunction_Sequence.A)
   > RigVMVariableNode(Value=@"Delta Rotation Y".Delta Rotation) {variable=spineRotZpos}
   > Delta Rotation Y_1 = RigVMFunctionReferenceNode(Child=spine_04, Parent=spine_04_null, Source Maximum=-18, Source Minimum=-2)
   > RigVMVariableNode(Value=@"Delta Rotation Y_1".Delta Rotation) {variable=spineRotZneg}
   &RigUnit_ModifyTransforms(
    ItemToModify=((Item=(Type=Bone,Name="base_ribbonChest_l"),Transform=(Rotation=(X=0,Y=0,Z=-0.1305,W=0.9914),Translation=(X=0,Y=0,Z=0),Scale3D=(X=1,Y=1,Z=1))))
    WeightMinimum=0
    ExecutePin=@RigVMFunction_Sequence.B
    Weight=RigVMVariableNode() {variable=spineRotZpos}.Value
   )
   > &RigUnit_ModifyTransforms(
    ItemToModify=((Item=(Type=Bone,Name="base_ribbonChest_r"),Transform=(Rotation=(X=0,Y=0,Z=-0.1045,W=0.9945),Translation=(X=0,Y=0,Z=0),Scale3D=(X=1,Y=1,Z=1))))
    WeightMinimum=0
    Weight=RigVMVariableNode() {variable=spineRotZneg}.Value
   )
   RigVMFunctionReturnNode(ExecuteContext=@RigVMFunction_Sequence.C)
  ]}
 RigVMFunctionLibrary {nodes [
   ChestRibbons = RigVMCollapseNode(ExecuteContext=())
   Control Setup = RigVMCollapseNode(Debug Controls=False, ExecuteContext=())
   Delta Rotation Y = RigVMCollapseNode(ExecuteContext=())
   Delta Rotation Z = RigVMCollapseNode(ExecuteContext=())
   Skirt = RigVMCollapseNode(ExecuteContext=())
  ]}
 RigVMFunction_Sequence_1_ContainedGraph {nodes [
   n1 = RigVMFunctionEntryNode()
   RigVMFunction_Sequence = &RigVMFunction_Sequence(ExecuteContext=@n1.ExecuteContext)
   n2 = &RigVMFunction_Sequence(ExecuteContext=@RigVMFunction_Sequence.B)
   n3 = &RigVMFunction_Sequence(ExecuteContext=@n2.B)
   n4 = &RigVMFunction_Sequence(ExecuteContext=@n3.B)
   n5 = &RigVMFunction_Sequence(ExecuteContext=@n4.B)
   n6 = &RigVMFunction_Sequence(ExecuteContext=@n5.B)
   RigVMFunctionReturnNode(A=@RigVMFunction_Sequence.A, B=@n2.A, C=@n3.A, D=@n4.A, E=@n5.A, F=@n6.A, G=@n6.B)
  ]}
 RigVMFunction_Sequence_1_ContainedGraph {nodes [
   n1 = RigVMFunctionEntryNode()
   RigVMFunction_Sequence = &RigVMFunction_Sequence(ExecuteContext=@n1.ExecuteContext)
   n2 = &RigVMFunction_Sequence(ExecuteContext=@RigVMFunction_Sequence.B)
   RigVMFunctionReturnNode(A=@RigVMFunction_Sequence.A, B=@n2.A, C=@n2.B)
  ]}
 RigVMModel {nodes [
   RigUnit_BeginExecution = &RigUnit_BeginExecution()
   RigVMFunction_Sequence = &RigVMFunction_Sequence(ExecuteContext=@RigUnit_BeginExecution.ExecutePin)
   Control Setup = RigVMFunctionReferenceNode(Debug Controls=false, ExecuteContext=@RigVMFunction_Sequence.A)
   FN_DeformRig_FNC = RigVMFunctionReferenceNode(
    DeformRig_ Curve="(deformRig_l_upperarm_up_alpha_72_#1=False,deformRig_r_upperarm_up_alpha_73_#2=False,deformRig_l_upperarm_fr_alpha_74_#3=False,deformRig_r_upperarm_fr_alpha_75_CB51B8 ...(truncated)"
    DeformRig_Bool="(upperarm_1_#4=True,elbow_3_#5=True,wrist_7_#6=True,groin_8_#7=True,glute_17_#8=True,knee_9_40C408324CA261 ...(truncated)"
    DeformRig_Debug=(upperarm_debug_18_#4=false,wrist_debug_20_#6=False,lat_debug_26_#9=False,groinGlute_debug_27_#7=False)
    DeformRig_Value="(upperarm_up_18_#10=(Rotation=(X=0,Y=0,Z=0,W=1),Translation=(X=0,Y=0,Z=8),Scale3D=(X=0.7,Y=1,Z=1)),upperarm_fwd_fr_19_#11=(Rotation=(X=0,Y=0,Z=0,W=1),Translation=(X=0,Y=10,Z=0),Scale3D=(X=0.5,Y=1,Z= ...(truncated)"
    ExecuteContext=@RigVMFunction_Sequence.B
   )
   > Skirt = RigVMFunctionReferenceNode()
   > ChestRibbons = RigVMFunctionReferenceNode()
  ]}
]
hierarchy [
 root {loc=<0 0 0>, rot=<0 0 0 1>}
 pelvis {parent=root, loc=<0 -0.5301 98.0571>, rot=<0.0127 0.707 -0.0127 -0.707>}
 spine_01 {parent=pelvis, loc=<6.1208 -0.9328 0>, rot=<0 0 0.034 -0.9994>}
 spine_02 {parent=spine_01, loc=<6.5997 -0.2095 0>, rot=<0 0 0.0238 -0.9997>}
 spine_03 {parent=spine_02, loc=<6.4914 -0.1237 0>, rot=<0 0 0.0149 -0.9999>}
 spine_04 {parent=spine_03, loc=<6.9598 0.0794 0>, rot=<0 0 -0.0132 -0.9999>}
 spine_05 {parent=spine_04, loc=<6.9987 0 0>, rot=<0 0 -0.0015 -1>}
 neck_01 {parent=spine_05, loc=<11.3138 0.1529 0>, rot=<0 0 -0.2003 -0.9797>}
 neck_02 {parent=neck_01, loc=<4.5627 0 0>, rot=<0 0 0.0504 -0.9987>}
 head {parent=neck_02, loc=<5.3914 0 0>, rot=<0 0 0.115 -0.9934>}
 clavicle_l {parent=spine_05, loc=<5.4831 -0.1053 -3.2946>, rot=<0.0465 -0.7289 -0.009 -0.6829>}
 upperarm_l {parent=clavicle_l, loc=<9.8658 -1.2136 -0.4105>, rot=<0.0164 -0.3245 -0.0269 -0.9454>}
 lowerarm_l {parent=upperarm_l, loc=<26.465 -0.8345 0.259>, rot=<0.0036 0.0161 -0.2298 -0.9731>}
 hand_l {parent=lowerarm_l, loc=<22.7355 -0.3203 -0.6105>, rot=<-0.6784 -0.0223 0.0735 -0.7307>}
 thumb_01_l {parent=hand_l, loc=<1.3118 -0.3191 -2.459>, rot=<0.6546 -0.4042 0.0664 -0.6354>}
 thumb_02_l {parent=thumb_01_l, loc=<5.9336 -0.0448 0.0401>, rot=<-0.0037 0.0213 0.298 -0.9543>}
 thumb_03_l {parent=thumb_02_l, loc=<2.2223 0.3682 -0.0907>, rot=<0.0118 -0.013 0.0066 -0.9998>}
 index_metacarpal_l {parent=hand_l, loc=<1.6535 0.0183 -2.96>, rot=<0.0158 -0.0124 -0.0173 -0.9996>}
 index_01_l {parent=index_metacarpal_l, loc=<8.3008 0.0357 -0.0524>, rot=<-0.0044 0.0485 0.0924 -0.9945>}
 index_02_l {parent=index_01_l, loc=<3.3587 -0.3004 0.0452>, rot=<-0.002 0.0027 0.3225 -0.9466>}
 index_03_l {parent=index_02_l, loc=<2.5578 -0.123 -0.0252>, rot=<-0.002 0.0067 0.2558 -0.9667>}
 middle_metacarpal_l {parent=hand_l, loc=<2.1856 -0.1725 -0.5984>, rot=<0.006 0.0168 -0.0329 -0.9993>}
 middle_01_l {parent=middle_metacarpal_l, loc=<7.3275 -0.2447 0.0129>, rot=<-0.0208 0.0042 0.0858 -0.9961>}
 middle_02_l {parent=middle_01_l, loc=<4.0123 0.1401 0.1175>, rot=<-0.0218 0.0258 0.3512 -0.9357>}
 middle_03_l {parent=middle_02_l, loc=<2.6077 0.0534 0.1373>, rot=<0.0082 0.0188 0.1708 -0.9851>}
 ring_metacarpal_l {parent=hand_l, loc=<2.2074 -0.3625 1.3638>, rot=<-0.0284 0.0722 -0.0458 -0.9959>}
 ring_01_l {parent=ring_metacarpal_l, loc=<7.0295 -0.6197 -0.1446>, rot=<-0.0489 -0.0252 0.1892 -0.9804>}
 ring_02_l {parent=ring_01_l, loc=<3.6174 0.4647 -0.1027>, rot=<0 0 0.2806 -0.9598>}
 ring_03_l {parent=ring_02_l, loc=<2.4152 0.114 0.0199>, rot=<0 0 0.1189 -0.9929>}
 pinky_metacarpal_l {parent=hand_l, loc=<1.7616 -0.9753 3.3814>, rot=<-0.0397 0.0814 -0.069 -0.9935>}
 pinky_01_l {parent=pinky_metacarpal_l, loc=<7.1709 -0.4831 -0.1094>, rot=<-0.0654 -0.0035 0.2242 -0.9723>}
 pinky_02_l {parent=pinky_01_l, loc=<2.8274 0.0893 -0.1043>, rot=<0.0037 0.0011 0.2854 -0.9584>}
 pinky_03_l {parent=pinky_02_l, loc=<2.2046 0.0307 -0.0078>, rot=<-0.0002 0.0098 0.1445 -0.9895>}
 deform_wrist_dn_base_l {parent=hand_l, loc=<0 -1.5326 0>, rot=<0 0 0 -1>}
 deform_wrist_dn_l {parent=deform_wrist_dn_base_l, loc=<0 0 0>, rot=<0 0 0 1>}
 deform_wrist_up_base_l {parent=hand_l, loc=<0 2.3009 0>, rot=<0 0 0 -1>}
 deform_wrist_up_l {parent=deform_wrist_up_base_l, loc=<0 0 0>, rot=<0 0 0 1>}
 lowerarm_twist_02_l {parent=lowerarm_l, loc=<10.6707 0 0>, rot=<0 0 0 1>}
 lowerarm_twist_01_l {parent=lowerarm_l, loc=<18.316 0 0>, rot=<0 0 0 1>}
 deform_elbow_in_base_l {parent=lowerarm_l, loc=<0.1834 3.6062 -0.2572>, rot=<0 0 0 -1>}
 deform_elbow_in_l {parent=deform_elbow_in_base_l, loc=<0 0 0>, rot=<0 0 0 1>}
 deform_elbow_out_base_l {parent=lowerarm_l, loc=<0 -4 0>, rot=<0 0 0 1>}
 deform_elbow_out_l {parent=deform_elbow_out_base_l, loc=<0 0 0>, rot=<0 0 0 1>}
 upperarm_twist_01_l {parent=upperarm_l, loc=<7.2135 -0.1134 0>, rot=<0 0 0 1>}
 deform_upperarm_bk_base_l {parent=upperarm_twist_01_l, loc=<-6.4053 -4.3746 0.314>, rot=<0 0 0 -1>}
 deform_upperarm_bk_l {parent=deform_upperarm_bk_base_l, loc=<0 0 0>, rot=<0 0 0 1>}
 deform_upperarm_fr_base_l {parent=upperarm_twist_01_l, loc=<-5.6312 5.092 0.3354>, rot=<0 0 0 -1>}
 deform_upperarm_fr_l {parent=deform_upperarm_fr_base_l, loc=<0 0 0>, rot=<0 0 0 1>}
 deform_upperarm_up_base_l {parent=upperarm_twist_01_l, loc=<-8.891 0.4825 3.9517>, rot=<0 0 0 -1>}
 deform_upperarm_up_l {parent=deform_upperarm_up_base_l, loc=<0 0 0>, rot=<0 0 0 1>}
 upperarm_twist_02_l {parent=upperarm_l, loc=<19.7998 -0.1134 0>, rot=<0 0 0 1>}
 clavicle_r {parent=spine_05, loc=<5.4831 -0.1053 3.2946>, rot=<0.7289 0.0465 -0.6829 0.009>}
 upperarm_r {parent=clavicle_r, loc=<-9.8658 1.2136 0.4105>, rot=<0.0164 -0.3245 -0.0269 -0.9454>}
 lowerarm_r {parent=upperarm_r, loc=<-26.465 0.8345 -0.259>, rot=<0.0036 0.0161 -0.2298 -0.9731>}
 hand_r {parent=lowerarm_r, loc=<-22.7355 0.3203 0.6106>, rot=<-0.6784 -0.0223 0.0735 -0.7307>}
 thumb_01_r {parent=hand_r, loc=<-1.3118 0.3191 2.459>, rot=<0.6546 -0.4042 0.0664 -0.6354>}
 thumb_02_r {parent=thumb_01_r, loc=<-5.9336 0.0448 -0.0401>, rot=<-0.0037 0.0213 0.298 -0.9543>}
 thumb_03_r {parent=thumb_02_r, loc=<-2.2223 -0.3682 0.0907>, rot=<0.0118 -0.013 0.0066 -0.9998>}
 index_metacarpal_r {parent=hand_r, loc=<-1.6535 -0.0183 2.96>, rot=<0.0158 -0.0124 -0.0173 -0.9996>}
 index_01_r {parent=index_metacarpal_r, loc=<-8.3008 -0.0357 0.0524>, rot=<-0.0044 0.0485 0.0924 -0.9945>}
 index_02_r {parent=index_01_r, loc=<-3.3587 0.3004 -0.0452>, rot=<-0.002 0.0027 0.3225 -0.9466>}
 index_03_r {parent=index_02_r, loc=<-2.5578 0.123 0.0252>, rot=<-0.002 0.0067 0.2558 -0.9667>}
 middle_metacarpal_r {parent=hand_r, loc=<-2.1856 0.1725 0.5984>, rot=<0.006 0.0168 -0.0329 -0.9993>}
 middle_01_r {parent=middle_metacarpal_r, loc=<-7.3275 0.2447 -0.0129>, rot=<-0.0208 0.0042 0.0858 -0.9961>}
 middle_02_r {parent=middle_01_r, loc=<-4.0123 -0.1401 -0.1175>, rot=<-0.0218 0.0258 0.3512 -0.9357>}
 middle_03_r {parent=middle_02_r, loc=<-2.6077 -0.0534 -0.1373>, rot=<0.0082 0.0188 0.1708 -0.9851>}
 ring_metacarpal_r {parent=hand_r, loc=<-2.2074 0.3625 -1.3638>, rot=<-0.0284 0.0722 -0.0458 -0.9959>}
 ring_01_r {parent=ring_metacarpal_r, loc=<-7.0295 0.6197 0.1446>, rot=<-0.0489 -0.0252 0.1892 -0.9804>}
 ring_02_r {parent=ring_01_r, loc=<-3.6174 -0.4647 0.1027>, rot=<0 0 0.2806 -0.9598>}
 ring_03_r {parent=ring_02_r, loc=<-2.4152 -0.114 -0.0199>, rot=<0 0 0.1189 -0.9929>}
 pinky_metacarpal_r {parent=hand_r, loc=<-1.7616 0.9753 -3.3814>, rot=<-0.0397 0.0814 -0.069 -0.9935>}
 pinky_01_r {parent=pinky_metacarpal_r, loc=<-7.1709 0.4831 0.1094>, rot=<-0.0654 -0.0035 0.2242 -0.9723>}
 pinky_02_r {parent=pinky_01_r, loc=<-2.8274 -0.0893 0.1043>, rot=<0.0037 0.0011 0.2854 -0.9584>}
 pinky_03_r {parent=pinky_02_r, loc=<-2.2046 -0.0307 0.0078>, rot=<-0.0002 0.0098 0.1445 -0.9895>}
 deform_wrist_dn_base_r {parent=hand_r, loc=<-0.0001 1.5328 0>, rot=<0 0 0 -1>}
 deform_wrist_dn_r {parent=deform_wrist_dn_base_r, loc=<0 0 0>, rot=<0 0 0 1>}
 deform_wrist_up_base_r {parent=hand_r, loc=<-0.0002 -2.3007 -0.0001>, rot=<0 0 0 -1>}
 deform_wrist_up_r {parent=deform_wrist_up_base_r, loc=<0 0 0>, rot=<0 0 0 1>}
 lowerarm_twist_02_r {parent=lowerarm_r, loc=<-10.6707 0 0>, rot=<0 0 0 -1>}
 lowerarm_twist_01_r {parent=lowerarm_r, loc=<-18.316 0 0>, rot=<0 0 0 -1>}
 deform_elbow_in_base_r {parent=lowerarm_r, loc=<-0.1834 -3.6062 0.2572>, rot=<0 0 0 -1>}
 deform_elbow_in_r {parent=deform_elbow_in_base_r, loc=<0 0 0>, rot=<0 0 0 1>}
 deform_elbow_out_base_r {parent=lowerarm_r, loc=<0 4 -0.0001>, rot=<0 0 0 -1>}
 deform_elbow_out_r {parent=deform_elbow_out_base_r, loc=<0 0 0>, rot=<0 0 0 1>}
 upperarm_twist_01_r {parent=upperarm_r, loc=<-7.2135 0.1134 0>, rot=<0 0 0 -1>}
 deform_upperarm_bk_base_r {parent=upperarm_twist_01_r, loc=<6.4053 4.3746 -0.314>, rot=<0 0 0 -1>}
 deform_upperarm_bk_r {parent=deform_upperarm_bk_base_r, loc=<0 0 0>, rot=<0 0 0 1>}
 deform_upperarm_fr_base_r {parent=upperarm_twist_01_r, loc=<5.6312 -5.092 -0.3354>, rot=<0 0 0 -1>}
 deform_upperarm_fr_r {parent=deform_upperarm_fr_base_r, loc=<0 0 0>, rot=<0 0 0 1>}
 deform_upperarm_up_base_r {parent=upperarm_twist_01_r, loc=<8.891 -0.4825 -3.9517>, rot=<0 0 0 -1>}
 deform_upperarm_up_r {parent=deform_upperarm_up_base_r, loc=<0 0 0>, rot=<0 0 0 1>}
 upperarm_twist_02_r {parent=upperarm_r, loc=<-19.7998 0.1134 0>, rot=<0 0 0 -1>}
 base_ribbonArm_r {parent=upperarm_r, loc=<-14.9904 0.3763 -4.1036>, rot=<0.0444 -0.7402 -0.6708 0.016>}
 dyn_ribbonArm_a01_r {parent=base_ribbonArm_r, loc=<0.5597 0.024 1.1808>, rot=<-0.0105 0.0026 0.0474 -0.9988>}
 dyn_ribbonArm_b01_r {parent=base_ribbonArm_r, loc=<0.9045 0.024 -1.5484>, rot=<0 0 0.0453 -0.999>}
 dyn_ribbonArm_c01_r {parent=base_ribbonArm_r, loc=<1.0141 0.024 0.8425>, rot=<-0.0009 0.0333 0.0262 -0.9991>}
 dyn_ribbonArm_c02_r {parent=dyn_ribbonArm_c01_r, loc=<7.1491 0 0>, rot=<0 0 0 1>}
 dyn_ribbonArm_d01_r {parent=base_ribbonArm_r, loc=<0.855 -0.0623 -0.9467>, rot=<0 0 0.0725 -0.9974>}
 attach_backpack {parent=spine_05, loc=<-0.1238 -13.4716 -0.0002>, rot=<-0.0093 -0.707 0.0093 -0.707>}
 Backpack_BR {parent=attach_backpack, type=Null, loc=<0 0 0>, rot=<0 0 0 1>}
 attach_cape {parent=spine_05, loc=<9.0148 -11.7125 0.0003>, rot=<0.0616 -0.7044 -0.0616 -0.7044>}
 cape {parent=attach_cape, type=Null, loc=<0 0 0>, rot=<0 0 0 1>}
 deform_pec_aim_base_l {parent=spine_04, loc=<3.2456 9.3309 -1.2237>, rot=<0.0603 -0.4033 0.3318 -0.8507>}
 deform_pec_aim_l {parent=deform_pec_aim_base_l, loc=<0 0 0>, rot=<0 0 0 1>}
 deform_pec_base_l {parent=deform_pec_aim_l, loc=<9.3665 0 0>, rot=<0 0 0 1>}
 deform_pec_l {parent=deform_pec_base_l, loc=<0 0 0>, rot=<0 0 0 1>}
 deform_pec_aim_base_r {parent=spine_04, loc=<3.2452 9.3309 1.2237>, rot=<-0.4033 -0.0603 0.8507 0.3318>}
 deform_pec_aim_r {parent=deform_pec_aim_base_r, loc=<0 0 0>, rot=<0 0 0 1>}
 deform_pec_base_r {parent=deform_pec_aim_r, loc=<-9.3667 -0.0003 -0.0004>, rot=<0 0 0 1>}
 deform_pec_r {parent=deform_pec_base_r, loc=<0 0 0>, rot=<0 0 0 1>}
 base_ribbonChest_l {parent=spine_04, loc=<-0.2965 9.1903 -10.8812>, rot=<0 0.5392 0.8422 0>}
 dyn_ribbonChest_a01_l {parent=base_ribbonChest_l, loc=<0.4091 0 -1.0338>, rot=<0 0 0 1>}
 dyn_ribbonChest_b01_l {parent=base_ribbonChest_l, loc=<0.4091 0 0.7249>, rot=<0 0 0 1>}
 dyn_ribbonChest_c01_l {parent=base_ribbonChest_l, loc=<0.4091 0 -0.7201>, rot=<0.0001 0.0051 -0.0275 -0.9996>}
 dyn_ribbonChest_c02_l {parent=dyn_ribbonChest_c01_l, loc=<7.4329 0 0.0318>, rot=<0 0 0 1>}
 dyn_ribbonChest_d01_l {parent=base_ribbonChest_l, loc=<0.4091 0.1995 0.4302>, rot=<0 0 -0.0524 -0.9986>}
 base_ribbonBack_c {parent=spine_04, loc=<0.8151 -9.2348 0>, rot=<0 -1 0 0>}
 dyn_ribbonBack_a01_c {parent=base_ribbonBack_c, loc=<0.8781 -0.5545 -1.1935>, rot=<0 0 0 1>}
 dyn_ribbonBack_b01_c {parent=base_ribbonBack_c, loc=<1.1939 -0.3543 1.5107>, rot=<0 0 0 1>}
 dyn_ribbonBack_c01_c {parent=base_ribbonBack_c, loc=<1.2918 -0.5557 1.2522>, rot=<-0.0003 0.0074 0.036 -0.9993>}
 dyn_ribbonBack_c02_c {parent=dyn_ribbonBack_c01_c, loc=<6.7616 0 0>, rot=<0 0 0 1>}
 dyn_ribbonBack_d01_c {parent=base_ribbonBack_c, loc=<1.3889 -0.4742 -0.767>, rot=<0 0 0 1>}
 base_ribbonChest_r {parent=spine_04, loc=<-0.2965 9.1903 10.881>, rot=<0 -0.5392 0.8422 0>}
 dyn_ribbonChest_a01_r {parent=base_ribbonChest_r, loc=<0.5497 0.1401 -0.8994>, rot=<0 0 0 1>}
 dyn_ribbonChest_b01_r {parent=base_ribbonChest_r, loc=<-0.0944 0 1.1136>, rot=<0 0.0776 0 -0.997>}
 dyn_ribbonChest_c01_r {parent=base_ribbonChest_r, loc=<0.3838 -0.0012 -0.4332>, rot=<-0.0021 0.0667 -0.0189 -0.9976>}
 dyn_ribbonChest_c02_r {parent=dyn_ribbonChest_c01_r, loc=<7.4329 0 0.0318>, rot=<0 0 0 1>}
 dyn_ribbonChest_d01_r {parent=base_ribbonChest_r, loc=<0.4348 -0.1431 1.1813>, rot=<-0.0034 0.0312 0.0051 -0.9995>}
 base_zipper {parent=spine_04, loc=<5.4475 11.6426 0>, rot=<0 0 1 0>}
 dyn_zipper {parent=base_zipper, loc=<0 0 0>, rot=<0 0 0 1>}
 spine_04_null {parent=spine_01, type=Null, loc=<20.023 -1.0996 0>, rot=<0 0 0.0255 -0.9997>}
 thigh_l {parent=pelvis, loc=<-1.0855 -0.2109 -8.7053>, rot=<0.0009 0.0355 0.0155 -0.9993>}
 calf_l {parent=thigh_l, loc=<-42.5284 -0.0019 -0.7297>, rot=<0.0006 -0.0161 -0.0391 -0.9991>}
 foot_l {parent=calf_l, loc=<-46.7401 -2.5014 -0.592>, rot=<0.0427 -0.0238 0.0353 -0.9982>}
 ball_l {parent=foot_l, loc=<-3.5658 10.8066 -0.1544>, rot=<0.0001 -0.0001 -0.7063 -0.7079>}
 calf_twist_02_l {parent=calf_l, loc=<-19.1546 -1.7943 -0.2638>, rot=<0.0237 0.0095 -0.0035 -0.9997>}
 calf_twist_01_l {parent=calf_l, loc=<-35.8483 -1.9197 -0.5794>, rot=<0.0237 0.0095 -0.0035 -0.9997>}
 deform_knee_in_base_l {parent=calf_l, loc=<0.8031 -5.0231 0.2456>, rot=<0 0 0 -1>}
 deform_knee_in_l {parent=deform_knee_in_base_l, loc=<0 0 0>, rot=<0 0 0 1>}
 deform_knee_out_base_l {parent=calf_l, loc=<0 4 0>, rot=<0 0 0 -1>}
 deform_knee_out_l {parent=deform_knee_out_base_l, loc=<0 0 0>, rot=<0 0 0 1>}
 thigh_twist_01_l {parent=thigh_l, loc=<-27.952 0 0>, rot=<0.0013 0 0 -1>}
 deform_groin_base_l {parent=thigh_l, loc=<0.4801 5.6776 -0.2486>, rot=<0 0 0 -1>}
 deform_groin_l {parent=deform_groin_base_l, loc=<0 0 0>, rot=<0 0 0 1>}
 base_ribbonLeg_l {parent=thigh_l, loc=<-35.0225 -6.888 -0.6208>, rot=<0 -1 0 0>}
 dyn_ribbonLeg_a01_l {parent=base_ribbonLeg_l, loc=<0.9188 -0.831 -0.7758>, rot=<0 -0.0676 0 -0.9977>}
 dyn_ribbonLeg_b01_l {parent=base_ribbonLeg_l, loc=<0.9252 -0.7622 1.3188>, rot=<0 0.0223 0 -0.9998>}
 dyn_ribbonLeg_c01_l {parent=base_ribbonLeg_l, loc=<0.8507 -0.7918 0.7541>, rot=<-0.0015 0.0223 0.0658 -0.9976>}
 dyn_ribbonLeg_c02_l {parent=dyn_ribbonLeg_c01_l, loc=<5.5458 0 0>, rot=<0 0 0 1>}
 dyn_ribbonLeg_d01_l {parent=base_ribbonLeg_l, loc=<0.9713 -0.5473 -0.3897>, rot=<0 -0.0676 0 -0.9977>}
 thigh_r {parent=pelvis, loc=<-1.0855 -0.2109 8.7053>, rot=<0.0355 -0.0009 0.9993 0.0155>}
 calf_r {parent=thigh_r, loc=<42.5284 0.0019 0.7297>, rot=<0.0006 -0.0161 -0.0391 -0.9991>}
 foot_r {parent=calf_r, loc=<46.7401 2.5014 0.592>, rot=<0.0427 -0.0238 0.0353 -0.9982>}
 ball_r {parent=foot_r, loc=<3.5658 -10.8066 0.1544>, rot=<0.0001 -0.0001 -0.7063 -0.7079>}
 calf_twist_02_r {parent=calf_r, loc=<19.1546 1.7943 0.2638>, rot=<0.0237 0.0095 -0.0035 -0.9997>}
 calf_twist_01_r {parent=calf_r, loc=<35.8483 1.9197 0.5794>, rot=<0.0237 0.0095 -0.0035 -0.9997>}
 deform_knee_in_base_r {parent=calf_r, loc=<-0.8031 5.0231 -0.2457>, rot=<0 0 0 -1>}
 deform_knee_in_r {parent=deform_knee_in_base_r, loc=<0 0 0>, rot=<0 0 0 1>}
 deform_knee_out_base_r {parent=calf_r, loc=<0 -4 0>, rot=<0 0 0 -1>}
 deform_knee_out_r {parent=deform_knee_out_base_r, loc=<0 0 0>, rot=<0 0 0 1>}
 thigh_twist_01_r {parent=thigh_r, loc=<27.952 0 0>, rot=<0.0013 0 0 -1>}
 deform_groin_base_r {parent=thigh_r, loc=<-0.4801 -5.6776 0.2486>, rot=<0 0 0 -1>}
 deform_groin_r {parent=deform_groin_base_r, loc=<0 0 0>, rot=<0 0 0 1>}
 base_ribbonLeg_r {parent=thigh_r, loc=<35.0225 6.888 0.6205>, rot=<0.9999 -0.0001 0.0163 -0.0007>}
 dyn_ribbonLeg_a01_r {parent=base_ribbonLeg_r, loc=<1.0453 -0.8323 -1.1327>, rot=<0 0.0101 0 -0.9999>}
 dyn_ribbonLeg_b01_r {parent=base_ribbonLeg_r, loc=<0.8637 -0.7206 1.0978>, rot=<0 0.0223 0 -0.9998>}
 dyn_ribbonLeg_c01_r {parent=base_ribbonLeg_r, loc=<0.8507 -0.7918 0.7541>, rot=<0 0.0447 0.0658 -0.9968>}
 dyn_ribbonLeg_c02_r {parent=dyn_ribbonLeg_c01_r, loc=<5.5458 0 0>, rot=<0 0 0 1>}
 dyn_ribbonLeg_d01_r {parent=base_ribbonLeg_r, loc=<0.9767 -0.641 -0.6537>, rot=<0 0.0101 0 -0.9999>}
 deform_glute_base_l {parent=pelvis, loc=<-7.3466 -9.0213 -9.1245>, rot=<0.0172 0.031 -0.4477 -0.8935>}
 deform_glute_l {parent=deform_glute_base_l, loc=<0 0 0>, rot=<0 0 0 1>}
 deform_glute_base_r {parent=pelvis, loc=<-7.3464 -9.0213 9.1245>, rot=<0.031 -0.0172 0.8935 -0.4477>}
 deform_glute_r {parent=deform_glute_base_r, loc=<0 0 0>, rot=<0 0 0 1>}
 base_skirt_fr_c {parent=pelvis, loc=<-5.2823 10.5444 0.0137>, rot=<0.0041 -0.0143 0.9959 0.0891>}
 dyn_skirt_fr_c_01 {parent=base_skirt_fr_c, loc=<0 0 0>, rot=<0 0 0 1>}
 dyn_skirt_fr_c_02 {parent=dyn_skirt_fr_c_01, loc=<5.9868 0 0>, rot=<0 0 0 1>}
 base_skirt_sd_l {parent=pelvis, loc=<0.4733 -1.4621 -15.6195>, rot=<0.1411 -0.6929 -0.6929 -0.1411>}
 dyn_skirt_sd_l_01 {parent=base_skirt_sd_l, loc=<0 0 0>, rot=<0 0 0 1>}
 dyn_skirt_sd_l_02 {parent=dyn_skirt_sd_l_01, loc=<9.4784 0 0>, rot=<0 0 0 1>}
 base_skirt_bk_c {parent=pelvis, loc=<-0.8598 -14.3871 0>, rot=<0.0853 -0.9964 0 0>}
 dyn_skirt_bk_c_01 {parent=base_skirt_bk_c, loc=<0 0 0>, rot=<0 0 0 1>}
 dyn_skirt_bk_c_02 {parent=dyn_skirt_bk_c_01, loc=<9.7259 0 0>, rot=<0 0 0 1>}
 base_skirt_sd_r {parent=pelvis, loc=<0.4711 -1.4621 15.667>, rot=<0.1567 -0.6895 0.6895 0.1567>}
 dyn_skirt_sd_r_01 {parent=base_skirt_sd_r, loc=<0 0 0>, rot=<0 0 0 1>}
 dyn_skirt_sd_r_02 {parent=dyn_skirt_sd_r_01, loc=<9.4784 0 0>, rot=<0 0 0 1>}
 base_skirt_bk_l {parent=pelvis, loc=<-0.8671 -11.9695 -11.4846>, rot=<0.0853 -0.9658 -0.2285 -0.0877>}
 dyn_skirt_bk_l_01 {parent=base_skirt_bk_l, loc=<0 0 0>, rot=<0 0 0 1>}
 dyn_skirt_bk_l_02 {parent=dyn_skirt_bk_l_01, loc=<8.8558 0 0>, rot=<0 0 0 1>}
 base_skirt_bk_r {parent=pelvis, loc=<-0.9134 -11.8084 11.468>, rot=<0.1268 -0.9573 0.2478 0.0788>}
 dyn_skirt_bk_r_01 {parent=base_skirt_bk_r, loc=<0 0 0>, rot=<0 0 0 1>}
 dyn_skirt_bk_r_02 {parent=dyn_skirt_bk_r_01, loc=<8.8558 0 0>, rot=<0 0 0 1>}
 base_skirt_fr_l {parent=pelvis, loc=<-1.0452 6.9758 -11.1313>, rot=<-0.1084 0.1687 0.9734 0.1111>}
 dyn_skirt_fr_l_01 {parent=base_skirt_fr_l, loc=<0 0 0>, rot=<0 0 0 1>}
 dyn_skirt_fr_l_02 {parent=dyn_skirt_fr_l_01, loc=<8.9692 0 0>, rot=<0 0 0 1>}
 base_skirt_fr_r {parent=pelvis, loc=<-0.9672 6.9932 11.2995>, rot=<0.0895 -0.2614 0.9506 0.1417>}
 dyn_skirt_fr_r_01 {parent=base_skirt_fr_r, loc=<0 0 0>, rot=<0 0 0 1>}
 dyn_skirt_fr_r_02 {parent=dyn_skirt_fr_r_01, loc=<8.9692 0 0>, rot=<0 0 0 1>}
 thigh_r_null {parent=pelvis, type=Null, loc=<-1.0855 -0.2109 8.7053>, rot=<0.0355 -0.0009 0.9993 0.0155>}
 thigh_l_null {parent=pelvis, type=Null, loc=<-1.0855 -0.2109 -8.7053>, rot=<0.0009 0.0355 0.0155 -0.9993>}
 thigh_l_ctrl {controlType=EULER_TRANSFORM, type=Control, loc=<0 0 0>, rot=<0 0 0 1>}
 thigh_r_ctrl {controlType=EULER_TRANSFORM, type=Control, loc=<0 0 0>, rot=<0 0 0 1>}
 spine_04_ctrl {controlType=EULER_TRANSFORM, type=Control, loc=<0 0 0>, rot=<0 0 0 1>}
]
variables [
 Thigh_L_BackRotation {type=double, default=0}
 Thigh_L_FrontRotation {type=double, default=0}
 Thigh_L_SideRotation {type=double, default=0}
 Thigh_R_BackRotation {type=double, default=0}
 Thigh_R_FrontRotation {type=double, default=0}
 Thigh_R_SideRotation {type=double, default=0}
 extremeThighRotation {type=double, default=0}
 isCrouching_CR {type=bool, default=False}
 isEmoting_CR {type=bool, default=False}
 isKneeSliding_CR {type=bool, default=False}
 isOnGrindRail_CR {type=bool, default=False}
 spineRotZneg {type=double, default=0}
 spineRotZpos {type=double, default=0}
]
