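# Physics asset for the ToothGreek medium female body cosmetic, set up to simulate a ribbon on the right arm.
# Four strands of capsule bodies hang from the ribbon base through swing and twist limited constraints.
# The C strand adds a second segment with tighter swing and locked twist.
# Upper and lower right arm capsules have no constraints, apparently serving as colliders for the ribbons.
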PhysicsAsset /BRCosmetics/Characters/Player/Female/Medium/Bodies/F_MED_ToothGreek/Meshes/F_MED_ToothGreek_Physics
bodies {
 base_ribbonArm_r [{
   shapeType=Capsule
   length=1.9849
   radius=1.2626
   shapeName=base_ribbonArm_r_capsule
   center=<0.6206 0.0306 -0.5429>
   rotation=<rot 0.4661 -90 -19.5859>
  }]
 dyn_ribbonArm_a01_r [{
   shapeType=Capsule
   length=6.5312
   radius=1.1292
   shapeName=dyn_ribbonArm_b01_r_capsule
   center=<4.7651 0.042 1.8816>
   rotation=<rot 27.6194 -87.5411 -107.697>
  }]
 dyn_ribbonArm_b01_r [{
   shapeType=Capsule
   length=4.7494
   radius=1.1292
   shapeName=dyn_ribbonArm_b01_r_capsule
   center=<4.2992 -0.1771 -0.5657>
   rotation=<rot -0.4101 -78.2787 -83.304>
  }]
 dyn_ribbonArm_c01_r [{
   shapeType=Capsule
   length=5.3597
   radius=1.1292
   shapeName=dyn_ribbonArm_b01_r_capsule
   center=<4.4495 -0.7899 -0.1523>
   rotation=<rot 3.4299 -87.7364 -91.9374>
  }]
 dyn_ribbonArm_c02_r [{
   shapeType=Capsule
   length=4.605
   radius=1.1292
   shapeName=dyn_ribbonArm_b01_r_capsule
   center=<3.7974 -0.6876 0.0834>
   rotation=<rot 3.4299 -87.7364 -91.9374>
  }]
 dyn_ribbonArm_d01_r [{
   shapeType=Capsule
   length=7.4341
   radius=1.1292
   shapeName=dyn_ribbonArm_b01_r_capsule
   center=<5.3783 -0.0471 -0.0046>
   rotation=<rot -0.579 -81.3469 -94.4796>
  }]
 lowerarm_r [{shapeType=Capsule, length=16.7071, radius=4.5089, center=<-7.8642 -0.2306 0.428>, rotation=<rot 0.0083 -91.0079 -87.3702>}]
 upperarm_r [{shapeType=Capsule, length=15.7317, radius=5.0276, center=<-15.2118 0.5943 1.0834>, rotation=<rot 0.0004 88.2176 -86.7478>}]
}
constraints [
 {bone1Name=dyn_ribbonArm_a01_r, bone2Name=base_ribbonArm_r, swing1LimitDegrees=35, swing2LimitDegrees=45, twistLimitDegrees=45}
 {bone1Name=dyn_ribbonArm_b01_r, bone2Name=base_ribbonArm_r, swing1LimitDegrees=35, swing2LimitDegrees=25, twistLimitDegrees=45}
 {bone1Name=dyn_ribbonArm_c01_r, bone2Name=base_ribbonArm_r, swing1LimitDegrees=35, swing2LimitDegrees=35, twistLimitDegrees=45}
 {
  bone1Name=dyn_ribbonArm_c02_r
  bone2Name=dyn_ribbonArm_c01_r
  swing1LimitDegrees=25
  swing2LimitDegrees=15
  twistLimitDegrees=45
  twistMotion=Locked
 }
 {bone1Name=dyn_ribbonArm_d01_r, bone2Name=base_ribbonArm_r, swing1LimitDegrees=35, swing2LimitDegrees=35, twistLimitDegrees=45}
]
