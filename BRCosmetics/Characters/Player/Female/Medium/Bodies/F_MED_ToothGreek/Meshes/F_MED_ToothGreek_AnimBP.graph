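# Post-process style animation blueprint for the F_MED_ToothGreek cosmetic body, animating its ribbons, skirt and zipper.
# It copies the parent mesh pose, positions base bones with weighted constraints, then runs a Control Rig fed crouch, slide, grind and emote states.
# Anim Dynamics chains add spring-driven secondary motion, with two differently tuned skirt variants from that cached pose.
# The output selects tighter back-ribbon limits when a backbling or cape is worn.
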AnimBlueprint /BRCosmetics/Characters/Player/Female/Medium/Bodies/F_MED_ToothGreek/Meshes/F_MED_ToothGreek_AnimBP
target_skeleton=&/F_MED_ToothGreek_Skeleton
graphs [
 AnimGraph {nodes [
   n1 = Anim Dynamics - Bone: dyn_ribbonArm_a01_r!(ComponentPose, Alpha=1, GravityOverride:Vector="0,0,-980", ExternalForce="0,0,0") -> (Pose) {
    angularDampingOverride=0.999
    angularSpringConstant=3000
    bAngularSpring=true
    bOverrideAngularDamping=true
    bUseGravityOverride=true
    gravityScale=0.35
    simulationSpace=RootRelative
    boundBone {boneName=dyn_ribbonArm_a01_r}
    componentAppliedLinearAccClamp=<3000 3000 3000>
    componentLinearAccScale=<0.3 0.3 0.6>
    componentLinearVelScale=<0.3 0.3 0.6>
    gravityOverride {z=-980}
    physicsBodyDefinitions [{
      boundBone {boneName=dyn_ribbonArm_a01_r}
      boxExtents=<16 16 16>
      constraintSetup {angularLimitsMax=<0 55 -3>, angularLimitsMin=<0 -70 -70>, angularTarget=<1 0 0>}
      localJointOffset=<3 0 0>
     }]
    relativeSpaceBone {boneName=upperarm_r}
    simSpaceSettings {maxAngularAcceleration=3000, maxAngularVelocity=3000, simSpaceAngularAlpha=0.1}
   }
   n2 = Anim Dynamics - Bone: dyn_ribbonArm_b01_r!(ComponentPose=@n1, Alpha=1, GravityOverride:Vector="0,0,-980", ExternalForce="0,0,0") -> (Pose) {
    angularDampingOverride=0.999
    angularSpringConstant=3000
    bAngularSpring=true
    bOverrideAngularDamping=true
    bUseGravityOverride=true
    gravityScale=0.35
    simulationSpace=RootRelative
    boundBone {boneName=dyn_ribbonArm_b01_r}
    componentAppliedLinearAccClamp=<3000 3000 3000>
    componentLinearAccScale=<0.3 0.3 0.6>
    componentLinearVelScale=<0.3 0.3 0.6>
    gravityOverride {z=-980}
    physicsBodyDefinitions [{
      boundBone {boneName=dyn_ribbonArm_b01_r}
      boxExtents=<16 16 16>
      constraintSetup {angularLimitsMax=<0 70 -7.5>, angularLimitsMin=<0 -45 -70>, angularTarget=<1 0 0>}
      localJointOffset=<3 0 0>
     }]
    relativeSpaceBone {boneName=upperarm_r}
    simSpaceSettings {maxAngularAcceleration=3000, maxAngularVelocity=3000, simSpaceAngularAlpha=0.1}
   }
   n3 = Anim Dynamics - Bone: dyn_ribbonArm_c01_r!(ComponentPose=@n2, Alpha=1, GravityOverride:Vector="0,0,-980", ExternalForce="0,0,0") -> (Pose) {
    angularDampingOverride=0.999
    angularSpringConstant=3000
    bAngularSpring=true
    bOverrideAngularDamping=true
    bUseGravityOverride=true
    gravityScale=0.35
    simulationSpace=RootRelative
    boundBone {boneName=dyn_ribbonArm_c01_r}
    componentAppliedLinearAccClamp=<3000 3000 3000>
    componentLinearAccScale=<0.3 0.3 0.6>
    componentLinearVelScale=<0.3 0.3 0.6>
    gravityOverride {z=-980}
    physicsBodyDefinitions [{
      boundBone {boneName=dyn_ribbonArm_c01_r}
      boxExtents=<16 16 16>
      constraintSetup {angularLimitsMax=<0 70 0>, angularLimitsMin=<0 -70 -70>, angularTarget=<1 0 0>}
      localJointOffset=<3 0 0>
     }]
    relativeSpaceBone {boneName=upperarm_r}
    simSpaceSettings {maxAngularAcceleration=3000, maxAngularVelocity=3000, simSpaceAngularAlpha=0.1}
   }
   n4 = Anim Dynamics - Bone: dyn_ribbonArm_c02_r!(ComponentPose=@n3, Alpha=1, GravityOverride:Vector="0,0,-980", ExternalForce="0,0,0") -> (Pose) {
    angularDampingOverride=0.999
    angularSpringConstant=3000
    bAngularSpring=true
    bOverrideAngularDamping=true
    bUseGravityOverride=true
    gravityScale=0.35
    simulationSpace=RootRelative
    boundBone {boneName=dyn_ribbonArm_c02_r}
    componentAppliedLinearAccClamp=<3000 3000 3000>
    componentLinearAccScale=<0.3 0.3 0.3>
    componentLinearVelScale=<0.3 0.3 0.3>
    gravityOverride {z=-980}
    physicsBodyDefinitions [{
      boundBone {boneName=dyn_ribbonArm_c02_r}
      boxExtents=<16 16 16>
      constraintSetup {angularLimitsMax=<0 70 0>, angularLimitsMin=<0 -70 -40>, angularTarget=<1 0 0>}
      localJointOffset=<1 0 0>
     }]
    relativeSpaceBone {boneName=upperarm_r}
    simSpaceSettings {maxAngularAcceleration=3000, maxAngularVelocity=3000, simSpaceAngularAlpha=0.1}
   }
   Anim Dynamics - Bone: dyn_ribbonArm_d01_r!(ComponentPose=@n4, Alpha=1, GravityOverride:Vector="0,0,-980", ExternalForce="0,0,0") -> (Pose) {
    angularDampingOverride=0.999
    angularSpringConstant=3000
    bAngularSpring=true
    bOverrideAngularDamping=true
    bUseGravityOverride=true
    gravityScale=0.35
    simulationSpace=RootRelative
    boundBone {boneName=dyn_ribbonArm_d01_r}
    componentAppliedLinearAccClamp=<3000 3000 3000>
    componentLinearAccScale=<0.3 0.3 0.6>
    componentLinearVelScale=<0.3 0.3 0.6>
    gravityOverride {z=-980}
    physicsBodyDefinitions [{
      boundBone {boneName=dyn_ribbonArm_d01_r}
      boxExtents=<16 16 16>
      constraintSetup {angularLimitsMax=<0 70 -2>, angularLimitsMin=<0 -50 -70>, angularTarget=<1 0 0>}
      localJointOffset=<3 0 0>
     }]
    relativeSpaceBone {boneName=upperarm_r}
    simSpaceSettings {maxAngularAcceleration=3000, maxAngularVelocity=3000, simSpaceAngularAlpha=0.1}
   }
   n5 = Anim Dynamics - Bone: dyn_ribbonBack_a01_c!(
    ComponentPose=Local To Component(LocalPose=Use cached pose 'ribbonAnimDyn'())
    Alpha=1
    GravityOverride:Vector="0,0,-980"
    ExternalForce="0,0,0"
   ) -> (Pose) {
    angularDampingOverride=0.999
    angularSpringConstant=3000
    bAngularSpring=true
    bOverrideAngularDamping=true
    bUseGravityOverride=true
    gravityScale=0.35
    simulationSpace=RootRelative
    boundBone {boneName=dyn_ribbonBack_a01_c}
    componentAppliedLinearAccClamp=<3000 3000 3000>
    componentLinearAccScale=<0.3 0.3 0.5>
    componentLinearVelScale=<0.3 0.3 0.5>
    gravityOverride {z=-980}
    physicsBodyDefinitions [{
      boundBone {boneName=dyn_ribbonBack_a01_c}
      boxExtents=<16 16 16>
      constraintSetup {angularLimitsMax=<0 55 6>, angularLimitsMin=<0 -2 -60>, angularTarget=<1 0 0>}
      localJointOffset=<2 0 -2>
     }]
    relativeSpaceBone {boneName=spine_05}
    simSpaceSettings {simSpaceAngularAlpha=0.1}
   }
   n6 = Anim Dynamics - Bone: dyn_ribbonBack_b01_c!(ComponentPose=@n5, Alpha=1, GravityOverride:Vector="0,0,-980", ExternalForce="0,0,0") -> (Pose) {
    angularDampingOverride=0.999
    angularSpringConstant=3000
    bAngularSpring=true
    bOverrideAngularDamping=true
    bUseGravityOverride=true
    gravityScale=0.35
    simulationSpace=RootRelative
    boundBone {boneName=dyn_ribbonBack_b01_c}
    componentAppliedLinearAccClamp=<3000 3000 3000>
    componentLinearAccScale=<0.3 0.3 0.5>
    componentLinearVelScale=<0.3 0.3 0.5>
    gravityOverride {z=-980}
    physicsBodyDefinitions [{
      boundBone {boneName=dyn_ribbonBack_b01_c}
      boxExtents=<16 16 16>
      constraintSetup {angularLimitsMax=<0 2 8>, angularLimitsMin=<0 -55 -60>, angularTarget=<1 0 0>}
      localJointOffset=<2 0 2>
     }]
    relativeSpaceBone {boneName=spine_05}
    simSpaceSettings {simSpaceAngularAlpha=0.1}
   }
   n7 = Anim Dynamics - Bone: dyn_ribbonBack_c01_c!(ComponentPose=@n6, Alpha=1, GravityOverride:Vector="0,0,-980", ExternalForce="0,0,0") -> (Pose) {
    angularDampingOverride=0.999
    angularSpringConstant=3000
    bAngularSpring=true
    bOverrideAngularDamping=true
    bUseGravityOverride=true
    gravityScale=0.35
    simulationSpace=RootRelative
    boundBone {boneName=dyn_ribbonBack_c01_c}
    componentAppliedLinearAccClamp=<3000 3000 3000>
    componentLinearAccScale=<0.3 0.3 0.5>
    componentLinearVelScale=<0.3 0.3 0.5>
    gravityOverride {z=-980}
    physicsBodyDefinitions [{
      boundBone {boneName=dyn_ribbonBack_c01_c}
      boxExtents=<16 16 16>
      constraintSetup {angularLimitsMax=<0 2 8>, angularLimitsMin=<0 -55 -60>, angularTarget=<1 0 0>}
      localJointOffset=<2 0 2>
     }]
    relativeSpaceBone {boneName=spine_05}
    simSpaceSettings {simSpaceAngularAlpha=0.1}
   }
   n8 = Anim Dynamics - Bone: dyn_ribbonBack_c02_c!(ComponentPose=@n7, Alpha=1, GravityOverride:Vector="0,0,-980", ExternalForce="0,0,0") -> (Pose) {
    angularDampingOverride=0.999
    angularSpringConstant=3000
    bAngularSpring=true
    bOverrideAngularDamping=true
    bUseGravityOverride=true
    gravityScale=0.35
    simulationSpace=RootRelative
    boundBone {boneName=dyn_ribbonBack_c02_c}
    componentAppliedLinearAccClamp=<3000 3000 3000>
    componentLinearAccScale=<0.3 0.3 0.5>
    componentLinearVelScale=<0.3 0.3 0.5>
    gravityOverride {z=-980}
    physicsBodyDefinitions [{
      boundBone {boneName=dyn_ribbonBack_c02_c}
      boxExtents=<16 16 16>
      constraintSetup {angularLimitsMax=<0 25 2>, angularLimitsMin=<0 -25 -60>, angularTarget=<1 0 0>}
      localJointOffset=<2 0 0>
     }]
    relativeSpaceBone {boneName=spine_05}
    simSpaceSettings {simSpaceAngularAlpha=0.1}
   }
   n9 = Anim Dynamics - Bone: dyn_ribbonBack_d01_c!(ComponentPose=@n8, Alpha=1, GravityOverride:Vector="0,0,-980", ExternalForce="0,0,0") -> (Pose) {
    angularDampingOverride=0.999
    angularSpringConstant=3000
    bAngularSpring=true
    bOverrideAngularDamping=true
    bUseGravityOverride=true
    gravityScale=0.35
    simulationSpace=RootRelative
    boundBone {boneName=dyn_ribbonBack_d01_c}
    componentAppliedLinearAccClamp=<3000 3000 3000>
    componentLinearAccScale=<0.3 0.3 0.5>
    componentLinearVelScale=<0.3 0.3 0.5>
    gravityOverride {z=-980}
    physicsBodyDefinitions [{
      boundBone {boneName=dyn_ribbonBack_d01_c}
      boxExtents=<16 16 16>
      constraintSetup {angularLimitsMax=<0 55 6>, angularLimitsMin=<0 -2 -60>, angularTarget=<1 0 0>}
      localJointOffset=<2 0 -2>
     }]
    relativeSpaceBone {boneName=spine_05}
    simSpaceSettings {simSpaceAngularAlpha=0.1}
   }
   n10 = Anim Dynamics - Bone: dyn_ribbonBack_a01_c(ComponentPose=Local To Component(LocalPose=Use cached pose 'ribbonAnimDyn'()), Alpha=1, GravityOverride="0,0,-980", ExternalForce="0,0,0") {
    angularDampingOverride=0.999
    angularSpringConstant=3000
    bAngularSpring=true
    bOverrideAngularDamping=true
    bUseGravityOverride=true
    gravityScale=0.35
    simulationSpace=RootRelative
    boundBone {boneName=dyn_ribbonBack_a01_c}
    componentAppliedLinearAccClamp=<3000 3000 3000>
    componentLinearAccScale=<0.3 0.3 0.5>
    componentLinearVelScale=<0.3 0.3 0.5>
    gravityOverride {z=-980}
    physicsBodyDefinitions [{
      boundBone {boneName=dyn_ribbonBack_a01_c}
      boxExtents=<16 16 16>
      constraintSetup {angularLimitsMax=<0 55 6>, angularLimitsMin=<0 -2 -10>, angularTarget=<1 0 0>}
      localJointOffset=<2 0 -2>
     }]
    relativeSpaceBone {boneName=spine_05}
    simSpaceSettings {simSpaceAngularAlpha=0.1}
   }
   n11 = Anim Dynamics - Bone: dyn_ribbonBack_b01_c(ComponentPose=@n10, Alpha=1, GravityOverride="0,0,-980", ExternalForce="0,0,0") {
    angularDampingOverride=0.999
    angularSpringConstant=3000
    bAngularSpring=true
    bOverrideAngularDamping=true
    bUseGravityOverride=true
    gravityScale=0.35
    simulationSpace=RootRelative
    boundBone {boneName=dyn_ribbonBack_b01_c}
    componentAppliedLinearAccClamp=<3000 3000 3000>
    componentLinearAccScale=<0.3 0.3 0.5>
    componentLinearVelScale=<0.3 0.3 0.5>
    gravityOverride {z=-980}
    physicsBodyDefinitions [{
      boundBone {boneName=dyn_ribbonBack_b01_c}
      boxExtents=<16 16 16>
      constraintSetup {angularLimitsMax=<0 2 8>, angularLimitsMin=<0 -55 -10>, angularTarget=<1 0 0>}
      localJointOffset=<2 0 2>
     }]
    relativeSpaceBone {boneName=spine_05}
    simSpaceSettings {simSpaceAngularAlpha=0.1}
   }
   n12 = Anim Dynamics - Bone: dyn_ribbonBack_c01_c(ComponentPose=@n11, Alpha=1, GravityOverride="0,0,-980", ExternalForce="0,0,0") {
    angularDampingOverride=0.999
    angularSpringConstant=3000
    bAngularSpring=true
    bOverrideAngularDamping=true
    bUseGravityOverride=true
    gravityScale=0.35
    simulationSpace=RootRelative
    boundBone {boneName=dyn_ribbonBack_c01_c}
    componentAppliedLinearAccClamp=<3000 3000 3000>
    componentLinearAccScale=<0.3 0.3 0.5>
    componentLinearVelScale=<0.3 0.3 0.5>
    gravityOverride {z=-980}
    physicsBodyDefinitions [{
      boundBone {boneName=dyn_ribbonBack_c01_c}
      boxExtents=<16 16 16>
      constraintSetup {angularLimitsMax=<0 2 8>, angularLimitsMin=<0 -55 -10>, angularTarget=<1 0 0>}
      localJointOffset=<2 0 2>
     }]
    relativeSpaceBone {boneName=spine_05}
    simSpaceSettings {simSpaceAngularAlpha=0.1}
   }
   n13 = Anim Dynamics - Bone: dyn_ribbonBack_c02_c(ComponentPose=@n12, Alpha=1, GravityOverride="0,0,-980", ExternalForce="0,0,0") {
    angularDampingOverride=0.999
    angularSpringConstant=3000
    bAngularSpring=true
    bOverrideAngularDamping=true
    bUseGravityOverride=true
    gravityScale=0.35
    simulationSpace=RootRelative
    boundBone {boneName=dyn_ribbonBack_c02_c}
    componentAppliedLinearAccClamp=<3000 3000 3000>
    componentLinearAccScale=<0.3 0.3 0.5>
    componentLinearVelScale=<0.3 0.3 0.5>
    gravityOverride {z=-980}
    physicsBodyDefinitions [{
      boundBone {boneName=dyn_ribbonBack_c02_c}
      boxExtents=<16 16 16>
      constraintSetup {angularLimitsMax=<0 25 2>, angularLimitsMin=<0 -25 -10>, angularTarget=<1 0 0>}
      localJointOffset=<2 0 0>
     }]
    relativeSpaceBone {boneName=spine_05}
    simSpaceSettings {simSpaceAngularAlpha=0.1}
   }
   n14 = Anim Dynamics - Bone: dyn_ribbonBack_d01_c(ComponentPose=@n13, Alpha=1, GravityOverride="0,0,-980", ExternalForce="0,0,0") {
    angularDampingOverride=0.999
    angularSpringConstant=3000
    bAngularSpring=true
    bOverrideAngularDamping=true
    bUseGravityOverride=true
    gravityScale=0.35
    simulationSpace=RootRelative
    boundBone {boneName=dyn_ribbonBack_d01_c}
    componentAppliedLinearAccClamp=<3000 3000 3000>
    componentLinearAccScale=<0.3 0.3 0.5>
    componentLinearVelScale=<0.3 0.3 0.5>
    gravityOverride {z=-980}
    physicsBodyDefinitions [{
      boundBone {boneName=dyn_ribbonBack_d01_c}
      boxExtents=<16 16 16>
      constraintSetup {angularLimitsMax=<0 55 6>, angularLimitsMin=<0 -2 -10>, angularTarget=<1 0 0>}
      localJointOffset=<2 0 -2>
     }]
    relativeSpaceBone {boneName=spine_05}
    simSpaceSettings {simSpaceAngularAlpha=0.1}
   }
   Output Pose(Result=Blend Poses by bool(
     BlendPose_0=Component To Local(ComponentPose=@n14)
     BlendPose_1=Component To Local(ComponentPose=@n9)
     bActiveValue=Get isWearingBackbling / Cape / NoRBAN() -> (isWearingBackbling / Cape / NoRBAN:bool)
    )) {layerGroup=None}
   n15 = "Constraint - base_ribbonChest_r (PPPPPP)"!(
    ComponentPose=Local To Component(LocalPose=Copy Pose From Mesh() {bUseAttachedParent=true})
    Alpha=1
    ConstraintWeights_0=0.57
    ConstraintWeights_1:float=0.2508
    ConstraintWeights_2:float=0.0655
    ConstraintWeights_3:float=0.0412
    ConstraintWeights_4:float=0.0384
    ConstraintWeights_5:float=0.0341
   ) -> (Pose) {
    boneToModify {boneName=base_ribbonChest_r}
    constraintSetup [
     {targetBone {boneName=spine_04}}
     {targetBone {boneName=deform_pec_r}}
     {targetBone {boneName=spine_05}}
     {targetBone {boneName=clavicle_r}}
     {targetBone {boneName=spine_02}}
     {targetBone {boneName=spine_03}}
    ]
    constraintWeights [1,1,1,1,1,1]
   }
   n16 = "Constraint - base_ribbonChest_l (PPPPPP)"!(
    ComponentPose=@n15
    Alpha=1
    ConstraintWeights_0=0.57
    ConstraintWeights_1:float=0.2508
    ConstraintWeights_2:float=0.0655
    ConstraintWeights_3:float=0.0412
    ConstraintWeights_4:float=0.0384
    ConstraintWeights_5:float=0.0341
   ) -> (Pose) {
    boneToModify {boneName=base_ribbonChest_l}
    constraintSetup [
     {targetBone {boneName=spine_04}}
     {targetBone {boneName=deform_pec_l}}
     {targetBone {boneName=spine_05}}
     {targetBone {boneName=clavicle_l}}
     {targetBone {boneName=spine_02}}
     {targetBone {boneName=spine_03}}
    ]
    constraintWeights [1,1,1,1,1,1]
   }
   n17 = "Constraint - base_ribbonBack_c (PPPPPPPP)"!(
    ComponentPose=@n16
    Alpha=1
    ConstraintWeights_0=0.5821
    ConstraintWeights_1:float=0.1646
    ConstraintWeights_2:float=0.0679
    ConstraintWeights_3:float=0.0579
    ConstraintWeights_4:float=0.0549
    ConstraintWeights_5:float=0.0543
    ConstraintWeights_6:float=0.0182
    ConstraintWeights_7:float=0.0001
   ) -> (Pose) {
    boneToModify {boneName=base_ribbonBack_c}
    constraintSetup [
     {targetBone {boneName=spine_04}}
     {targetBone {boneName=spine_05}}
     {targetBone {boneName=clavicle_r}}
     {targetBone {boneName=spine_02}}
     {targetBone {boneName=spine_03}}
     {targetBone {boneName=clavicle_l}}
     {targetBone {boneName=spine_01}}
     {targetBone {boneName=upperarm_twist_01_l}}
    ]
    constraintWeights [1,1,1,1,1,1,1,1]
   }
   n18 = "Constraint - base_ribbonArm_r (PPP)"!(ComponentPose=@n17, Alpha=1, ConstraintWeights_0=0.5435, ConstraintWeights_1:float=0.2385, ConstraintWeights_2:float=0.2179) -> (Pose) {
    boneToModify {boneName=base_ribbonArm_r}
    constraintSetup [{targetBone {boneName=upperarm_twist_01_r}}, {targetBone {boneName=upperarm_r}}, {targetBone {boneName=upperarm_twist_02_r}}]
    constraintWeights [1,1,1]
   }
   n19 = "Constraint - base_ribbonLeg_r (PPPP)"!(
    ComponentPose=@n18
    Alpha=1
    ConstraintWeights_0=0.8575
    ConstraintWeights_1:float=0.0812
    ConstraintWeights_2:float=0.0545
    ConstraintWeights_3:float=0.0067
   ) -> (Pose) {
    boneToModify {boneName=base_ribbonLeg_r}
    constraintSetup [
     {targetBone {boneName=thigh_r}}
     {targetBone {boneName=deform_knee_in_r}}
     {targetBone {boneName=calf_r}}
     {targetBone {boneName=thigh_twist_01_r}}
    ]
    constraintWeights [1,1,1,1]
   }
   n20 = "Constraint - base_ribbonLeg_l (PPPP)"!(
    ComponentPose=@n19
    Alpha=1
    ConstraintWeights_0=0.8575
    ConstraintWeights_1:float=0.0812
    ConstraintWeights_2:float=0.0545
    ConstraintWeights_3:float=0.0067
   ) -> (Pose) {
    boneToModify {boneName=base_ribbonLeg_l}
    constraintSetup [
     {targetBone {boneName=thigh_l}}
     {targetBone {boneName=deform_knee_in_l}}
     {targetBone {boneName=calf_l}}
     {targetBone {boneName=thigh_twist_01_l}}
    ]
    constraintWeights [1,1,1,1]
   }
   n21 = "Constraint - base_zipper (PPPPPP)"!(
    ComponentPose=@n20
    Alpha=1
    ConstraintWeights_0=0.5985
    ConstraintWeights_1:float=0.3772
    ConstraintWeights_2:float=0.0115
    ConstraintWeights_3:float=0.0115
    ConstraintWeights_4:float=0.0006
    ConstraintWeights_5:float=0.0006
   ) -> (Pose) {
    boneToModify {boneName=base_zipper}
    constraintSetup [
     {targetBone {boneName=spine_04}}
     {targetBone {boneName=spine_05}}
     {targetBone {boneName=deform_pec_r}}
     {targetBone {boneName=deform_pec_l}}
     {targetBone {boneName=clavicle_r}}
     {targetBone {boneName=clavicle_l}}
    ]
    constraintWeights [1,1,1,1,1,1]
   }
   n22 = "Constraint - base_ribbonLeg_l (R)"(ComponentPose=@n21, Alpha=1, ConstraintWeights_0=1) {boneToModify {boneName=base_ribbonLeg_l}, constraintSetup [{transformType=Rotation, targetBone {boneName=calf_l}}], constraintWeights [1]}
   n23 = "Constraint - base_ribbonLeg_r (R)"(ComponentPose=@n22, Alpha=1, ConstraintWeights_0=1) {boneToModify {boneName=base_ribbonLeg_r}, constraintSetup [{transformType=Rotation, targetBone {boneName=calf_r}}], constraintWeights [1]}
   n24 = Control Rig!(
    Source=Component To Local(ComponentPose=@n23)
    Alpha
    isCrouching_CR:bool=Get bIsCrouching()
    isKneeSliding_CR:bool=Get bIsKneeSliding()
    isOnGrindRail_CR:bool=Get bIsOnGrindRail()
    isEmoting_CR:bool=Get bIsPlayingEmote()
   ) -> (Pose) {controlRigAssetReference {blueprintRigClass=&/F_MED_ToothGreek_CtrlRig.F_MED_ToothGreek_CtrlRig_C}}
   Save cached pose 'preAnimDyn'(Pose=@n24) {cacheName=preAnimDyn}
   n25 = Anim Dynamics - Bone: dyn_skirt_sd_r_02!(
    ComponentPose=Local To Component(LocalPose=Use cached pose 'preAnimDyn'())
    Alpha=1
    GravityOverride:Vector="0,0,-980"
    ExternalForce="0,0,0"
   ) -> (Pose) {
    angularDampingOverride=1
    angularSpringConstant=2000
    bAngularSpring=true
    bOverrideAngularDamping=true
    bUseGravityOverride=true
    gravityScale=0.35
    simulationSpace=BoneRelative
    alphaScaleBiasClamp {bInterpResult=true, interpSpeedDecreasing=5}
    boundBone {boneName=dyn_skirt_sd_r_02}
    componentLinearAccScale=<0.3 0.3 0.3>
    componentLinearVelScale=<0.3 0.3 0.3>
    gravityOverride {z=-980}
    physicsBodyDefinitions [{
      boundBone {boneName=dyn_skirt_sd_r_02}
      boxExtents=<20 30 30>
      constraintSetup {angularLimitsMax=<0 5 0>, angularLimitsMin=<0 -1 -12>, angularTarget=<1 0 0>}
      localJointOffset=<4 0 0>
     }]
    relativeSpaceBone {boneName=thigh_r}
    simSpaceSettings {simSpaceAngularAlpha=0.4}
   }
   n26 = Anim Dynamics - Bone: dyn_skirt_sd_r_01!(ComponentPose=@n25, Alpha=1, GravityOverride:Vector="0,0,-980", ExternalForce="0,0,0") -> (Pose) {
    angularDampingOverride=1
    angularSpringConstant=20000
    bAngularSpring=true
    bOverrideAngularDamping=true
    bUseGravityOverride=true
    gravityScale=0.35
    simulationSpace=RootRelative
    alphaScaleBiasClamp {bInterpResult=true, interpSpeedDecreasing=5}
    boundBone {boneName=dyn_skirt_sd_r_01}
    componentLinearAccScale=<0.3 0.3 0.3>
    componentLinearVelScale=<0.3 0.3 0.3>
    gravityOverride {z=-980}
    physicsBodyDefinitions [{
      boundBone {boneName=dyn_skirt_sd_r_01}
      boxExtents=<20 30 30>
      constraintSetup {angularLimitsMax=<0 5 0>, angularLimitsMin=<0 -1 -12>, angularTarget=<1 0 0>}
      localJointOffset=<4 0 0>
     }]
    simSpaceSettings {simSpaceAngularAlpha=0.4}
   }
   n27 = Anim Dynamics - Bone: dyn_skirt_sd_l_02!(ComponentPose=@n26, Alpha=1, GravityOverride:Vector="0,0,-980", ExternalForce="0,0,0") -> (Pose) {
    angularDampingOverride=1
    angularSpringConstant=2000
    bAngularSpring=true
    bOverrideAngularDamping=true
    bUseGravityOverride=true
    gravityScale=0.35
    simulationSpace=RootRelative
    alphaScaleBiasClamp {bInterpResult=true, interpSpeedDecreasing=5}
    boundBone {boneName=dyn_skirt_sd_l_02}
    componentLinearAccScale=<0.3 0.3 0.3>
    componentLinearVelScale=<0.3 0.3 0.3>
    gravityOverride {z=-980}
    physicsBodyDefinitions [{
      boundBone {boneName=dyn_skirt_sd_l_02}
      boxExtents=<20 30 30>
      constraintSetup {angularLimitsMax=<0 -1 0>, angularLimitsMin=<0 -5 -12>, angularTarget=<1 0 0>}
      localJointOffset=<4 0 0>
     }]
    simSpaceSettings {simSpaceAngularAlpha=0.4}
   }
   n28 = Anim Dynamics - Bone: dyn_skirt_sd_l_01!(ComponentPose=@n27, Alpha=1, GravityOverride:Vector="0,0,-980", ExternalForce="0,0,0") -> (Pose) {
    angularDampingOverride=1
    angularSpringConstant=20000
    bAngularSpring=true
    bOverrideAngularDamping=true
    bUseGravityOverride=true
    gravityScale=0.35
    simulationSpace=RootRelative
    alphaScaleBiasClamp {bInterpResult=true, interpSpeedDecreasing=5}
    boundBone {boneName=dyn_skirt_sd_l_01}
    componentLinearAccScale=<0.3 0.3 0.3>
    componentLinearVelScale=<0.3 0.3 0.3>
    gravityOverride {z=-980}
    physicsBodyDefinitions [{
      boundBone {boneName=dyn_skirt_sd_l_01}
      boxExtents=<20 30 30>
      constraintSetup {angularLimitsMax=<0 -1 0>, angularLimitsMin=<0 -5 -12>, angularTarget=<1 0 0>}
      localJointOffset=<4 0 0>
     }]
    simSpaceSettings {simSpaceAngularAlpha=0.4}
   }
   n29 = Anim Dynamics - Bone: dyn_skirt_bk_l_02!(ComponentPose=@n28, Alpha=1, GravityOverride:Vector="0,0,-980", ExternalForce="0,0,0") -> (Pose) {
    angularDampingOverride=1
    angularSpringConstant=2000
    bAngularSpring=true
    bOverrideAngularDamping=true
    bUseGravityOverride=true
    gravityScale=0.35
    simulationSpace=RootRelative
    alphaScaleBiasClamp {bInterpResult=true, interpSpeedDecreasing=5}
    boundBone {boneName=dyn_skirt_bk_l_02}
    componentLinearAccScale=<0.3 0.3 0.3>
    componentLinearVelScale=<0.3 0.3 0.3>
    gravityOverride {z=-980}
    physicsBodyDefinitions [{
      boundBone {boneName=dyn_skirt_bk_l_02}
      boxExtents=<20 30 30>
      constraintSetup {angularLimitsMax=<0 5 4>, angularLimitsMin=<0 -5 -10>, angularTarget=<1 0 0>}
      localJointOffset=<4 0 0>
     }]
    simSpaceSettings {simSpaceAngularAlpha=0.4}
   }
   n30 = Anim Dynamics - Bone: dyn_skirt_bk_l_01!(ComponentPose=@n29, Alpha=1, GravityOverride:Vector="0,0,-980", ExternalForce="0,0,0") -> (Pose) {
    angularDampingOverride=1
    angularSpringConstant=20000
    bAngularSpring=true
    bOverrideAngularDamping=true
    bUseGravityOverride=true
    gravityScale=0.35
    simulationSpace=RootRelative
    alphaScaleBiasClamp {bInterpResult=true, interpSpeedDecreasing=5}
    boundBone {boneName=dyn_skirt_bk_l_01}
    componentLinearAccScale=<0.3 0.3 0.3>
    componentLinearVelScale=<0.3 0.3 0.3>
    gravityOverride {z=-980}
    physicsBodyDefinitions [{
      boundBone {boneName=dyn_skirt_bk_l_01}
      boxExtents=<20 30 30>
      constraintSetup {angularLimitsMax=<0 5 4>, angularLimitsMin=<0 -5 -10>, angularTarget=<1 0 0>}
      localJointOffset=<4 0 0>
     }]
    simSpaceSettings {simSpaceAngularAlpha=0.4}
   }
   n31 = Anim Dynamics - Bone: dyn_skirt_bk_c_02!(ComponentPose=@n30, Alpha=1, GravityOverride:Vector="0,0,-980", ExternalForce="0,0,0") -> (Pose) {
    angularDampingOverride=1
    angularSpringConstant=2000
    bAngularSpring=true
    bOverrideAngularDamping=true
    bUseGravityOverride=true
    gravityScale=0.35
    simulationSpace=RootRelative
    alphaScaleBiasClamp {bInterpResult=true, interpSpeedDecreasing=5}
    boundBone {boneName=dyn_skirt_bk_c_02}
    componentLinearAccScale=<0.3 0.3 0.3>
    componentLinearVelScale=<0.3 0.3 0.3>
    gravityOverride {z=-980}
    physicsBodyDefinitions [{
      boundBone {boneName=dyn_skirt_bk_c_02}
      boxExtents=<20 30 30>
      constraintSetup {angularLimitsMax=<0 5 4>, angularLimitsMin=<0 -5 -10>, angularTarget=<1 0 0>}
      localJointOffset=<4 0 0>
     }]
    simSpaceSettings {simSpaceAngularAlpha=0.4}
   }
   n32 = Anim Dynamics - Bone: dyn_skirt_bk_c_01!(ComponentPose=@n31, Alpha=1, GravityOverride:Vector="0,0,-980", ExternalForce="0,0,0") -> (Pose) {
    angularDampingOverride=1
    angularSpringConstant=20000
    bAngularSpring=true
    bOverrideAngularDamping=true
    bUseGravityOverride=true
    gravityScale=0.35
    simulationSpace=RootRelative
    alphaScaleBiasClamp {bInterpResult=true, interpSpeedDecreasing=5}
    boundBone {boneName=dyn_skirt_bk_c_01}
    componentLinearAccScale=<0.3 0.3 0.3>
    componentLinearVelScale=<0.3 0.3 0.3>
    gravityOverride {z=-980}
    physicsBodyDefinitions [{
      boundBone {boneName=dyn_skirt_bk_c_01}
      boxExtents=<20 30 30>
      constraintSetup {angularLimitsMax=<0 5 4>, angularLimitsMin=<0 -5 -10>, angularTarget=<1 0 0>}
      localJointOffset=<4 0 0>
     }]
    simSpaceSettings {simSpaceAngularAlpha=0.4}
   }
   n33 = Anim Dynamics - Bone: dyn_skirt_bk_r_02!(ComponentPose=@n32, Alpha=1, GravityOverride:Vector="0,0,-980", ExternalForce="0,0,0") -> (Pose) {
    angularDampingOverride=1
    angularSpringConstant=2000
    bAngularSpring=true
    bOverrideAngularDamping=true
    bUseGravityOverride=true
    gravityScale=0.35
    simulationSpace=RootRelative
    alphaScaleBiasClamp {bInterpResult=true, interpSpeedDecreasing=5}
    boundBone {boneName=dyn_skirt_bk_r_02}
    componentLinearAccScale=<0.3 0.3 0.3>
    componentLinearVelScale=<0.3 0.3 0.3>
    gravityOverride {z=-980}
    physicsBodyDefinitions [{
      boundBone {boneName=dyn_skirt_bk_r_02}
      boxExtents=<20 30 30>
      constraintSetup {angularLimitsMax=<0 5 4>, angularLimitsMin=<0 -5 -10>, angularTarget=<1 0 0>}
      localJointOffset=<4 0 0>
     }]
    simSpaceSettings {simSpaceAngularAlpha=0.4}
   }
   n34 = Anim Dynamics - Bone: dyn_skirt_bk_r_01!(ComponentPose=@n33, Alpha=1, GravityOverride:Vector="0,0,-980", ExternalForce="0,0,0") -> (Pose) {
    angularDampingOverride=1
    angularSpringConstant=20000
    bAngularSpring=true
    bOverrideAngularDamping=true
    bUseGravityOverride=true
    gravityScale=0.35
    simulationSpace=RootRelative
    alphaScaleBiasClamp {bInterpResult=true, interpSpeedDecreasing=5}
    boundBone {boneName=dyn_skirt_bk_r_01}
    componentLinearAccScale=<0.3 0.3 0.3>
    componentLinearVelScale=<0.3 0.3 0.3>
    gravityOverride {z=-980}
    physicsBodyDefinitions [{
      boundBone {boneName=dyn_skirt_bk_r_01}
      boxExtents=<20 30 30>
      constraintSetup {angularLimitsMax=<0 5 4>, angularLimitsMin=<0 -5 -10>, angularTarget=<1 0 0>}
      localJointOffset=<4 0 0>
     }]
    simSpaceSettings {simSpaceAngularAlpha=0.4}
   }
   n35 = Anim Dynamics - Bone: dyn_skirt_sd_r_02(ComponentPose=Local To Component(LocalPose=Use cached pose 'preAnimDyn'()), Alpha=1, GravityOverride="0,0,-980", ExternalForce="0,0,0") {
    angularDampingOverride=1
    angularSpringConstant=2000
    bAngularSpring=true
    bOverrideAngularDamping=true
    bUseGravityOverride=true
    gravityScale=0.35
    simulationSpace=RootRelative
    alphaScaleBiasClamp {bInterpResult=true, interpSpeedDecreasing=5}
    boundBone {boneName=dyn_skirt_sd_r_02}
    componentLinearAccScale=<0.3 0.3 0.6>
    componentLinearVelScale=<0.3 0.3 0.6>
    gravityOverride {z=-980}
    physicsBodyDefinitions [{
      boundBone {boneName=dyn_skirt_sd_r_02}
      boxExtents=<20 30 30>
      constraintSetup {angularLimitsMax=<0 5 0>, angularLimitsMin=<0 -5 -25>, angularTarget=<1 0 0>}
      localJointOffset=<4 0 0>
     }]
    relativeSpaceBone {boneName=thigh_r}
    simSpaceSettings {maxAngularAcceleration=3000, maxAngularVelocity=3000, simSpaceAngularAlpha=0.1}
   }
   n36 = Anim Dynamics - Bone: dyn_skirt_sd_r_01(ComponentPose=@n35, Alpha=1, GravityOverride="0,0,-980", ExternalForce="0,0,0") {
    angularDampingOverride=1
    angularSpringConstant=2000
    bAngularSpring=true
    bOverrideAngularDamping=true
    bUseGravityOverride=true
    gravityScale=0.35
    simulationSpace=RootRelative
    alphaScaleBiasClamp {bInterpResult=true, interpSpeedDecreasing=5}
    boundBone {boneName=dyn_skirt_sd_r_01}
    componentLinearAccScale=<0.3 0.3 0.4>
    componentLinearVelScale=<0.3 0.3 0.4>
    gravityOverride {z=-980}
    physicsBodyDefinitions [{
      boundBone {boneName=dyn_skirt_sd_r_01}
      boxExtents=<20 30 30>
      constraintSetup {angularLimitsMax=<0 5 0>, angularLimitsMin=<0 -3 -25>, angularTarget=<1 0 0>}
      localJointOffset=<4 0 1.5>
     }]
    simSpaceSettings {maxAngularAcceleration=3000, maxAngularVelocity=3000, simSpaceAngularAlpha=0.1}
   }
   n37 = Anim Dynamics - Bone: dyn_skirt_sd_l_02(ComponentPose=@n36, Alpha=1, GravityOverride="0,0,-980", ExternalForce="0,0,0") {
    angularDampingOverride=1
    angularSpringConstant=2000
    bAngularSpring=true
    bOverrideAngularDamping=true
    bUseGravityOverride=true
    gravityScale=0.35
    simulationSpace=RootRelative
    alphaScaleBiasClamp {bInterpResult=true, interpSpeedDecreasing=5}
    boundBone {boneName=dyn_skirt_sd_l_02}
    componentLinearAccScale=<0.3 0.3 0.6>
    componentLinearVelScale=<0.3 0.3 0.6>
    gravityOverride {z=-980}
    physicsBodyDefinitions [{
      boundBone {boneName=dyn_skirt_sd_l_02}
      boxExtents=<20 30 30>
      constraintSetup {angularLimitsMax=<0 5 2>, angularLimitsMin=<0 -5 -25>, angularTarget=<1 0 0>}
      localJointOffset=<4 0 0>
     }]
    simSpaceSettings {maxAngularAcceleration=3000, maxAngularVelocity=3000, simSpaceAngularAlpha=0.1}
   }
   n38 = Anim Dynamics - Bone: dyn_skirt_sd_l_01(ComponentPose=@n37, Alpha=1, GravityOverride="0,0,-980", ExternalForce="0,0,0") {
    angularDampingOverride=1
    angularSpringConstant=2000
    bAngularSpring=true
    bOverrideAngularDamping=true
    bUseGravityOverride=true
    gravityScale=0.35
    simulationSpace=RootRelative
    alphaScaleBiasClamp {bInterpResult=true, interpSpeedDecreasing=5}
    boundBone {boneName=dyn_skirt_sd_l_01}
    componentLinearAccScale=<0.3 0.3 0.4>
    componentLinearVelScale=<0.3 0.3 0.4>
    gravityOverride {z=-980}
    physicsBodyDefinitions [{
      boundBone {boneName=dyn_skirt_sd_l_01}
      boxExtents=<20 30 30>
      constraintSetup {angularLimitsMax=<0 5 2>, angularLimitsMin=<0 -5 -25>, angularTarget=<1 0 0>}
      localJointOffset=<4 0 0>
     }]
    simSpaceSettings {maxAngularAcceleration=3000, maxAngularVelocity=3000, simSpaceAngularAlpha=0.1}
   }
   n39 = Anim Dynamics - Bone: dyn_skirt_fr_r_02!(ComponentPose=@n38, Alpha=1, GravityOverride:Vector="0,0,-980", ExternalForce="0,0,0") -> (Pose) {
    angularDampingOverride=1
    angularSpringConstant=2000
    bAngularSpring=true
    bOverrideAngularDamping=true
    bUseGravityOverride=true
    gravityScale=0.35
    simulationSpace=RootRelative
    alphaScaleBiasClamp {bInterpResult=true, interpSpeedDecreasing=5}
    boundBone {boneName=dyn_skirt_fr_r_02}
    componentLinearAccScale=<0.3 0.3 0.6>
    componentLinearVelScale=<0.3 0.3 0.6>
    gravityOverride {z=-980}
    physicsBodyDefinitions [{
      boundBone {boneName=dyn_skirt_fr_r_02}
      boxExtents=<20 30 30>
      constraintSetup {angularLimitsMax=<0 5 0>, angularLimitsMin=<0 -10 -25>, angularTarget=<1 0 0>}
      localJointOffset=<4 0 0>
     }]
    simSpaceSettings {maxAngularAcceleration=3000, maxAngularVelocity=3000, simSpaceAngularAlpha=0.1}
   }
   n40 = Anim Dynamics - Bone: dyn_skirt_fr_r_01!(ComponentPose=@n39, Alpha=1, GravityOverride:Vector="0,0,-980", ExternalForce="0,0,0") -> (Pose) {
    angularDampingOverride=1
    angularSpringConstant=2000
    bAngularSpring=true
    bOverrideAngularDamping=true
    bUseGravityOverride=true
    gravityScale=0.35
    simulationSpace=RootRelative
    alphaScaleBiasClamp {bInterpResult=true, interpSpeedDecreasing=5}
    boundBone {boneName=dyn_skirt_fr_r_01}
    componentLinearAccScale=<0.3 0.3 0.4>
    componentLinearVelScale=<0.3 0.3 0.4>
    gravityOverride {z=-980}
    physicsBodyDefinitions [{
      boundBone {boneName=dyn_skirt_fr_r_01}
      boxExtents=<20 30 30>
      constraintSetup {angularLimitsMax=<0 10 0>, angularLimitsMin=<0 0 -25>, angularTarget=<1 0 0>}
      localJointOffset=<4 0 2>
     }]
    simSpaceSettings {maxAngularAcceleration=3000, maxAngularVelocity=3000, simSpaceAngularAlpha=0.1}
   }
   n41 = Anim Dynamics - Bone: dyn_skirt_fr_l_02!(ComponentPose=@n40, Alpha=1, GravityOverride:Vector="0,0,-980", ExternalForce="0,0,0") -> (Pose) {
    angularDampingOverride=1
    angularSpringConstant=2000
    bAngularSpring=true
    bOverrideAngularDamping=true
    bUseGravityOverride=true
    gravityScale=0.35
    simulationSpace=RootRelative
    alphaScaleBiasClamp {bInterpResult=true, interpSpeedDecreasing=5}
    boundBone {boneName=dyn_skirt_fr_l_02}
    componentLinearAccScale=<0.3 0.3 0.6>
    componentLinearVelScale=<0.3 0.3 0.6>
    gravityOverride {z=-980}
    physicsBodyDefinitions [{
      boundBone {boneName=dyn_skirt_fr_l_02}
      boxExtents=<30 30 30>
      constraintSetup {angularLimitsMax=<0 5 0>, angularLimitsMin=<0 -10 -25>, angularTarget=<1 0 0>}
      localJointOffset=<4 0 0>
     }]
    relativeSpaceBone {boneName=thigh_r}
    simSpaceSettings {maxAngularAcceleration=3000, maxAngularVelocity=3000, simSpaceAngularAlpha=0.1}
   }
   n42 = Anim Dynamics - Bone: dyn_skirt_fr_l_01!(ComponentPose=@n41, Alpha=1, GravityOverride:Vector="0,0,-980", ExternalForce="0,0,0") -> (Pose) {
    angularDampingOverride=1
    angularSpringConstant=2000
    bAngularSpring=true
    bOverrideAngularDamping=true
    bUseGravityOverride=true
    gravityScale=0.35
    simulationSpace=RootRelative
    alphaScaleBiasClamp {bInterpResult=true, interpSpeedDecreasing=5}
    boundBone {boneName=dyn_skirt_fr_l_01}
    componentLinearAccScale=<0.3 0.3 0.4>
    componentLinearVelScale=<0.3 0.3 0.4>
    gravityOverride {z=-980}
    physicsBodyDefinitions [{
      boundBone {boneName=dyn_skirt_fr_l_01}
      boxExtents=<20 30 30>
      constraintSetup {angularLimitsMax=<0 5 0>, angularLimitsMin=<0 -10 -25>, angularTarget=<1 0 0>}
      localJointOffset=<4 0 0>
     }]
    simSpaceSettings {maxAngularAcceleration=3000, maxAngularVelocity=3000, simSpaceAngularAlpha=0.1}
   }
   n43 = Anim Dynamics - Bone: dyn_skirt_bk_l_02(ComponentPose=@n42, Alpha=1, GravityOverride="0,0,-980", ExternalForce="0,0,0") {
    angularDampingOverride=1
    angularSpringConstant=2000
    bAngularSpring=true
    bOverrideAngularDamping=true
    bUseGravityOverride=true
    gravityScale=0.35
    simulationSpace=RootRelative
    alphaScaleBiasClamp {bInterpResult=true, interpSpeedDecreasing=5}
    boundBone {boneName=dyn_skirt_bk_l_02}
    componentLinearAccScale=<0.3 0.3 0.6>
    componentLinearVelScale=<0.3 0.3 0.6>
    gravityOverride {z=-980}
    physicsBodyDefinitions [{
      boundBone {boneName=dyn_skirt_bk_l_02}
      boxExtents=<20 30 30>
      constraintSetup {angularLimitsMax=<0 5 4>, angularLimitsMin=<0 -5 -25>, angularTarget=<1 0 0>}
      localJointOffset=<4 0 0>
     }]
    simSpaceSettings {maxAngularAcceleration=3000, maxAngularVelocity=3000, simSpaceAngularAlpha=0.1}
   }
   n44 = Anim Dynamics - Bone: dyn_skirt_bk_l_01(ComponentPose=@n43, Alpha=1, GravityOverride="0,0,-980", ExternalForce="0,0,0") {
    angularDampingOverride=1
    angularSpringConstant=2000
    bAngularSpring=true
    bOverrideAngularDamping=true
    bUseGravityOverride=true
    gravityScale=0.35
    simulationSpace=RootRelative
    alphaScaleBiasClamp {bInterpResult=true, interpSpeedDecreasing=5}
    boundBone {boneName=dyn_skirt_bk_l_01}
    componentLinearAccScale=<0.3 0.3 0.4>
    componentLinearVelScale=<0.3 0.3 0.4>
    gravityOverride {z=-980}
    physicsBodyDefinitions [{
      boundBone {boneName=dyn_skirt_bk_l_01}
      boxExtents=<20 30 30>
      constraintSetup {angularLimitsMax=<0 5 4>, angularLimitsMin=<0 -5 -25>, angularTarget=<1 0 0>}
      localJointOffset=<4 0 0>
     }]
    simSpaceSettings {maxAngularAcceleration=3000, maxAngularVelocity=3000, simSpaceAngularAlpha=0.1}
   }
   n45 = Anim Dynamics - Bone: dyn_skirt_bk_c_02(ComponentPose=@n44, Alpha=1, GravityOverride="0,0,-980", ExternalForce="0,0,0") {
    angularDampingOverride=1
    angularSpringConstant=2000
    bAngularSpring=true
    bOverrideAngularDamping=true
    bUseGravityOverride=true
    gravityScale=0.35
    simulationSpace=RootRelative
    alphaScaleBiasClamp {bInterpResult=true, interpSpeedDecreasing=5}
    boundBone {boneName=dyn_skirt_bk_c_02}
    componentLinearAccScale=<0.3 0.3 0.6>
    componentLinearVelScale=<0.3 0.3 0.6>
    gravityOverride {z=-980}
    physicsBodyDefinitions [{
      boundBone {boneName=dyn_skirt_bk_c_02}
      boxExtents=<20 30 30>
      constraintSetup {angularLimitsMax=<0 5 4>, angularLimitsMin=<0 -5 -25>, angularTarget=<1 0 0>}
      localJointOffset=<4 0 0>
     }]
    simSpaceSettings {maxAngularAcceleration=3000, maxAngularVelocity=3000, simSpaceAngularAlpha=0.1}
   }
   n46 = Anim Dynamics - Bone: dyn_skirt_bk_c_01(ComponentPose=@n45, Alpha=1, GravityOverride="0,0,-980", ExternalForce="0,0,0") {
    angularDampingOverride=1
    angularSpringConstant=2000
    bAngularSpring=true
    bOverrideAngularDamping=true
    bUseGravityOverride=true
    gravityScale=0.35
    simulationSpace=RootRelative
    alphaScaleBiasClamp {bInterpResult=true, interpSpeedDecreasing=5}
    boundBone {boneName=dyn_skirt_bk_c_01}
    componentLinearAccScale=<0.3 0.3 0.4>
    componentLinearVelScale=<0.3 0.3 0.4>
    gravityOverride {z=-980}
    physicsBodyDefinitions [{
      boundBone {boneName=dyn_skirt_bk_c_01}
      boxExtents=<20 30 30>
      constraintSetup {angularLimitsMax=<0 5 4>, angularLimitsMin=<0 -5 -35>, angularTarget=<1 0 0>}
      localJointOffset=<4 0 0>
     }]
    simSpaceSettings {maxAngularAcceleration=3000, maxAngularVelocity=3000, simSpaceAngularAlpha=0.1}
   }
   n47 = Anim Dynamics - Bone: dyn_skirt_bk_r_02(ComponentPose=@n46, Alpha=1, GravityOverride="0,0,-980", ExternalForce="0,0,0") {
    angularDampingOverride=1
    angularSpringConstant=2000
    bAngularSpring=true
    bOverrideAngularDamping=true
    bUseGravityOverride=true
    gravityScale=0.35
    simulationSpace=RootRelative
    alphaScaleBiasClamp {bInterpResult=true, interpSpeedDecreasing=5}
    boundBone {boneName=dyn_skirt_bk_r_02}
    componentLinearAccScale=<0.3 0.3 0.6>
    componentLinearVelScale=<0.3 0.3 0.6>
    gravityOverride {z=-980}
    physicsBodyDefinitions [{
      boundBone {boneName=dyn_skirt_bk_r_02}
      boxExtents=<20 30 30>
      constraintSetup {angularLimitsMax=<0 5 4>, angularLimitsMin=<0 -5 -25>, angularTarget=<1 0 0>}
      localJointOffset=<4 0 0>
     }]
    simSpaceSettings {maxAngularAcceleration=3000, maxAngularVelocity=3000, simSpaceAngularAlpha=0.1}
   }
   n48 = Anim Dynamics - Bone: dyn_skirt_bk_r_01(ComponentPose=@n47, Alpha=1, GravityOverride="0,0,-980", ExternalForce="0,0,0") {
    angularDampingOverride=1
    angularSpringConstant=2000
    bAngularSpring=true
    bOverrideAngularDamping=true
    bUseGravityOverride=true
    gravityScale=0.35
    simulationSpace=RootRelative
    alphaScaleBiasClamp {bInterpResult=true, interpSpeedDecreasing=5}
    boundBone {boneName=dyn_skirt_bk_r_01}
    componentLinearAccScale=<0.3 0.3 0.4>
    componentLinearVelScale=<0.3 0.3 0.4>
    gravityOverride {z=-980}
    physicsBodyDefinitions [{
      boundBone {boneName=dyn_skirt_bk_r_01}
      boxExtents=<20 30 30>
      constraintSetup {angularLimitsMax=<0 5 4>, angularLimitsMin=<0 -5 -25>, angularTarget=<1 0 0>}
      localJointOffset=<4 0 0>
     }]
    simSpaceSettings {maxAngularAcceleration=3000, maxAngularVelocity=3000, simSpaceAngularAlpha=0.1}
   }
   n49 = Blend Poses by bool(
    BlendPose_0=Component To Local(ComponentPose=@n34)
    BlendPose_1=Component To Local(ComponentPose=@n48)
    bActiveValue=Get isCrouchingOrSkydiving() -> (isCrouchingOrSkydiving:bool)
   )
   n50 = Anim Dynamics - Bone: dyn_skirt_sd_r_02(ComponentPose=Local To Component(LocalPose=Use cached pose 'preAnimDyn'()), Alpha=1, GravityOverride="0,0,-980", ExternalForce="0,0,0") {
    angularDampingOverride=1
    angularSpringConstant=2000
    bAngularSpring=true
    bOverrideAngularDamping=true
    bUseGravityOverride=true
    gravityScale=0.35
    simulationSpace=RootRelative
    alphaScaleBiasClamp {bInterpResult=true, interpSpeedDecreasing=5}
    boundBone {boneName=dyn_skirt_sd_r_02}
    componentLinearAccScale=<0.3 0.3 0.6>
    componentLinearVelScale=<0.3 0.3 0.6>
    gravityOverride {z=-980}
    physicsBodyDefinitions [{
      boundBone {boneName=dyn_skirt_sd_r_02}
      boxExtents=<20 30 30>
      constraintSetup {angularLimitsMax=<0 5 0>, angularLimitsMin=<0 -5 -25>, angularTarget=<1 0 0>}
      localJointOffset=<4 0 0>
     }]
    relativeSpaceBone {boneName=thigh_r}
    simSpaceSettings {maxAngularAcceleration=3000, maxAngularVelocity=3000, simSpaceAngularAlpha=0.1}
   }
   n51 = Anim Dynamics - Bone: dyn_skirt_sd_r_01(ComponentPose=@n50, Alpha=1, GravityOverride="0,0,-980", ExternalForce="0,0,0") {
    angularDampingOverride=1
    angularSpringConstant=2000
    bAngularSpring=true
    bOverrideAngularDamping=true
    bUseGravityOverride=true
    gravityScale=0.35
    simulationSpace=RootRelative
    alphaScaleBiasClamp {bInterpResult=true, interpSpeedDecreasing=5}
    boundBone {boneName=dyn_skirt_sd_r_01}
    componentLinearAccScale=<0.3 0.3 0.4>
    componentLinearVelScale=<0.3 0.3 0.4>
    gravityOverride {z=-980}
    physicsBodyDefinitions [{
      boundBone {boneName=dyn_skirt_sd_r_01}
      boxExtents=<20 30 30>
      constraintSetup {angularLimitsMax=<0 5 0>, angularLimitsMin=<0 -3 -25>, angularTarget=<1 0 0>}
      localJointOffset=<4 0 1.5>
     }]
    simSpaceSettings {maxAngularAcceleration=3000, maxAngularVelocity=3000, simSpaceAngularAlpha=0.1}
   }
   n52 = Anim Dynamics - Bone: dyn_skirt_sd_l_02(ComponentPose=@n51, Alpha=1, GravityOverride="0,0,-980", ExternalForce="0,0,0") {
    angularDampingOverride=1
    angularSpringConstant=2000
    bAngularSpring=true
    bOverrideAngularDamping=true
    bUseGravityOverride=true
    gravityScale=0.35
    simulationSpace=RootRelative
    alphaScaleBiasClamp {bInterpResult=true, interpSpeedDecreasing=5}
    boundBone {boneName=dyn_skirt_sd_l_02}
    componentLinearAccScale=<0.3 0.3 0.6>
    componentLinearVelScale=<0.3 0.3 0.6>
    gravityOverride {z=-980}
    physicsBodyDefinitions [{
      boundBone {boneName=dyn_skirt_sd_l_02}
      boxExtents=<20 30 30>
      constraintSetup {angularLimitsMax=<0 5 2>, angularLimitsMin=<0 -5 -25>, angularTarget=<1 0 0>}
      localJointOffset=<4 0 0>
     }]
    simSpaceSettings {maxAngularAcceleration=3000, maxAngularVelocity=3000, simSpaceAngularAlpha=0.1}
   }
   n53 = Anim Dynamics - Bone: dyn_skirt_sd_l_01(ComponentPose=@n52, Alpha=1, GravityOverride="0,0,-980", ExternalForce="0,0,0") {
    angularDampingOverride=1
    angularSpringConstant=2000
    bAngularSpring=true
    bOverrideAngularDamping=true
    bUseGravityOverride=true
    gravityScale=0.35
    simulationSpace=RootRelative
    alphaScaleBiasClamp {bInterpResult=true, interpSpeedDecreasing=5}
    boundBone {boneName=dyn_skirt_sd_l_01}
    componentLinearAccScale=<0.3 0.3 0.4>
    componentLinearVelScale=<0.3 0.3 0.4>
    gravityOverride {z=-980}
    physicsBodyDefinitions [{
      boundBone {boneName=dyn_skirt_sd_l_01}
      boxExtents=<20 30 30>
      constraintSetup {angularLimitsMax=<0 5 2>, angularLimitsMin=<0 -5 -25>, angularTarget=<1 0 0>}
      localJointOffset=<4 0 0>
     }]
    simSpaceSettings {maxAngularAcceleration=3000, maxAngularVelocity=3000, simSpaceAngularAlpha=0.1}
   }
   n54 = Anim Dynamics - Bone: dyn_skirt_fr_r_02(ComponentPose=@n53, Alpha=1, GravityOverride="0,0,-980", ExternalForce="0,0,0") {
    angularDampingOverride=1
    angularSpringConstant=2000
    bAngularSpring=true
    bOverrideAngularDamping=true
    bUseGravityOverride=true
    gravityScale=0.35
    simulationSpace=RootRelative
    alphaScaleBiasClamp {bInterpResult=true, interpSpeedDecreasing=5}
    boundBone {boneName=dyn_skirt_fr_r_02}
    componentLinearAccScale=<0.3 0.3 0.6>
    componentLinearVelScale=<0.3 0.3 0.6>
    gravityOverride {z=-980}
    physicsBodyDefinitions [{
      boundBone {boneName=dyn_skirt_fr_r_02}
      boxExtents=<20 30 30>
      constraintSetup {angularLimitsMax=<0 5 0>, angularLimitsMin=<0 -10 -25>, angularTarget=<1 0 0>}
      localJointOffset=<4 0 0>
     }]
    simSpaceSettings {maxAngularAcceleration=3000, maxAngularVelocity=3000, simSpaceAngularAlpha=0.1}
   }
   n55 = Anim Dynamics - Bone: dyn_skirt_fr_r_01(ComponentPose=@n54, Alpha=1, GravityOverride="0,0,-980", ExternalForce="0,0,0") {
    angularDampingOverride=1
    angularSpringConstant=2000
    bAngularSpring=true
    bOverrideAngularDamping=true
    bUseGravityOverride=true
    gravityScale=0.35
    simulationSpace=RootRelative
    alphaScaleBiasClamp {bInterpResult=true, interpSpeedDecreasing=5}
    boundBone {boneName=dyn_skirt_fr_r_01}
    componentLinearAccScale=<0.3 0.3 0.4>
    componentLinearVelScale=<0.3 0.3 0.4>
    gravityOverride {z=-980}
    physicsBodyDefinitions [{
      boundBone {boneName=dyn_skirt_fr_r_01}
      boxExtents=<20 30 30>
      constraintSetup {angularLimitsMax=<0 10 0>, angularLimitsMin=<0 0 -25>, angularTarget=<1 0 0>}
      localJointOffset=<4 0 2>
     }]
    simSpaceSettings {maxAngularAcceleration=3000, maxAngularVelocity=3000, simSpaceAngularAlpha=0.1}
   }
   n56 = Anim Dynamics - Bone: dyn_skirt_fr_c_02!(ComponentPose=@n55, Alpha=1, GravityOverride:Vector="0,0,-980", ExternalForce="0,0,0") -> (Pose) {
    angularDampingOverride=1
    angularSpringConstant=2000
    bAngularSpring=true
    bOverrideAngularDamping=true
    bUseGravityOverride=true
    gravityScale=0.35
    simulationSpace=RootRelative
    alphaScaleBiasClamp {bInterpResult=true, interpSpeedDecreasing=5}
    boundBone {boneName=dyn_skirt_fr_c_02}
    componentLinearAccScale=<0.3 0.3 0.6>
    componentLinearVelScale=<0.3 0.3 0.6>
    gravityOverride {z=-980}
    physicsBodyDefinitions [{
      boundBone {boneName=dyn_skirt_fr_c_02}
      boxExtents=<20 30 30>
      constraintSetup {angularLimitsMax=<0 5 0>, angularLimitsMin=<0 -5 -25>, angularTarget=<1 0 0>}
      localJointOffset=<4 0 0>
     }]
    simSpaceSettings {maxAngularAcceleration=3000, maxAngularVelocity=3000, simSpaceAngularAlpha=0.1}
   }
   n57 = Anim Dynamics - Bone: dyn_skirt_fr_c_01!(ComponentPose=@n56, Alpha=1, GravityOverride:Vector="0,0,-980", ExternalForce="0,0,0") -> (Pose) {
    angularDampingOverride=1
    angularSpringConstant=2000
    bAngularSpring=true
    bOverrideAngularDamping=true
    bUseGravityOverride=true
    gravityScale=0.35
    simulationSpace=RootRelative
    alphaScaleBiasClamp {bInterpResult=true, interpSpeedDecreasing=5}
    boundBone {boneName=dyn_skirt_fr_c_01}
    componentLinearAccScale=<0.3 0.3 0.4>
    componentLinearVelScale=<0.3 0.3 0.4>
    gravityOverride {z=-980}
    physicsBodyDefinitions [{
      boundBone {boneName=dyn_skirt_fr_c_01}
      boxExtents=<20 30 30>
      constraintSetup {angularLimitsMax=<0 5 0>, angularLimitsMin=<0 -5 -25>, angularTarget=<1 0 0>}
      localJointOffset=<4 0 0>
     }]
    simSpaceSettings {maxAngularAcceleration=3000, maxAngularVelocity=3000, simSpaceAngularAlpha=0.1}
   }
   n58 = Anim Dynamics - Bone: dyn_skirt_fr_l_02(ComponentPose=@n57, Alpha=1, GravityOverride="0,0,-980", ExternalForce="0,0,0") {
    angularDampingOverride=1
    angularSpringConstant=2000
    bAngularSpring=true
    bOverrideAngularDamping=true
    bUseGravityOverride=true
    gravityScale=0.35
    simulationSpace=RootRelative
    alphaScaleBiasClamp {bInterpResult=true, interpSpeedDecreasing=5}
    boundBone {boneName=dyn_skirt_fr_l_02}
    componentLinearAccScale=<0.3 0.3 0.6>
    componentLinearVelScale=<0.3 0.3 0.6>
    gravityOverride {z=-980}
    physicsBodyDefinitions [{
      boundBone {boneName=dyn_skirt_fr_l_02}
      boxExtents=<30 30 30>
      constraintSetup {angularLimitsMax=<0 3 0>, angularLimitsMin=<0 -3 -25>, angularTarget=<1 0 0>}
      localJointOffset=<4 0 0>
     }]
    relativeSpaceBone {boneName=thigh_r}
    simSpaceSettings {maxAngularAcceleration=3000, maxAngularVelocity=3000, simSpaceAngularAlpha=0.1}
   }
   n59 = Anim Dynamics - Bone: dyn_skirt_fr_l_01(ComponentPose=@n58, Alpha=1, GravityOverride="0,0,-980", ExternalForce="0,0,0") {
    angularDampingOverride=1
    angularSpringConstant=2000
    bAngularSpring=true
    bOverrideAngularDamping=true
    bUseGravityOverride=true
    gravityScale=0.35
    simulationSpace=RootRelative
    alphaScaleBiasClamp {bInterpResult=true, interpSpeedDecreasing=5}
    boundBone {boneName=dyn_skirt_fr_l_01}
    componentLinearAccScale=<0.3 0.3 0.4>
    componentLinearVelScale=<0.3 0.3 0.4>
    gravityOverride {z=-980}
    physicsBodyDefinitions [{
      boundBone {boneName=dyn_skirt_fr_l_01}
      boxExtents=<20 30 30>
      constraintSetup {angularLimitsMax=<0 3 0>, angularLimitsMin=<0 -3 -25>, angularTarget=<1 0 0>}
      localJointOffset=<4 0 0>
     }]
    simSpaceSettings {maxAngularAcceleration=3000, maxAngularVelocity=3000, simSpaceAngularAlpha=0.1}
   }
   n60 = Anim Dynamics - Bone: dyn_skirt_bk_l_02(ComponentPose=@n59, Alpha=1, GravityOverride="0,0,-980", ExternalForce="0,0,0") {
    angularDampingOverride=1
    angularSpringConstant=2000
    bAngularSpring=true
    bOverrideAngularDamping=true
    bUseGravityOverride=true
    gravityScale=0.35
    simulationSpace=RootRelative
    alphaScaleBiasClamp {bInterpResult=true, interpSpeedDecreasing=5}
    boundBone {boneName=dyn_skirt_bk_l_02}
    componentLinearAccScale=<0.3 0.3 0.6>
    componentLinearVelScale=<0.3 0.3 0.6>
    gravityOverride {z=-980}
    physicsBodyDefinitions [{
      boundBone {boneName=dyn_skirt_bk_l_02}
      boxExtents=<20 30 30>
      constraintSetup {angularLimitsMax=<0 5 4>, angularLimitsMin=<0 -5 -25>, angularTarget=<1 0 0>}
      localJointOffset=<4 0 0>
     }]
    simSpaceSettings {maxAngularAcceleration=3000, maxAngularVelocity=3000, simSpaceAngularAlpha=0.1}
   }
   n61 = Anim Dynamics - Bone: dyn_skirt_bk_l_01(ComponentPose=@n60, Alpha=1, GravityOverride="0,0,-980", ExternalForce="0,0,0") {
    angularDampingOverride=1
    angularSpringConstant=2000
    bAngularSpring=true
    bOverrideAngularDamping=true
    bUseGravityOverride=true
    gravityScale=0.35
    simulationSpace=RootRelative
    alphaScaleBiasClamp {bInterpResult=true, interpSpeedDecreasing=5}
    boundBone {boneName=dyn_skirt_bk_l_01}
    componentLinearAccScale=<0.3 0.3 0.4>
    componentLinearVelScale=<0.3 0.3 0.4>
    gravityOverride {z=-980}
    physicsBodyDefinitions [{
      boundBone {boneName=dyn_skirt_bk_l_01}
      boxExtents=<20 30 30>
      constraintSetup {angularLimitsMax=<0 5 4>, angularLimitsMin=<0 -5 -25>, angularTarget=<1 0 0>}
      localJointOffset=<4 0 0>
     }]
    simSpaceSettings {maxAngularAcceleration=3000, maxAngularVelocity=3000, simSpaceAngularAlpha=0.1}
   }
   n62 = Anim Dynamics - Bone: dyn_skirt_bk_c_02(ComponentPose=@n61, Alpha=1, GravityOverride="0,0,-980", ExternalForce="0,0,0") {
    angularDampingOverride=1
    angularSpringConstant=2000
    bAngularSpring=true
    bOverrideAngularDamping=true
    bUseGravityOverride=true
    gravityScale=0.35
    simulationSpace=RootRelative
    alphaScaleBiasClamp {bInterpResult=true, interpSpeedDecreasing=5}
    boundBone {boneName=dyn_skirt_bk_c_02}
    componentLinearAccScale=<0.3 0.3 0.6>
    componentLinearVelScale=<0.3 0.3 0.6>
    gravityOverride {z=-980}
    physicsBodyDefinitions [{
      boundBone {boneName=dyn_skirt_bk_c_02}
      boxExtents=<20 30 30>
      constraintSetup {angularLimitsMax=<0 5 4>, angularLimitsMin=<0 -5 -25>, angularTarget=<1 0 0>}
      localJointOffset=<4 0 0>
     }]
    simSpaceSettings {maxAngularAcceleration=3000, maxAngularVelocity=3000, simSpaceAngularAlpha=0.1}
   }
   n63 = Anim Dynamics - Bone: dyn_skirt_bk_c_01(ComponentPose=@n62, Alpha=1, GravityOverride="0,0,-980", ExternalForce="0,0,0") {
    angularDampingOverride=1
    angularSpringConstant=2000
    bAngularSpring=true
    bOverrideAngularDamping=true
    bUseGravityOverride=true
    gravityScale=0.35
    simulationSpace=RootRelative
    alphaScaleBiasClamp {bInterpResult=true, interpSpeedDecreasing=5}
    boundBone {boneName=dyn_skirt_bk_c_01}
    componentLinearAccScale=<0.3 0.3 0.4>
    componentLinearVelScale=<0.3 0.3 0.4>
    gravityOverride {z=-980}
    physicsBodyDefinitions [{
      boundBone {boneName=dyn_skirt_bk_c_01}
      boxExtents=<20 30 30>
      constraintSetup {angularLimitsMax=<0 5 4>, angularLimitsMin=<0 -5 -35>, angularTarget=<1 0 0>}
      localJointOffset=<4 0 0>
     }]
    simSpaceSettings {maxAngularAcceleration=3000, maxAngularVelocity=3000, simSpaceAngularAlpha=0.1}
   }
   n64 = Anim Dynamics - Bone: dyn_skirt_bk_r_02(ComponentPose=@n63, Alpha=1, GravityOverride="0,0,-980", ExternalForce="0,0,0") {
    angularDampingOverride=1
    angularSpringConstant=2000
    bAngularSpring=true
    bOverrideAngularDamping=true
    bUseGravityOverride=true
    gravityScale=0.35
    simulationSpace=RootRelative
    alphaScaleBiasClamp {bInterpResult=true, interpSpeedDecreasing=5}
    boundBone {boneName=dyn_skirt_bk_r_02}
    componentLinearAccScale=<0.3 0.3 0.6>
    componentLinearVelScale=<0.3 0.3 0.6>
    gravityOverride {z=-980}
    physicsBodyDefinitions [{
      boundBone {boneName=dyn_skirt_bk_r_02}
      boxExtents=<20 30 30>
      constraintSetup {angularLimitsMax=<0 5 4>, angularLimitsMin=<0 -5 -25>, angularTarget=<1 0 0>}
      localJointOffset=<4 0 0>
     }]
    simSpaceSettings {maxAngularAcceleration=3000, maxAngularVelocity=3000, simSpaceAngularAlpha=0.1}
   }
   n65 = Anim Dynamics - Bone: dyn_skirt_bk_r_01(ComponentPose=@n64, Alpha=1, GravityOverride="0,0,-980", ExternalForce="0,0,0") {
    angularDampingOverride=1
    angularSpringConstant=2000
    bAngularSpring=true
    bOverrideAngularDamping=true
    bUseGravityOverride=true
    gravityScale=0.35
    simulationSpace=RootRelative
    alphaScaleBiasClamp {bInterpResult=true, interpSpeedDecreasing=5}
    boundBone {boneName=dyn_skirt_bk_r_01}
    componentLinearAccScale=<0.3 0.3 0.4>
    componentLinearVelScale=<0.3 0.3 0.4>
    gravityOverride {z=-980}
    physicsBodyDefinitions [{
      boundBone {boneName=dyn_skirt_bk_r_01}
      boxExtents=<20 30 30>
      constraintSetup {angularLimitsMax=<0 5 4>, angularLimitsMin=<0 -5 -25>, angularTarget=<1 0 0>}
      localJointOffset=<4 0 0>
     }]
    simSpaceSettings {maxAngularAcceleration=3000, maxAngularVelocity=3000, simSpaceAngularAlpha=0.1}
   }
   n66 = Component To Local(ComponentPose=@n65)
   n67 = Get bIsInFrontEnd()
   n68 = Anim Dynamics - Bone: dyn_zipper!(
    ComponentPose=Local To Component(LocalPose=Blend Poses by bool(BlendPose_0=@n66, BlendPose_1=@n49, bActiveValue=@n67))
    Alpha=1
    GravityOverride:Vector="0,0,-980"
    ExternalForce
   ) -> (Pose) {
    angularDampingOverride=0.999
    angularSpringConstant=2000
    bAngularSpring=true
    bOverrideAngularDamping=true
    bUseGravityOverride=true
    gravityScale=0.35
    boundBone {boneName=dyn_zipper}
    componentAppliedLinearAccClamp=<3000 3000 3000>
    componentLinearAccScale=<0.3 0.3 0.3>
    componentLinearVelScale=<0.3 0.3 0.3>
    gravityOverride {z=-980}
    physicsBodyDefinitions [{
      boundBone {boneName=dyn_zipper}
      boxExtents=<10 15 15>
      constraintSetup {angularLimitsMax=<0 30 5>, angularLimitsMin=<0 -10 -40>, angularTarget=<1 0 0>}
      localJointOffset=<2.5 0 0>
     }]
    relativeSpaceBone {boneName=spine_05}
    simSpaceSettings {simSpaceAngularAlpha=0.1}
   }
   n69 = Anim Dynamics - Bone: dyn_ribbonChest_a01_l!(ComponentPose=@n68, Alpha=1, GravityOverride:Vector="0,0,-980", ExternalForce="0,0,0") -> (Pose) {
    angularDampingOverride=0.999
    angularSpringConstant=3000
    bAngularSpring=true
    bOverrideAngularDamping=true
    bUseGravityOverride=true
    gravityScale=0.35
    simulationSpace=RootRelative
    boundBone {boneName=dyn_ribbonChest_a01_l}
    componentAppliedLinearAccClamp=<3000 3000 3000>
    componentLinearAccScale=<0.3 0.3 0.3>
    componentLinearVelScale=<0.3 0.3 0.3>
    gravityOverride {z=-980}
    physicsBodyDefinitions [{
      boundBone {boneName=dyn_ribbonChest_a01_l}
      boxExtents=<16 16 16>
      constraintSetup {angularLimitsMax=<0 40 3>, angularLimitsMin=<0 -25 -40>, angularTarget=<1 0 0>}
      localJointOffset=<2 -1 0>
     }]
    relativeSpaceBone {boneName=spine_05}
    simSpaceSettings {simSpaceAngularAlpha=0.1}
   }
   n70 = Anim Dynamics - Bone: dyn_ribbonChest_b01_l!(ComponentPose=@n69, Alpha=1, GravityOverride:Vector="0,0,-980", ExternalForce="0,0,0") -> (Pose) {
    angularDampingOverride=0.999
    angularSpringConstant=3000
    bAngularSpring=true
    bOverrideAngularDamping=true
    bUseGravityOverride=true
    gravityScale=0.35
    simulationSpace=RootRelative
    boundBone {boneName=dyn_ribbonChest_b01_l}
    componentAppliedLinearAccClamp=<3000 3000 3000>
    componentLinearAccScale=<0.3 0.3 0.3>
    componentLinearVelScale=<0.3 0.3 0.3>
    gravityOverride {z=-980}
    physicsBodyDefinitions [{
      boundBone {boneName=dyn_ribbonChest_b01_l}
      boxExtents=<16 16 16>
      constraintSetup {angularLimitsMax=<0 10 3>, angularLimitsMin=<0 -25 -40>, angularTarget=<1 0 0>}
      localJointOffset=<2 -1 0>
     }]
    relativeSpaceBone {boneName=spine_05}
    simSpaceSettings {simSpaceAngularAlpha=0.1}
   }
   n71 = Anim Dynamics - Bone: dyn_ribbonChest_c01_l!(ComponentPose=@n70, Alpha=1, GravityOverride:Vector="0,0,-980", ExternalForce="0,0,0") -> (Pose) {
    angularDampingOverride=0.999
    angularSpringConstant=3000
    bAngularSpring=true
    bOverrideAngularDamping=true
    bUseGravityOverride=true
    gravityScale=0.35
    simulationSpace=RootRelative
    boundBone {boneName=dyn_ribbonChest_c01_l}
    componentAppliedLinearAccClamp=<3000 3000 3000>
    componentLinearAccScale=<0.3 0.3 0.3>
    componentLinearVelScale=<0.3 0.3 0.3>
    gravityOverride {z=-980}
    physicsBodyDefinitions [{
      boundBone {boneName=dyn_ribbonChest_c01_l}
      boxExtents=<16 16 16>
      constraintSetup {angularLimitsMax=<0 30 3>, angularLimitsMin=<0 -25 -40>, angularTarget=<1 0 0>}
      localJointOffset=<2 -1 0>
     }]
    relativeSpaceBone {boneName=spine_05}
    simSpaceSettings {simSpaceAngularAlpha=0.1}
   }
   n72 = Anim Dynamics - Bone: dyn_ribbonChest_c02_l!(ComponentPose=@n71, Alpha=1, GravityOverride:Vector="0,0,-980", ExternalForce="0,0,0") -> (Pose) {
    angularDampingOverride=0.999
    angularSpringConstant=3000
    bAngularSpring=true
    bOverrideAngularDamping=true
    bUseGravityOverride=true
    gravityScale=0.35
    simulationSpace=RootRelative
    boundBone {boneName=dyn_ribbonChest_c02_l}
    componentAppliedLinearAccClamp=<3000 3000 3000>
    componentLinearAccScale=<0.3 0.3 0.3>
    componentLinearVelScale=<0.3 0.3 0.3>
    gravityOverride {z=-980}
    physicsBodyDefinitions [{
      boundBone {boneName=dyn_ribbonChest_c02_l}
      boxExtents=<16 16 16>
      constraintSetup {angularLimitsMax=<0 40 3>, angularLimitsMin=<0 -25 -40>, angularTarget=<1 0 0>}
      localJointOffset=<2 -1 0>
     }]
    relativeSpaceBone {boneName=spine_05}
    simSpaceSettings {simSpaceAngularAlpha=0.1}
   }
   n73 = Anim Dynamics - Bone: dyn_ribbonChest_d01_l!(ComponentPose=@n72, Alpha=1, GravityOverride:Vector="0,0,-980", ExternalForce="0,0,0") -> (Pose) {
    angularDampingOverride=0.999
    angularSpringConstant=3000
    bAngularSpring=true
    bOverrideAngularDamping=true
    bUseGravityOverride=true
    gravityScale=0.35
    simulationSpace=RootRelative
    boundBone {boneName=dyn_ribbonChest_d01_l}
    componentAppliedLinearAccClamp=<3000 3000 3000>
    componentLinearAccScale=<0.3 0.3 0.3>
    componentLinearVelScale=<0.3 0.3 0.3>
    gravityOverride {z=-980}
    physicsBodyDefinitions [{
      boundBone {boneName=dyn_ribbonChest_d01_l}
      boxExtents=<16 16 16>
      constraintSetup {angularLimitsMax=<0 30 3>, angularLimitsMin=<0 -25 -40>, angularTarget=<1 0 0>}
      localJointOffset=<2 -1 0>
     }]
    relativeSpaceBone {boneName=spine_05}
    simSpaceSettings {simSpaceAngularAlpha=0.1}
   }
   n74 = Anim Dynamics - Bone: dyn_ribbonChest_a01_r!(ComponentPose=@n73, Alpha=1, GravityOverride:Vector="0,0,-980", ExternalForce="0,0,0") -> (Pose) {
    angularDampingOverride=0.999
    angularSpringConstant=3000
    bAngularSpring=true
    bOverrideAngularDamping=true
    bUseGravityOverride=true
    gravityScale=0.35
    simulationSpace=RootRelative
    boundBone {boneName=dyn_ribbonChest_a01_r}
    componentAppliedLinearAccClamp=<3000 3000 3000>
    componentLinearAccScale=<0.3 0.3 0.3>
    componentLinearVelScale=<0.3 0.3 0.3>
    gravityOverride {z=-980}
    physicsBodyDefinitions [{
      boundBone {boneName=dyn_ribbonChest_a01_r}
      boxExtents=<16 16 16>
      constraintSetup {angularLimitsMax=<0 25 3>, angularLimitsMin=<0 -30 -40>, angularTarget=<1 0 0>}
      localJointOffset=<2 -1 0>
     }]
    relativeSpaceBone {boneName=spine_05}
    simSpaceSettings {simSpaceAngularAlpha=0.1}
   }
   n75 = Anim Dynamics - Bone: dyn_ribbonChest_b01_r!(ComponentPose=@n74, Alpha=1, GravityOverride:Vector="0,0,-980", ExternalForce="0,0,0") -> (Pose) {
    angularDampingOverride=0.999
    angularSpringConstant=3000
    bAngularSpring=true
    bOverrideAngularDamping=true
    bUseGravityOverride=true
    gravityScale=0.35
    simulationSpace=RootRelative
    boundBone {boneName=dyn_ribbonChest_b01_r}
    componentAppliedLinearAccClamp=<3000 3000 3000>
    componentLinearAccScale=<0.3 0.3 0.3>
    componentLinearVelScale=<0.3 0.3 0.3>
    gravityOverride {z=-980}
    physicsBodyDefinitions [{
      boundBone {boneName=dyn_ribbonChest_b01_r}
      boxExtents=<16 16 16>
      constraintSetup {angularLimitsMax=<0 25 0>, angularLimitsMin=<0 -30 -40>, angularTarget=<1 0 0>}
      localJointOffset=<2 -1 0>
     }]
    relativeSpaceBone {boneName=spine_05}
    simSpaceSettings {simSpaceAngularAlpha=0.1}
   }
   n76 = Anim Dynamics - Bone: dyn_ribbonChest_c01_r!(ComponentPose=@n75, Alpha=1, GravityOverride:Vector="0,0,-980", ExternalForce="0,0,0") -> (Pose) {
    angularDampingOverride=0.999
    angularSpringConstant=3000
    bAngularSpring=true
    bOverrideAngularDamping=true
    bUseGravityOverride=true
    gravityScale=0.35
    simulationSpace=RootRelative
    boundBone {boneName=dyn_ribbonChest_c01_r}
    componentAppliedLinearAccClamp=<3000 3000 3000>
    componentLinearAccScale=<0.3 0.3 0.3>
    componentLinearVelScale=<0.3 0.3 0.3>
    gravityOverride {z=-980}
    physicsBodyDefinitions [{
      boundBone {boneName=dyn_ribbonChest_c01_r}
      boxExtents=<16 16 16>
      constraintSetup {angularLimitsMax=<0 25 3>, angularLimitsMin=<0 -30 -40>, angularTarget=<1 0 0>}
      localJointOffset=<2 -1 0>
     }]
    relativeSpaceBone {boneName=spine_05}
    simSpaceSettings {simSpaceAngularAlpha=0.1}
   }
   n77 = Anim Dynamics - Bone: dyn_ribbonChest_c02_r!(ComponentPose=@n76, Alpha=1, GravityOverride:Vector="0,0,-980", ExternalForce="0,0,0") -> (Pose) {
    angularDampingOverride=0.999
    angularSpringConstant=3000
    bAngularSpring=true
    bOverrideAngularDamping=true
    bUseGravityOverride=true
    gravityScale=0.35
    simulationSpace=RootRelative
    boundBone {boneName=dyn_ribbonChest_c02_r}
    componentAppliedLinearAccClamp=<3000 3000 3000>
    componentLinearAccScale=<0.3 0.3 0.3>
    componentLinearVelScale=<0.3 0.3 0.3>
    gravityOverride {z=-980}
    physicsBodyDefinitions [{
      boundBone {boneName=dyn_ribbonChest_c02_r}
      boxExtents=<16 16 16>
      constraintSetup {angularLimitsMax=<0 25 3>, angularLimitsMin=<0 -20 -40>, angularTarget=<1 0 0>}
      localJointOffset=<2 -1 0>
     }]
    relativeSpaceBone {boneName=spine_05}
    simSpaceSettings {simSpaceAngularAlpha=0.1}
   }
   n78 = Anim Dynamics - Bone: dyn_ribbonChest_d01_r!(ComponentPose=@n77, Alpha=1, GravityOverride:Vector="0,0,-980", ExternalForce="0,0,0") -> (Pose) {
    angularDampingOverride=0.999
    angularSpringConstant=3000
    bAngularSpring=true
    bOverrideAngularDamping=true
    bUseGravityOverride=true
    gravityScale=0.35
    simulationSpace=RootRelative
    boundBone {boneName=dyn_ribbonChest_d01_r}
    componentAppliedLinearAccClamp=<3000 3000 3000>
    componentLinearAccScale=<0.3 0.3 0.3>
    componentLinearVelScale=<0.3 0.3 0.3>
    gravityOverride {z=-980}
    physicsBodyDefinitions [{
      boundBone {boneName=dyn_ribbonChest_d01_r}
      boxExtents=<16 16 16>
      constraintSetup {angularLimitsMax=<0 25 3>, angularLimitsMin=<0 -45 -40>, angularTarget=<1 0 0>}
      localJointOffset=<2 -1 0>
     }]
    relativeSpaceBone {boneName=spine_05}
    simSpaceSettings {simSpaceAngularAlpha=0.1}
   }
   n79 = Anim Dynamics - Bone: dyn_ribbonLeg_a01_l!(ComponentPose=@n78, Alpha=1, GravityOverride:Vector="0,0,-980", ExternalForce="0,0,0") -> (Pose) {
    angularDampingOverride=0.999
    angularSpringConstant=3000
    bAngularSpring=true
    bOverrideAngularDamping=true
    bUseGravityOverride=true
    gravityScale=0.35
    simulationSpace=RootRelative
    boundBone {boneName=dyn_ribbonLeg_a01_l}
    componentAppliedLinearAccClamp=<3000 3000 3000>
    componentLinearAccScale=<0.3 0.3 0.6>
    componentLinearVelScale=<0.3 0.3 0.6>
    gravityOverride {z=-980}
    physicsBodyDefinitions [{
      boundBone {boneName=dyn_ribbonLeg_a01_l}
      boxExtents=<13 13 13>
      constraintSetup {angularLimitsMax=<0 60 3>, angularLimitsMin=<0 -2 -60>, angularTarget=<1 0 0>}
      localJointOffset=<3 0 -2>
     }]
    relativeSpaceBone {boneName=thigh_l}
    simSpaceSettings {simSpaceAngularAlpha=0.1}
   }
   n80 = Anim Dynamics - Bone: dyn_ribbonLeg_b01_l!(ComponentPose=@n79, Alpha=1, GravityOverride:Vector="0,0,-980", ExternalForce="0,0,0") -> (Pose) {
    angularDampingOverride=0.999
    angularSpringConstant=3000
    bAngularSpring=true
    bOverrideAngularDamping=true
    bUseGravityOverride=true
    gravityScale=0.35
    simulationSpace=RootRelative
    boundBone {boneName=dyn_ribbonLeg_b01_l}
    componentAppliedLinearAccClamp=<3000 3000 3000>
    componentLinearAccScale=<0.3 0.3 0.6>
    componentLinearVelScale=<0.3 0.3 0.6>
    gravityOverride {z=-980}
    physicsBodyDefinitions [{
      boundBone {boneName=dyn_ribbonLeg_b01_l}
      boxExtents=<13 13 13>
      constraintSetup {angularLimitsMax=<0 2 3>, angularLimitsMin=<0 -60 -60>, angularTarget=<1 0 0>}
      localJointOffset=<3 0 2>
     }]
    relativeSpaceBone {boneName=thigh_l}
    simSpaceSettings {simSpaceAngularAlpha=0.1}
   }
   n81 = Anim Dynamics - Bone: dyn_ribbonLeg_c01_l!(ComponentPose=@n80, Alpha=1, GravityOverride:Vector="0,0,-980", ExternalForce="0,0,0") -> (Pose) {
    angularDampingOverride=0.999
    angularSpringConstant=3000
    bAngularSpring=true
    bOverrideAngularDamping=true
    bUseGravityOverride=true
    gravityScale=0.35
    simulationSpace=RootRelative
    boundBone {boneName=dyn_ribbonLeg_c01_l}
    componentAppliedLinearAccClamp=<3000 3000 3000>
    componentLinearAccScale=<0.3 0.3 0.6>
    componentLinearVelScale=<0.3 0.3 0.6>
    gravityOverride {z=-980}
    physicsBodyDefinitions [{
      boundBone {boneName=dyn_ribbonLeg_c01_l}
      boxExtents=<13 13 13>
      constraintSetup {angularLimitsMax=<0 2 3>, angularLimitsMin=<0 -60 -60>, angularTarget=<1 0 0>}
      localJointOffset=<3 0 2>
     }]
    relativeSpaceBone {boneName=thigh_l}
    simSpaceSettings {simSpaceAngularAlpha=0.1}
   }
   n82 = Anim Dynamics - Bone: dyn_ribbonLeg_c02_l!(ComponentPose=@n81, Alpha=1, GravityOverride:Vector="0,0,-980", ExternalForce="0,0,0") -> (Pose) {
    angularDampingOverride=0.999
    angularSpringConstant=3000
    bAngularSpring=true
    bOverrideAngularDamping=true
    bUseGravityOverride=true
    gravityScale=0.35
    simulationSpace=RootRelative
    boundBone {boneName=dyn_ribbonLeg_c02_l}
    componentAppliedLinearAccClamp=<3000 3000 3000>
    componentLinearAccScale=<0.3 0.3 0.3>
    componentLinearVelScale=<0.3 0.3 0.3>
    gravityOverride {z=-980}
    physicsBodyDefinitions [{
      boundBone {boneName=dyn_ribbonLeg_c02_l}
      boxExtents=<16 16 16>
      constraintSetup {angularLimitsMax=<0 25 3>, angularLimitsMin=<0 -25 -20>, angularTarget=<1 0 0>}
      localJointOffset=<1.5 0 0>
     }]
    relativeSpaceBone {boneName=thigh_l}
    simSpaceSettings {simSpaceAngularAlpha=0.1}
   }
   n83 = Anim Dynamics - Bone: dyn_ribbonLeg_d01_l!(ComponentPose=@n82, Alpha=1, GravityOverride:Vector="0,0,-980", ExternalForce="0,0,0") -> (Pose) {
    angularDampingOverride=0.999
    angularSpringConstant=3000
    bAngularSpring=true
    bOverrideAngularDamping=true
    bUseGravityOverride=true
    gravityScale=0.35
    simulationSpace=RootRelative
    boundBone {boneName=dyn_ribbonLeg_d01_l}
    componentAppliedLinearAccClamp=<3000 3000 3000>
    componentLinearAccScale=<0.3 0.3 0.6>
    componentLinearVelScale=<0.3 0.3 0.6>
    gravityOverride {z=-980}
    physicsBodyDefinitions [{
      boundBone {boneName=dyn_ribbonLeg_d01_l}
      boxExtents=<13 13 13>
      constraintSetup {angularLimitsMax=<0 60 3>, angularLimitsMin=<0 -2 -60>, angularTarget=<1 0 0>}
      localJointOffset=<3 0 -2>
     }]
    relativeSpaceBone {boneName=thigh_l}
    simSpaceSettings {simSpaceAngularAlpha=0.1}
   }
   n84 = Anim Dynamics - Bone: dyn_ribbonLeg_a01_r!(ComponentPose=@n83, Alpha=1, GravityOverride:Vector="0,0,-980", ExternalForce="0,0,0") -> (Pose) {
    angularDampingOverride=0.999
    angularSpringConstant=3000
    bAngularSpring=true
    bOverrideAngularDamping=true
    bUseGravityOverride=true
    gravityScale=0.35
    simulationSpace=RootRelative
    boundBone {boneName=dyn_ribbonLeg_a01_r}
    componentAppliedLinearAccClamp=<3000 3000 3000>
    componentLinearAccScale=<0.3 0.3 0.6>
    componentLinearVelScale=<0.3 0.3 0.6>
    gravityOverride {z=-980}
    physicsBodyDefinitions [{
      boundBone {boneName=dyn_ribbonLeg_a01_r}
      boxExtents=<13 13 13>
      constraintSetup {angularLimitsMax=<0 60 3>, angularLimitsMin=<0 -2 -60>, angularTarget=<1 0 0>}
      localJointOffset=<3 0 -2>
     }]
    relativeSpaceBone {boneName=thigh_r}
    simSpaceSettings {simSpaceAngularAlpha=0.1}
   }
   n85 = Anim Dynamics - Bone: dyn_ribbonLeg_b01_r!(ComponentPose=@n84, Alpha=1, GravityOverride:Vector="0,0,-980", ExternalForce="0,0,0") -> (Pose) {
    angularDampingOverride=0.999
    angularSpringConstant=3000
    bAngularSpring=true
    bOverrideAngularDamping=true
    bUseGravityOverride=true
    gravityScale=0.35
    simulationSpace=RootRelative
    boundBone {boneName=dyn_ribbonLeg_b01_r}
    componentAppliedLinearAccClamp=<3000 3000 3000>
    componentLinearAccScale=<0.3 0.3 0.6>
    componentLinearVelScale=<0.3 0.3 0.6>
    gravityOverride {z=-980}
    physicsBodyDefinitions [{
      boundBone {boneName=dyn_ribbonLeg_b01_r}
      boxExtents=<13 13 13>
      constraintSetup {angularLimitsMax=<0 2 3>, angularLimitsMin=<0 -60 -60>, angularTarget=<1 0 0>}
      localJointOffset=<3 0 2>
     }]
    relativeSpaceBone {boneName=thigh_r}
    simSpaceSettings {simSpaceAngularAlpha=0.1}
   }
   n86 = Anim Dynamics - Bone: dyn_ribbonLeg_c01_r!(ComponentPose=@n85, Alpha=1, GravityOverride:Vector="0,0,-980", ExternalForce="0,0,0") -> (Pose) {
    angularDampingOverride=0.999
    angularSpringConstant=3000
    bAngularSpring=true
    bOverrideAngularDamping=true
    bUseGravityOverride=true
    gravityScale=0.35
    simulationSpace=RootRelative
    boundBone {boneName=dyn_ribbonLeg_c01_r}
    componentAppliedLinearAccClamp=<3000 3000 3000>
    componentLinearAccScale=<0.3 0.3 0.6>
    componentLinearVelScale=<0.3 0.3 0.6>
    gravityOverride {z=-980}
    physicsBodyDefinitions [{
      boundBone {boneName=dyn_ribbonLeg_c01_r}
      boxExtents=<13 13 13>
      constraintSetup {angularLimitsMax=<0 2 3>, angularLimitsMin=<0 -60 -60>, angularTarget=<1 0 0>}
      localJointOffset=<3 0 2>
     }]
    relativeSpaceBone {boneName=thigh_r}
    simSpaceSettings {simSpaceAngularAlpha=0.1}
   }
   n87 = Anim Dynamics - Bone: dyn_ribbonLeg_c02_r!(ComponentPose=@n86, Alpha=1, GravityOverride:Vector="0,0,-980", ExternalForce="0,0,0") -> (Pose) {
    angularDampingOverride=0.999
    angularSpringConstant=3000
    bAngularSpring=true
    bOverrideAngularDamping=true
    bUseGravityOverride=true
    gravityScale=0.35
    simulationSpace=RootRelative
    boundBone {boneName=dyn_ribbonLeg_c02_r}
    componentAppliedLinearAccClamp=<3000 3000 3000>
    componentLinearAccScale=<0.3 0.3 0.3>
    componentLinearVelScale=<0.3 0.3 0.3>
    gravityOverride {z=-980}
    physicsBodyDefinitions [{
      boundBone {boneName=dyn_ribbonLeg_c02_r}
      boxExtents=<16 16 16>
      constraintSetup {angularLimitsMax=<0 25 3>, angularLimitsMin=<0 -25 -20>, angularTarget=<1 0 0>}
      localJointOffset=<1.5 0 0>
     }]
    relativeSpaceBone {boneName=thigh_r}
    simSpaceSettings {simSpaceAngularAlpha=0.1}
   }
   n88 = Anim Dynamics - Bone: dyn_ribbonLeg_d01_r!(ComponentPose=@n87, Alpha=1, GravityOverride:Vector="0,0,-980", ExternalForce="0,0,0") -> (Pose) {
    angularDampingOverride=0.999
    angularSpringConstant=3000
    bAngularSpring=true
    bOverrideAngularDamping=true
    bUseGravityOverride=true
    gravityScale=0.35
    simulationSpace=RootRelative
    boundBone {boneName=dyn_ribbonLeg_d01_r}
    componentAppliedLinearAccClamp=<3000 3000 3000>
    componentLinearAccScale=<0.3 0.3 0.6>
    componentLinearVelScale=<0.3 0.3 0.6>
    gravityOverride {z=-980}
    physicsBodyDefinitions [{
      boundBone {boneName=dyn_ribbonLeg_d01_r}
      boxExtents=<13 13 13>
      constraintSetup {angularLimitsMax=<0 60 3>, angularLimitsMin=<0 -2 -60>, angularTarget=<1 0 0>}
      localJointOffset=<3 0 -2>
     }]
    relativeSpaceBone {boneName=thigh_r}
    simSpaceSettings {simSpaceAngularAlpha=0.1}
   }
   n89 = RigidBody!(ComponentPose=@n88, Alpha, OverrideWorldGravity:Vector="0,0,0", ExternalForce) -> (Pose) {
    bOverrideWorldGravity=true
    overridePhysicsAsset=&/F_MED_ToothGreek_Physics
    simulationSpace=BaseBoneSpace
    baseBoneRef {boneName=root}
    componentAppliedLinearAccClamp=<3000 3000 5000>
    simSpaceSettings {maxAngularAcceleration=40, maxAngularVelocity=10, maxLinearAcceleration=500, maxLinearVelocity=200, worldAlpha=1, externalLinearDragV {z=1}}
   }
   Save cached pose 'ribbonAnimDyn'(Pose=Component To Local(ComponentPose=@n89)) {cacheName=ribbonAnimDyn}
  ]}
 EventGraph {nodes [
   Event BlueprintUpdateAnimation()
   > n1 = Sequence()
   > n2 = "Gravity Override State Function (Parameter Asset) (Nativized)"(
    CharacterPartAnimInstance=Self-Reference()
    GameTimeInSeconds=GetGameTimeInSeconds()
    DebugJointName=spine_05
    Parameters=&/F_MED_ToothGreek_Parameters
   )
   > Set Gravity Override(Gravity Override=@n2)
   Set isCrouchingOrSkydiving(execute=@n1.then_1, isCrouchingOrSkydiving:bool=OR Boolean(A=Get bIsCrouching(), B=Get bIsSkyDiving())) -> (Output_Get:bool)
   n3 = Get bRigidBodyEnabled()
   n4 = CheckForBackbling()
   > Set isWearingBackbling / Cape / NoRBAN(isWearingBackbling / Cape / NoRBAN:bool=OR Boolean!(A=@n4, B=@n4.bIsBackblingACape, C:bool=@n4.bIsBackblingLong, D:bool=NOT Boolean(A=@n3)) -> (ReturnValue)) -> (Output_Get:bool)
  ]}
]
variables [
 Gravity Override {type=Vector}
 isCrouchingOrSkydiving {type=boolean}
 isWearingBackbling / Cape / NoRBAN {type=boolean}
]
