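# Physics asset of cape colliders for the F_MED_ToothGreek female medium player body, giving collision volumes for the cape.
# Capsules cover head, clavicles, spine, pelvis and feet, with crossed second capsules on spine_02 to spine_04.
# Six back-skirt dyn bones (center, left, right, two segments each) carry same-sized capsules, so the skirt also acts as a collider.
# Arm, leg and neck bodies have no shapes, and no constraints exist.
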PhysicsAsset /BRCosmetics/Characters/Player/Female/Medium/Bodies/F_MED_ToothGreek/Meshes/F_MED_ToothGreek_CapeColliders
bodies {
 calf_l []
 calf_r []
 clavicle_l [{shapeType=Capsule, length=6.5623, radius=7.5194, center=<3.9566 -1.0862 -2.1127>, rotation=<rot -77.5566 177.3915 7.7314>}]
 clavicle_r [{shapeType=Capsule, length=6.5623, radius=7.5194, center=<-3.9566 1.0862 2.1127>, rotation=<rot -77.5566 177.3915 -172.2686>}]
 dyn_skirt_bk_c_01 [{
   shapeType=Capsule
   length=12.9794
   radius=6.9421
   shapeName=dyn_skirt_bk_c_02_capsule
   center=<7.7346 4.1032 0.1322>
   rotation=<rot 0.4929 -90 0.6881>
  }]
 dyn_skirt_bk_c_02 [{
   shapeType=Capsule
   length=12.9794
   radius=6.9421
   shapeName=dyn_skirt_bk_c_02_capsule
   center=<5.4358 4.0606 0.1601>
   rotation=<rot 0.4929 -90 0.6881>
  }]
 dyn_skirt_bk_l_01 [{
   shapeType=Capsule
   length=12.9794
   radius=6.9421
   shapeName=dyn_skirt_bk_c_02_capsule
   center=<7.7346 4.1032 0.1322>
   rotation=<rot 0.4929 -90 0.6881>
  }]
 dyn_skirt_bk_l_02 [{
   shapeType=Capsule
   length=12.9794
   radius=6.9421
   shapeName=dyn_skirt_bk_c_02_capsule
   center=<5.4353 4.6806 0.1096>
   rotation=<rot 4.6639 -89.9497 0.6904>
  }]
 dyn_skirt_bk_r_01 [{
   shapeType=Capsule
   length=12.9794
   radius=6.9421
   shapeName=dyn_skirt_bk_c_02_capsule
   center=<7.7346 4.1032 0.1322>
   rotation=<rot 0.4929 -90 0.6881>
  }]
 dyn_skirt_bk_r_02 [{
   shapeType=Capsule
   length=12.9794
   radius=6.9421
   shapeName=dyn_skirt_bk_c_02_capsule
   center=<5.4358 4.5375 0.156>
   rotation=<rot 5.9219 -89.9345 0.6918>
  }]
 foot_l [{shapeType=Capsule, length=11.0765, radius=8.0344, center=<-4.1811 6.0027 -0.338>, rotation=<rot 0.0388 1.4701 -89.3982>}]
 foot_r [{shapeType=Capsule, length=11.0765, radius=8.0344, center=<4.1811 -6.0027 0.338>, rotation=<rot 0.0388 1.4701 90.6018>}]
 head [{shapeType=Capsule, length=0, radius=18, center=<9.0697 0.0746 0>, rotation=<rot 0 -90 0>}]
 lowerarm_l []
 lowerarm_r []
 neck_02 []
 pelvis [{shapeType=Capsule, length=9.6042, radius=11.1441, center=<-2.4471 -2.5289 -0.1981>}]
 spine_01 [{shapeType=Capsule, length=10, radius=14.3638, center=<-8.9021 -1.1552 0.4432>}]
 spine_02 [
  {shapeType=Capsule, length=10, radius=11.7317, center=<-2.6885 0.4528 0.4431>}
  {shapeType=Capsule, length=20.6875, radius=9.3125, center=<-1.8317 1.9362 0>, rotation=<rot 90 0 0>}
 ]
 spine_03 [
  {shapeType=Capsule, length=6.1863, radius=12, center=<1.4542 -0.4545 0>}
  {shapeType=Capsule, length=20.7188, radius=9.2812, center=<-1.2163 0.549 0>, rotation=<rot 90 180 -175.0732>}
 ]
 spine_04 [
  {shapeType=Capsule, length=9.0076, radius=10.3916, center=<5.6234 -0.6779 0>}
  {shapeType=Capsule, length=10.0312, radius=9.2812, center=<-1.2326 1.308 0>, rotation=<rot 90 0 0>}
 ]
 spine_05 [{shapeType=Capsule, length=9.678, radius=12.3571, center=<-1.5315 0.762 0>}]
 thigh_l []
 thigh_r []
 upperarm_l []
 upperarm_r []
}
constraints []
