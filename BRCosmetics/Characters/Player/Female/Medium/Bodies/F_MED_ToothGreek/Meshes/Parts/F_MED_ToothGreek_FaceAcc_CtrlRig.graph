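# Control Rig for the F_MED_ToothGreek face accessory that applies corrective offsets, mainly to hair bones.
# Spherical pose readers on head and clavicles weight Modify Transforms on the shoulder hair bases; clavicle corrections fade out when emoting or skydiving.
# Interpolated bools blend extra adjustments for FE state and the GoldenGlory emote, some multiplied by head pose.
# The DeltaRotation function goes unused and the pec offset is disconnected from execution.
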ControlRigBlueprint /BRCosmetics/Characters/Player/Female/Medium/Bodies/F_MED_ToothGreek/Meshes/Parts/F_MED_ToothGreek_FaceAcc_CtrlRig
graphs [
 New Function_ContainedGraph {nodes [
   n1 = RigVMFunctionEntryNode()
   RigVMFunction_MathBoolToInteger = &RigVMFunction_MathBoolToInteger(Value=@n1.onlyTwist)
   RigVMFunction_MathQuaternionSwingTwist = &RigVMFunction_MathQuaternionSwingTwist(
    twistAxis=@n1.twistAxis
    Input=&RigUnit_GetRelativeTransformForItem(bChildInitial=False, bParentInitial=False, child=@n1.child, parent=@n1.parent).RelativeTransform.Rotation
   )
   n2 = RigVMDispatchNode(
    Values=((),())
    Index=@RigVMFunction_MathBoolToInteger.Result
    Values.0=@RigVMFunction_MathQuaternionSwingTwist.Swing
    Values.1=@RigVMFunction_MathQuaternionSwingTwist.Twist
   ) {template="DISPATCH_RigVMDispatch_SelectInt32(in Index,in Values,out Result)"}
   RigVMFunction_MathQuaternionToEuler = &RigVMFunction_MathQuaternionToEuler(Value=@n2.Result, RotationOrder=@n1.Rotation Order)
   > DISPATCH_RigVMDispatch_Print = RigVMDispatchNode(ScreenColor=(R=1,G=1,B=1,A=1), ScreenDuration=1.5, Enabled=@n1.debugRotation, Value=@RigVMFunction_MathQuaternionToEuler.Result) {template="DISPATCH_RigVMDispatch_Print(in Prefix,in Value,in Enabled,in ScreenDuration,in ScreenColor)"}
   Cast = RigVMDispatchNode(Value=@n1.triggerAxis) {template="DISPATCH_RigVMDispatch_CastEnumToInt(in Value,out Result)"}
   DISPATCH_RigVMDispatch_SelectInt32 = RigVMDispatchNode(
    Values=(0,0,0)
    Index=@Cast.Result
    Values.0=@RigVMFunction_MathQuaternionToEuler.Result.X
    Values.1=@RigVMFunction_MathQuaternionToEuler.Result.Y
    Values.2=@RigVMFunction_MathQuaternionToEuler.Result.Z
   ) {template="DISPATCH_RigVMDispatch_SelectInt32(in Index,in Values,out Result)"}
   Remap = &RigVMFunction_MathDoubleRemap(
    bClamp=False
    Value=@DISPATCH_RigVMDispatch_SelectInt32.Result
    TargetMinimum=@n1.outputMinMax.X
    TargetMaximum=@n1.outputMinMax.Y
    SourceMinimum=@n1.rotationMinMax.X
    SourceMaximum=@n1.rotationMinMax.Y
   )
   Absolute = &RigVMFunction_MathDoubleAbs(Value=@Remap.Result)
   Multiply = &RigVMFunction_MathDoubleMul(
    A=&RigVMFunction_MathDoubleClamp(Value=RigVMDispatchNode(True=@Absolute.Result, Condition=@n1.absoluteValue, False=@Remap.Result) {template="DISPATCH_RigVMDispatch_If(in Condition,in True,in False,out Result)"}.Result).Result
    B=@n1.premult
   )
   > RigVMFunctionReturnNode(outputQuat=@n2.Result, outputWeight=@Multiply.Result)
  ]}
 RigVMFunctionLibrary {nodes [DeltaRotation v4 = RigVMCollapseNode(
    ExecuteContext=()
    Rotation Order=ZYX
    child=(Type=Bone,Name="None")
    outputMinMax=(X=0,Y=1)
    parent=(Type=Bone,Name="None")
    premult=1
    rotationMinMax=(X=0,Y=0)
    twistAxis=(X=1,Y=0,Z=0)
   )]}
 RigVMModel {nodes [
   RigUnit_BeginExecution = &RigUnit_BeginExecution()
   n1 = &RigUnit_SphericalPoseReader(
    ActiveRegionScaleFactors=(PositiveWidth=1.2,NegativeWidth=1,PositiveHeight=1.5,NegativeHeight=1)
    ActiveRegionSize=0.15
    Debug=(bDrawDebug=false,bDraw2D=False,bDrawLocalAxes=False,DebugScale=25,DebugSegments=20,DebugThickness=0.25)
    DriverAxis=(X=1,Y=0,Z=0)
    DriverItem=(Type=Bone,Name="head")
    FalloffRegionScaleFactors=(PositiveWidth=2,NegativeWidth=1,PositiveHeight=2,NegativeHeight=2)
    FalloffSize=0.3
    FlipHeightScaling=False
    FlipWidthScaling=false
    OptionalParentItem=(Type=Bone,Name="neck_02")
    RotationOffset=(X=90,Y=0,Z=-45)
    ExecutePin=@RigUnit_BeginExecution.ExecutePin
   )
   > &RigUnit_ModifyTransforms(
    ItemToModify="((Item=(Type=Bone,Name=\"dyn_hair_shoulderBase_l\"),Transform=(Rotation=(X=0,Y=0,Z=0,W=1),Translation=(X=-0.5,Y=-1,Z=0),Scale3D=(X=1,Y=1,Z=1))),(Item=(Type=Bone,Name=\"dyn_hair_shoulderBase_r\"),Transform=(Rotation=(X=0,Y=0,Z=0,W=1),Translation=(X=3,Y=-1,Z=0), ...(truncated)"
    WeightMinimum=0
    Weight=@n1.OutputParam
   )
   > n2 = &RigUnit_SphericalPoseReader(
    ActiveRegionScaleFactors=(PositiveWidth=1,NegativeWidth=1,PositiveHeight=1,NegativeHeight=1)
    ActiveRegionSize=0.05
    Debug=(bDrawDebug=false,bDraw2D=False,bDrawLocalAxes=False,DebugScale=25,DebugSegments=20,DebugThickness=0.25)
    DriverAxis=(X=-1,Y=0,Z=0)
    DriverItem=(Type=Bone,Name="clavicle_r")
    FalloffRegionScaleFactors=(PositiveWidth=2.2,NegativeWidth=1,PositiveHeight=1,NegativeHeight=1)
    FalloffSize=0.15
    FlipHeightScaling=False
    FlipWidthScaling=False
    OptionalParentItem=(Type=Bone,Name="spine_05")
    RotationOffset=(X=-45,Y=140,Z=-40)
   )
   RigVMFunction_MathBoolToFloat = &RigVMFunction_MathBoolToFloat(Value=RigVMVariableNode() {variable=isEmotingOrSkydiving_CR}.Value)
   n3 = &RigVMFunction_AlphaInterp(InterpSpeedDecreasing=10, InterpSpeedIncreasing=2, Scale=1, bClampResult=true, bMapRange=true, Value=@RigVMFunction_MathBoolToFloat.Result)
   Remap = &RigVMFunction_MathDoubleRemap(SourceMaximum=1, SourceMinimum=0, TargetMaximum=0, TargetMinimum=1, bClamp=False, Value=@n3.Result)
   Multiply = &RigVMFunction_MathDoubleMul(B=@Remap.Result, A=@n2.OutputParam)
   > &RigUnit_ModifyTransforms(
    ItemToModify=((Item=(Type=Bone,Name="dyn_hair_shoulderBase_r"),Transform=(Rotation=(X=0,Y=0,Z=0,W=1),Translation=(X=-1,Y=-1.5,Z=-3),Scale3D=(X=1,Y=1,Z=1))))
    WeightMinimum=0
    Weight=@Multiply.Result
   )
   > n4 = &RigUnit_SphericalPoseReader(
    ActiveRegionScaleFactors=(PositiveWidth=1,NegativeWidth=1,PositiveHeight=1,NegativeHeight=1)
    ActiveRegionSize=0.05
    Debug=(bDrawDebug=false,bDraw2D=False,bDrawLocalAxes=False,DebugScale=25,DebugSegments=20,DebugThickness=0.25)
    DriverAxis=(X=1,Y=0,Z=0)
    DriverItem=(Type=Bone,Name="clavicle_l")
    FalloffRegionScaleFactors=(PositiveWidth=1,NegativeWidth=1,PositiveHeight=1,NegativeHeight=1)
    FalloffSize=0.15
    FlipHeightScaling=False
    FlipWidthScaling=False
    OptionalParentItem=(Type=Bone,Name="spine_05")
    RotationOffset=(X=-45,Y=40,Z=130)
   )
   > &RigUnit_ModifyTransforms(
    ItemToModify=((Item=(Type=Bone,Name="dyn_hair_shoulderBase_l"),Transform=(Rotation=(X=0,Y=0,Z=0,W=1),Translation=(X=0,Y=-2,Z=3),Scale3D=(X=1,Y=1,Z=1))))
    WeightMinimum=0
    Weight=&RigVMFunction_MathDoubleMul(B=@Remap.Result, A=@n4.OutputParam).Result
   )
   > n5 = &RigUnit_SphericalPoseReader(
    ActiveRegionScaleFactors=(PositiveWidth=1.2,NegativeWidth=1,PositiveHeight=1,NegativeHeight=1)
    ActiveRegionSize=0.15
    Debug=(bDrawDebug=false,bDraw2D=False,bDrawLocalAxes=False,DebugScale=25,DebugSegments=20,DebugThickness=0.25)
    DriverAxis=(X=1,Y=0,Z=0)
    DriverItem=(Type=Bone,Name="head")
    FalloffRegionScaleFactors=(PositiveWidth=2,NegativeWidth=1,PositiveHeight=1,NegativeHeight=1)
    FalloffSize=0.3
    FlipHeightScaling=False
    FlipWidthScaling=False
    OptionalParentItem=(Type=Bone,Name="neck_02")
    RotationOffset=(X=120,Y=0,Z=-140)
   )
   > &RigUnit_ModifyTransforms(
    ItemToModify="((Item=(Type=Bone,Name=\"dyn_hair_shoulderBase_r\"),Transform=(Rotation=(X=0,Y=0,Z=0,W=1),Translation=(X=4,Y=2,Z=0),Scale3D=(X=1,Y=1,Z=1))),(Item=(Type=Bone,Name=\"dyn_hair_shoulderBase_l\"),Transform=(Rotation=(X=0,Y=0,Z=0,W=1),Translation=(X=1,Y=2,Z=0),Scale ...(truncated)"
    WeightMinimum=0
    Weight=@n5.OutputParam
   )
   > n6 = &RigUnit_SphericalPoseReader(
    ActiveRegionScaleFactors=(PositiveWidth=1,NegativeWidth=1,PositiveHeight=1,NegativeHeight=1)
    ActiveRegionSize=0.15
    Debug=(bDrawDebug=false,bDraw2D=False,bDrawLocalAxes=False,DebugScale=25,DebugSegments=20,DebugThickness=0.25)
    DriverAxis=(X=0,Y=1,Z=0)
    DriverItem=(Type=Bone,Name="head")
    FalloffRegionScaleFactors=(PositiveWidth=1,NegativeWidth=1,PositiveHeight=1,NegativeHeight=1)
    FalloffSize=0.3
    FlipHeightScaling=False
    FlipWidthScaling=False
    OptionalParentItem=(Type=Bone,Name="neck_02")
    RotationOffset=(X=45,Y=0,Z=0)
   )
   > &RigUnit_ModifyTransforms(
    ItemToModify="((Item=(Type=Bone,Name=\"dyn_hair_shoulderBase_r\"),Transform=(Rotation=(X=0,Y=0,Z=0,W=1),Translation=(X=-1,Y=-2,Z=0),Scale3D=(X=1,Y=1,Z=1))),(Item=(Type=Bone,Name=\"dyn_hair_shoulderBase_l\"),Transform=(Rotation=(X=0,Y=0,Z=0,W=1),Translation=(X=0,Y=0,Z=0),Sca ...(truncated)"
    WeightMinimum=0
    Weight=@n6.OutputParam
   )
   RigVMFunction_MathBoolNot = &RigVMFunction_MathBoolNot(Value=RigVMVariableNode() {variable=GoldenGloryEmote}.Value)
   RigVMFunction_MathBoolAnd = &RigVMFunction_MathBoolAnd(B=@RigVMFunction_MathBoolNot.Result, A=RigVMVariableNode() {variable=isInFE_CR}.Value)
   n7 = &RigVMFunction_AlphaInterp(
    InterpSpeedDecreasing=5
    InterpSpeedIncreasing=5
    Scale=1
    bClampResult=true
    Value=&RigVMFunction_MathBoolToFloat(Value=@RigVMFunction_MathBoolAnd.Result).Result
   )
   > &RigUnit_ModifyTransforms(
    ItemToModify=((Item=(Type=Bone,Name="base_hair_r"),Transform=(Rotation=(X=0,Y=0,Z=0,W=-1),Translation=(X=-2.5,Y=0,Z=0),Scale3D=(X=1,Y=1,Z=1))))
    WeightMinimum=0
    Weight=@n7.Result
   )
   > n8 = &RigUnit_SphericalPoseReader(
    ActiveRegionScaleFactors=(PositiveWidth=1,NegativeWidth=1,PositiveHeight=1,NegativeHeight=1)
    ActiveRegionSize=0.1
    Debug=(bDrawDebug=false,bDraw2D=False,bDrawLocalAxes=False,DebugScale=25,DebugSegments=20,DebugThickness=0.25)
    DriverAxis=(X=0,Y=1,Z=0)
    DriverItem=(Type=Bone,Name="head")
    FalloffRegionScaleFactors=(PositiveWidth=1,NegativeWidth=1,PositiveHeight=1,NegativeHeight=1)
    FalloffSize=0.3
    FlipHeightScaling=False
    FlipWidthScaling=False
    OptionalParentItem=(Type=Bone,Name="spine_04")
    RotationOffset=(X=140,Y=0,Z=0)
   )
   > &RigUnit_ModifyTransforms(
    ItemToModify=((Item=(Type=Bone,Name="dyn_hair_shoulderBase_l"),Transform=(Rotation=(X=0,Y=0,Z=0,W=-1),Translation=(X=-3,Y=0,Z=0),Scale3D=(X=1,Y=1,Z=1))))
    WeightMinimum=0
    Weight=&RigVMFunction_MathDoubleMul(B=@n7.Result, A=@n8.OutputParam).Result
   )
   AlphaInterp = &RigVMFunction_AlphaInterp(
    InterpSpeedDecreasing=5
    InterpSpeedIncreasing=5
    Scale=1
    bClampResult=true
    Value=&RigVMFunction_MathBoolToFloat(Value=RigVMVariableNode() {variable=GoldenGloryEmote}.Value).Result
   )
   > &RigUnit_ModifyTransforms(
    ItemToModify="((Item=(Type=Bone,Name=\"base_hair_r\"),Transform=(Rotation=(X=0,Y=0,Z=0,W=1),Translation=(X=0,Y=0,Z=0),Scale3D=(X=1,Y=0.7,Z=1))),(Item=(Type=Bone,Name=\"lowerarm_twist_02_r\"),Transform=(Rotation=(X=0,Y=0,Z=0,W=1),Translation=(X=0,Y=-10,Z=0),Scale3D=(X=1,Y=1, ...(truncated)"
    WeightMinimum=0
    Weight=@AlphaInterp.Result
   )
   > n9 = &RigUnit_SphericalPoseReader(
    ActiveRegionScaleFactors=(PositiveWidth=1,NegativeWidth=1,PositiveHeight=1,NegativeHeight=1)
    ActiveRegionSize=0.1
    Debug=(bDrawDebug=true,bDraw2D=False,bDrawLocalAxes=False,DebugScale=25,DebugSegments=20,DebugThickness=0.25)
    DriverAxis=(X=0,Y=1,Z=0)
    DriverItem=(Type=Bone,Name="head")
    FalloffRegionScaleFactors=(PositiveWidth=1,NegativeWidth=1,PositiveHeight=2,NegativeHeight=1)
    FalloffSize=0.3
    FlipHeightScaling=False
    FlipWidthScaling=False
    OptionalParentItem=(Type=Bone,Name="spine_04")
    RotationOffset=(X=-10,Y=0,Z=0)
   )
   > &RigUnit_ModifyTransforms(
    ItemToModify=((Item=(Type=Bone,Name="base_hair_r"),Transform=(Rotation=(X=0,Y=0,Z=-0.1736,W=0.9848),Translation=(X=0,Y=0,Z=0),Scale3D=(X=1,Y=1,Z=1))))
    WeightMinimum=0
    Weight=&RigVMFunction_MathDoubleMul(A=@n9.OutputParam, B=@AlphaInterp.Result).Result
   )
   &RigUnit_ModifyTransforms(
    ExecutePin=()
    ItemToModify=((Item=(Type=Bone,Name="deform_pec_aim_base_r"),Transform=(Rotation=(X=0,Y=0,Z=0,W=-1),Translation=(X=0,Y=-3,Z=0),Scale3D=(X=1,Y=1,Z=1))))
    Weight=1
    WeightMinimum=0
   )
  ]}
]
hierarchy [
 root {loc=<0 0 0>, rot=<0 0 0 1>}
 pelvis {parent=root, loc=<0 -0.5301 98.0571>, rot=<0.0127 0.707 -0.0127 -0.707>}
 spine_01 {parent=pelvis, loc=<6.1208 -0.9328 0>, rot=<0 0 0.034 -0.9994>}
 spine_02 {parent=spine_01, loc=<6.5997 -0.2095 0>, rot=<0 0 0.0238 -0.9997>}
 spine_03 {parent=spine_02, loc=<6.4914 -0.1237 0>, rot=<0 0 0.0149 -0.9999>}
 spine_04 {parent=spine_03, loc=<6.9598 0.0794 0>, rot=<0 0 -0.0132 -0.9999>}
 spine_05 {parent=spine_04, loc=<6.9987 0 0>, rot=<0 0 -0.0015 -1>}
 neck_01 {parent=spine_05, loc=<11.3138 0.1529 0>, rot=<0 0 -0.2003 -0.9797>}
 neck_02 {parent=neck_01, loc=<4.5627 0 0>, rot=<0 0 0.0504 -0.9987>}
 head {parent=neck_02, loc=<5.3914 0 0>, rot=<0 0 0.115 -0.9934>}
 base_hair_l {parent=head, loc=<0.719 -0.5351 -7.9045>, rot=<-0.0976 -0.015 0.9847 0.1434>}
 dyn_hair_01_l {parent=base_hair_l, loc=<0 0 0>, rot=<0 0 0 1>}
 dyn_hair_shoulderBase_l {parent=dyn_hair_01_l, loc=<12.5269 0 0>, rot=<0.0125 -0.0658 -0.0181 -0.9976>}
 dyn_hair_02_l {parent=dyn_hair_shoulderBase_l, loc=<0 0 0>, rot=<0 0 0 1>}
 dyn_hair_03_l {parent=dyn_hair_02_l, loc=<7 0 0>, rot=<-0.008 0.0755 -0.0625 -0.9952>}
 dyn_hair_04_l {parent=dyn_hair_03_l, loc=<7 0 0>, rot=<-0.0022 0.0551 -0.0753 -0.9956>}
 dyn_hair_05_l {parent=dyn_hair_04_l, loc=<7 0 0>, rot=<0 0.0367 0 -0.9993>}
 dyn_hair_06_l {parent=dyn_hair_05_l, loc=<7 0 0>, rot=<0 0 0 1>}
 base_hairStrandD_r {parent=dyn_hair_03_l, loc=<-0.1366 -2.3906 -0.796>, rot=<0.008 -0.1793 0.0149 -0.9836>}
 dyn_hairStrandD_01_r {parent=base_hairStrandD_r, loc=<0 0 0>, rot=<0 0 0 1>}
 base_hairStrandA_l {parent=base_hair_l, loc=<-0.07 -7.7396 1.1505>, rot=<-0.0097 0.0925 -0.186 -0.9781>}
 dyn_hairStrandA_01_l {parent=base_hairStrandA_l, loc=<0 0 0>, rot=<0 0 0 1>}
 dyn_hairStrandA_02_l {parent=dyn_hairStrandA_01_l, loc=<7.5644 0 0>, rot=<0 0 0 1>}
 base_hair_r {parent=head, loc=<0.7893 -0.5327 7.4037>, rot=<0.0904 0.0158 0.9822 0.1636>}
 dyn_hair_01_r {parent=base_hair_r, loc=<0 0 0>, rot=<0 0 0 1>}
 dyn_hair_shoulderBase_r {parent=dyn_hair_01_r, loc=<12.7383 0 0>, rot=<-0.0228 0.092 -0.0494 -0.9943>}
 dyn_hair_02_r {parent=dyn_hair_shoulderBase_r, loc=<0 0 0>, rot=<0 0 0 1>}
 dyn_hair_03_r {parent=dyn_hair_02_r, loc=<7 0 0>, rot=<0.0056 -0.0307 -0.0668 -0.9973>}
 dyn_hair_04_r {parent=dyn_hair_03_r, loc=<7 0 0>, rot=<-0.007 -0.1197 -0.0741 -0.99>}
 dyn_hair_05_r {parent=dyn_hair_04_r, loc=<7 0 0>, rot=<0 -0.0263 0 -0.9997>}
 dyn_hair_06_r {parent=dyn_hair_05_r, loc=<7 0 0>, rot=<0 0 0 1>}
 base_hairStrandF_r {parent=dyn_hair_03_r, loc=<4.5684 -3.1262 0.9202>, rot=<0.0083 0.0515 0.0545 -0.9971>}
 dyn_hairStrandF_01_r {parent=base_hairStrandF_r, loc=<0 0 0>, rot=<0 0 0 1>}
 base_hairStrandE_r {parent=dyn_hair_02_r, loc=<0.3316 -3.7128 -1.936>, rot=<0.0153 -0.1413 -0.105 -0.9843>}
 dyn_hairStrandE_01_r {parent=base_hairStrandE_r, loc=<0 0 0>, rot=<0 0 0 1>}
 base_hairStrandA_r {parent=base_hair_r, loc=<0.971 -8.6134 -1.0258>, rot=<0.0196 -0.1102 -0.1635 -0.9802>}
 dyn_hairStrandA_01_r {parent=base_hairStrandA_r, loc=<0 0 0>, rot=<0 0 0 1>}
 dyn_hairStrandA_02_r {parent=dyn_hairStrandA_01_r, loc=<5.663 0 0>, rot=<0 -0.0358 0 -0.9994>}
 clavicle_l {parent=spine_05, loc=<5.4831 -0.1053 -3.2946>, rot=<0.0465 -0.7289 -0.009 -0.6829>}
 upperarm_l {parent=clavicle_l, loc=<9.8658 -1.2136 -0.4105>, rot=<0.0164 -0.3245 -0.0269 -0.9454>}
 lowerarm_l {parent=upperarm_l, loc=<26.465 -0.8345 0.259>, rot=<0.0036 0.0161 -0.2298 -0.9731>}
 lowerarm_twist_02_l {parent=lowerarm_l, loc=<10.6707 0 0>, rot=<0 0 0 1>}
 clavicle_r {parent=spine_05, loc=<5.4831 -0.1053 3.2946>, rot=<0.7289 0.0465 -0.6829 0.009>}
 upperarm_r {parent=clavicle_r, loc=<-9.8658 1.2136 0.4105>, rot=<0.0164 -0.3245 -0.0269 -0.9454>}
 lowerarm_r {parent=upperarm_r, loc=<-26.465 0.8345 -0.259>, rot=<0.0036 0.0161 -0.2298 -0.9731>}
 lowerarm_twist_02_r {parent=lowerarm_r, loc=<-10.6707 0 0>, rot=<0 0 0 -1>}
 deform_pec_aim_base_l {parent=spine_04, loc=<3.2456 9.3309 -1.2237>, rot=<0.0603 -0.4033 0.3318 -0.8507>}
 deform_pec_aim_l {parent=deform_pec_aim_base_l, loc=<0 0 0>, rot=<0 0 0 1>}
 deform_pec_aim_base_r {parent=spine_04, loc=<3.2452 9.3309 1.2237>, rot=<-0.4033 -0.0603 0.8507 0.3318>}
 deform_pec_aim_r {parent=deform_pec_aim_base_r, loc=<0 0 0>, rot=<0 0 0 1>}
 head_null {parent=spine_01, type=Null, loc=<47.8163 1.0983 0>, rot=<0 0 -0.012 -0.9999>}
]
variables [
 GoldenGloryEmote {type=bool, default=False}
 isEmotingOrSkydiving_CR {type=bool, default=False}
 isInFE_CR {type=bool, default=False}
]
