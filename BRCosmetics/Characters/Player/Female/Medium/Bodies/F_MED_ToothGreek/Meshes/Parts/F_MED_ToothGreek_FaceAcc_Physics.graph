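# Physics asset for the face accessory part of the ToothGreek medium female player body, setting up physics for left and right dynamic hair strands.
# On each side, the first hair capsule hangs from the head, while segments 02 to 06 form a separate chain rooted at a small shoulder-base body.
# All joints lock twist and limit swing, more tightly on the second axis.
# Capsules on head, spine, clavicles, pecs, arms and thighs provide body collision shapes.
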PhysicsAsset /BRCosmetics/Characters/Player/Female/Medium/Bodies/F_MED_ToothGreek/Meshes/Parts/F_MED_ToothGreek_FaceAcc_Physics
bodies {
 clavicle_l [{shapeType=Capsule, length=6.7676, radius=4.6883, center=<4.9945 -0.6077 -0.132>, rotation=<rot 90 -3.7421 0>}]
 clavicle_r [{shapeType=Capsule, length=6.7676, radius=4.6883, center=<-5.017 0.2628 0.132>, rotation=<rot 90 176.2579 0>}]
 deform_pec_aim_l [
  {shapeType=Capsule, length=0, radius=5.7252, shapeName=deform_pec_aim_l_capsule, center=<4.7747 -0.9046 -0.7691>}
  {
   shapeType=Capsule
   length=1.6031
   radius=4.6275
   shapeName=deform_pec_aim_l_capsule
   center=<0.7067 -1.5047 2.2254>
   rotation=<rot 9.4485 -51.212 -45.7738>
  }
 ]
 deform_pec_aim_r [{shapeType=Capsule, length=0, radius=5.7252, shapeName=deform_pec_aim_r_capsule, center=<-4.7747 0.9046 0.7691>, rotation=<rot 0 0 -180>}]
 dyn_hair_01_l [{shapeType=Capsule, length=8.3442, radius=5.1061, center=<4.6419 -0.5489 0>, rotation=<rot 90 -5.0088 0>}]
 dyn_hair_01_r [{shapeType=Capsule, length=8.3442, radius=5.1061, center=<4.3017 -0.3671 0.7343>, rotation=<rot 85.0156 -5.0088 0.0001>}]
 dyn_hair_02_l [{shapeType=Capsule, length=4.8231, radius=3.3304, center=<2.8337 -0.5136 -0.2707>, rotation=<rot 88.9111 -13.2146 -9.8777>}]
 dyn_hair_02_r [{shapeType=Capsule, length=4.8231, radius=3.3304, center=<2.8349 -0.2852 -0.2697>, rotation=<rot 88.9111 -13.21 -12.9213>}]
 dyn_hair_03_l [{shapeType=Capsule, length=5.6163, radius=2.9336, center=<3.4667 -0.1902 0.1275>, rotation=<rot 88.1412 -179.9962 176.4302>}]
 dyn_hair_03_r [{shapeType=Capsule, length=5.2592, radius=2.9336, center=<3.4556 -0.3159 -0.1086>, rotation=<rot 86.7127 3.0857 1.1181>}]
 dyn_hair_04_l [{shapeType=Capsule, length=5.4873, radius=2.7888, center=<3.9753 0.2363 -0.1847>, rotation=<rot 88.8471 -178.4678 -178.5908>}]
 dyn_hair_04_r [{shapeType=Capsule, length=5.2692, radius=2.7888, center=<3.6711 0.3505 -0.1907>, rotation=<rot 88.8471 -178.4678 -178.5908>}]
 dyn_hair_05_l [{shapeType=Capsule, length=6.038, radius=2.3561, center=<3.548 0.7367 -0.08>, rotation=<rot 89.3968 -179.9992 177.9629>}]
 dyn_hair_05_r [{shapeType=Capsule, length=6.0209, radius=2.3561, center=<3.6482 0.9551 -0.3192>, rotation=<rot 89.3968 -179.9988 174.7178>}]
 dyn_hair_06_l [{shapeType=Capsule, length=8.0926, radius=1.9118, center=<5.2745 1.1171 0.2052>, rotation=<rot 89.8781 151.4418 149.5366>}]
 dyn_hair_06_r [{shapeType=Capsule, length=8.0926, radius=1.9118, center=<5.2829 1.1652 -0.3436>, rotation=<rot 89.7918 164.6354 162.7303>}]
 dyn_hair_shoulderBase_l [{shapeType=Capsule, length=0.505, radius=0.505, shapeName=dyn_hair_shoulderBase_l_capsule, rotation=<rot 0 -90 0>}]
 dyn_hair_shoulderBase_r [{shapeType=Capsule, length=0.505, radius=0.505, shapeName=dyn_hair_shoulderBase_r_capsule, rotation=<rot 0 -90 0>}]
 head [{shapeType=Capsule, length=4.7821, radius=6.0924, center=<3.548 -0.3123 0>, rotation=<rot 0 -95.0309 90>}]
 lowerarm_twist_02_l [{shapeType=Capsule, length=23.3421, radius=4.5344, center=<2.0527 -1.094 0>, rotation=<rot 90 -4.3246 0>}]
 lowerarm_twist_02_r [{shapeType=Capsule, length=23.3421, radius=4.5344, center=<-1.9827 2.0189 0>, rotation=<rot 90 175.6754 0>}]
 spine_01 [{shapeType=Capsule, length=9.8886, radius=7.4755, shapeName=spine_04_capsule, center=<0.0099 2.0493 0.2716>, rotation=<rot 0 -90 0>}]
 spine_02 [{shapeType=Capsule, length=7.9835, radius=7.4755, shapeName=spine_04_capsule, center=<0.0099 3.0012 0>, rotation=<rot 0 -90 0>}]
 spine_03 [{shapeType=Capsule, length=8.5042, radius=7.4755, shapeName=spine_04_capsule, center=<0.0099 3.2287 0>, rotation=<rot 0 -90 0>}]
 spine_04 [
  {shapeType=Capsule, length=16.2249, radius=7.4755, shapeName=spine_04_capsule, center=<0.0099 -2.4138 0>, rotation=<rot 0 -90 0>}
  {shapeType=Capsule, length=10.1779, radius=6.0025, shapeName=spine_04_capsule, center=<0.9893 5.5944 0.338>}
 ]
 spine_05 [{
   shapeType=Capsule
   length=17.7102
   radius=7.4755
   shapeName=spine_04_capsule
   center=<0.1458 -2.3472 0.0717>
   rotation=<rot -0.7437 -90.0002 0>
  }]
 thigh_l [{shapeType=Capsule, length=26.7584, radius=7.6629, center=<-18.0575 -1.829 2.2728>, rotation=<rot 90 175.6754 0>}]
 thigh_r [{shapeType=Capsule, length=26.7584, radius=7.6629, center=<18.0575 1.829 0>, rotation=<rot 90 -4.3246 0>}]
 upperarm_l [{shapeType=Capsule, length=26.8381, radius=4.0917, center=<13.3887 0 3.1079>, rotation=<rot 90 0 0>}]
 upperarm_r [{shapeType=Capsule, length=26.8381, radius=4.0917, center=<-12.5664 0 -3.1079>, rotation=<rot 90 180 0>}]
}
constraints [
 {
  bone1Name=dyn_hair_02_l
  bone2Name=dyn_hair_shoulderBase_l
  swing1LimitDegrees=45
  swing2LimitDegrees=10
  twistLimitDegrees=45
  twistMotion=Locked
 }
 {
  bone1Name=dyn_hair_02_r
  bone2Name=dyn_hair_shoulderBase_r
  swing1LimitDegrees=35
  swing2LimitDegrees=10
  twistLimitDegrees=45
  twistMotion=Locked
 }
 {bone1Name=dyn_hair_03_l, bone2Name=dyn_hair_02_l, swing1LimitDegrees=45, swing2LimitDegrees=15, twistLimitDegrees=45, twistMotion=Locked}
 {bone1Name=dyn_hair_04_l, bone2Name=dyn_hair_03_l, swing1LimitDegrees=45, swing2LimitDegrees=20, twistLimitDegrees=45, twistMotion=Locked}
 {bone1Name=dyn_hair_03_r, bone2Name=dyn_hair_02_r, swing1LimitDegrees=45, swing2LimitDegrees=15, twistLimitDegrees=45, twistMotion=Locked}
 {bone1Name=dyn_hair_04_r, bone2Name=dyn_hair_03_r, swing1LimitDegrees=45, swing2LimitDegrees=20, twistLimitDegrees=45, twistMotion=Locked}
 {bone1Name=dyn_hair_01_l, bone2Name=head, swing1LimitDegrees=40, swing2LimitDegrees=10, twistLimitDegrees=45, twistMotion=Locked}
 {bone1Name=dyn_hair_01_r, bone2Name=head, swing1LimitDegrees=40, swing2LimitDegrees=10, twistLimitDegrees=45, twistMotion=Locked}
 {bone1Name=dyn_hair_05_l, bone2Name=dyn_hair_04_l, swing1LimitDegrees=45, swing2LimitDegrees=20, twistLimitDegrees=45, twistMotion=Locked}
 {bone1Name=dyn_hair_06_l, bone2Name=dyn_hair_05_l, swing1LimitDegrees=45, swing2LimitDegrees=20, twistLimitDegrees=45, twistMotion=Locked}
 {bone1Name=dyn_hair_05_r, bone2Name=dyn_hair_04_r, swing1LimitDegrees=45, swing2LimitDegrees=20, twistLimitDegrees=45, twistMotion=Locked}
 {bone1Name=dyn_hair_06_r, bone2Name=dyn_hair_05_r, swing1LimitDegrees=45, swing2LimitDegrees=20, twistLimitDegrees=45, twistMotion=Locked}
]
