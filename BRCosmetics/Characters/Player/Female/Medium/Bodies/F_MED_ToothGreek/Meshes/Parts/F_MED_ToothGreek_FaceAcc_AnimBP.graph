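# Animation Blueprint for the ToothGreek female FaceAcc part, driving secondary motion on its hair bones.
# It copies the parent pose, pins forearms and thighs to owner-mesh world transforms, and adds skydive and emote offsets.
# Weighted constraints, a Control Rig, RigidBody and spring AnimDynamics then drive the hair, with a final front-end strand fix outside Golden Glory emotes.
# The event graph updates transforms, state flags and a gravity override the AnimGraph never reads.
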AnimBlueprint /BRCosmetics/Characters/Player/Female/Medium/Bodies/F_MED_ToothGreek/Meshes/Parts/F_MED_ToothGreek_FaceAcc_AnimBP
target_skeleton=&/F_MED_ToothGreek_FaceAcc_Skeleton
graphs [
 AnimGraph {nodes [
   n1 = "Transform (Modify) Bone - Bone: lowerarm_r"(
    ComponentPose=Local To Component(LocalPose=Copy Pose From Mesh() {bUseAttachedParent=true})
    Alpha=1
    Translation=Get lowerarm_right_loc() -> (lowerarm_right_loc:Vector)
    Rotation=Get lowerarm_right_rot() -> (lowerarm_right_rot:Rotator)
    Scale="1,1,1"
   ) {
    rotationMode=BMM_Replace
    rotationSpace=BCS_WorldSpace
    scaleMode=BMM_Replace
    scaleSpace=BCS_WorldSpace
    translationMode=BMM_Replace
    translationSpace=BCS_WorldSpace
    boneToModify {boneName=lowerarm_r}
   }
   n2 = "Transform (Modify) Bone - Bone: lowerarm_l"(
    ComponentPose=@n1
    Alpha=1
    Translation=Get lowerarm_left_loc() -> (lowerarm_left_loc:Vector)
    Rotation=Get lowerarm_left_rot() -> (lowerarm_left_rot:Rotator)
    Scale="1,1,1"
   ) {
    rotationMode=BMM_Replace
    rotationSpace=BCS_WorldSpace
    scaleMode=BMM_Replace
    translationMode=BMM_Replace
    translationSpace=BCS_WorldSpace
    boneToModify {boneName=lowerarm_l}
   }
   n3 = "Transform (Modify) Bone - Bone: thigh_l"(
    ComponentPose=@n2
    Alpha=1
    Translation=Get thigh_left_loc() -> (thigh_left_loc:Vector)
    Rotation=Get thigh_left_rot() -> (thigh_left_rot:Rotator)
    Scale="1,1,1"
   ) {
    rotationMode=BMM_Replace
    rotationSpace=BCS_WorldSpace
    scaleMode=BMM_Replace
    scaleSpace=BCS_WorldSpace
    translationMode=BMM_Replace
    translationSpace=BCS_WorldSpace
    boneToModify {boneName=thigh_l}
   }
   n4 = "Transform (Modify) Bone - Bone: thigh_r"(
    ComponentPose=@n3
    Alpha=1
    Translation=Get thigh_right_loc() -> (thigh_right_loc:Vector)
    Rotation=Get thigh_right_rot() -> (thigh_right_rot:Rotator)
    Scale="1,1,1"
   ) {
    rotationMode=BMM_Replace
    rotationSpace=BCS_WorldSpace
    scaleMode=BMM_Replace
    translationMode=BMM_Replace
    translationSpace=BCS_WorldSpace
    boneToModify {boneName=thigh_r}
   }
   n5 = Get bIsSkyDiving()
   n6 = "Transform (Modify) Bone - Bone: clavicle_l"!(ComponentPose=@n4, bAlphaBoolEnabled:bool=@n5, Translation="0,1.5,0", Rotation="0,0,0", Scale="1,1,1") -> (Pose) {
    alphaInputType=Bool
    rotationMode=BMM_Additive
    rotationSpace=BCS_BoneSpace
    translationMode=BMM_Additive
    translationSpace=BCS_BoneSpace
    alphaBoolBlend {blendInTime=0.3, blendOutTime=0.3}
    boneToModify {boneName=clavicle_l}
    translation {y=1.5}
   }
   n7 = "Transform (Modify) Bone - Bone: clavicle_r"!(ComponentPose=@n6, bAlphaBoolEnabled:bool=@n5, Translation="0,-1.5,0", Rotation="0,0,0", Scale="1,1,1") -> (Pose) {
    alphaInputType=Bool
    rotationMode=BMM_Additive
    rotationSpace=BCS_BoneSpace
    translationMode=BMM_Additive
    translationSpace=BCS_BoneSpace
    alphaBoolBlend {blendInTime=0.3, blendOutTime=0.3}
    boneToModify {boneName=clavicle_r}
    translation {y=-1.5}
   }
   n8 = "Transform (Modify) Bone - Bone: upperarm_l"!(ComponentPose=@n7, bAlphaBoolEnabled:bool=@n5, Translation="0,0,-5", Rotation="0,0,0", Scale="1,1,1") -> (Pose) {
    alphaInputType=Bool
    rotationMode=BMM_Additive
    rotationSpace=BCS_BoneSpace
    translationMode=BMM_Additive
    translationSpace=BCS_BoneSpace
    alphaBoolBlend {blendInTime=0.3, blendOutTime=0.3}
    boneToModify {boneName=upperarm_l}
    translation {z=-5}
   }
   n9 = "Transform (Modify) Bone - Bone: upperarm_r"!(ComponentPose=@n8, bAlphaBoolEnabled:bool=@n5, Translation="0,0,5", Rotation="0,0,0", Scale="1,1,1") -> (Pose) {
    alphaInputType=Bool
    rotationMode=BMM_Additive
    rotationSpace=BCS_BoneSpace
    translationMode=BMM_Additive
    translationSpace=BCS_BoneSpace
    alphaBoolBlend {blendInTime=0.3, blendOutTime=0.3}
    boneToModify {boneName=upperarm_r}
    translation {z=5}
   }
   n10 = "Transform (Modify) Bone - Bone: lowerarm_twist_02_r"!(ComponentPose=@n9, bAlphaBoolEnabled:bool=Get bIsPlayingEmote(), Translation="0,0,-5", Rotation="0,0,0", Scale="1,1,1") -> (Pose) {
    alphaInputType=Bool
    rotationMode=BMM_Additive
    rotationSpace=BCS_BoneSpace
    translationMode=BMM_Additive
    translationSpace=BCS_BoneSpace
    alphaBoolBlend {blendInTime=0.3, blendOutTime=0.3}
    boneToModify {boneName=lowerarm_twist_02_r}
    translation {z=-5}
   }
   n11 = "Constraint - dyn_hair_shoulderBase_r (PPPPP)"!(
    ComponentPose=@n10
    Alpha=1
    ConstraintWeights_0=0.4337
    ConstraintWeights_1:float=0.15
    ConstraintWeights_2:float=0.122
    ConstraintWeights_3:float=0.0943
    ConstraintWeights_4:float=0.2
   ) -> (Pose) {
    boneToModify {boneName=dyn_hair_shoulderBase_r}
    constraintSetup [
     {targetBone {boneName=clavicle_r}}
     {targetBone {boneName=spine_05}}
     {targetBone {boneName=upperarm_r}}
     {targetBone {boneName=spine_04}}
     {targetBone {boneName=head}}
    ]
    constraintWeights [1,1,1,1,1]
   }
   n12 = "Constraint - dyn_hair_shoulderBase_l (PPPPP)"!(
    ComponentPose=@n11
    Alpha=1
    ConstraintWeights_0=0.4337
    ConstraintWeights_1:float=0.15
    ConstraintWeights_2:float=0.122
    ConstraintWeights_3:float=0.0943
    ConstraintWeights_4:float=0.2
   ) -> (Pose) {
    boneToModify {boneName=dyn_hair_shoulderBase_l}
    constraintSetup [
     {targetBone {boneName=clavicle_l}}
     {targetBone {boneName=spine_05}}
     {targetBone {boneName=upperarm_l}}
     {targetBone {boneName=spine_04}}
     {targetBone {boneName=head}}
    ]
    constraintWeights [1,1,1,1,1]
   }
   n13 = Control Rig!(
    Source=Component To Local(ComponentPose=@n12)
    Alpha
    isEmotingOrSkydiving_CR:bool=Get isEmotingOrSkydiving() -> (isEmotingOrSkydiving:bool)
    isInFE_CR:bool=Get bIsInFrontEnd()
    GoldenGloryEmote:bool=Get GoldenGloryEmote() -> (GoldenGloryEmote:bool)
   ) -> (Pose) {controlRigAssetReference {blueprintRigClass=&/F_MED_ToothGreek_FaceAcc_CtrlRig.F_MED_ToothGreek_FaceAcc_CtrlRig_C}}
   n14 = RigidBody!(ComponentPose=Local To Component(LocalPose=@n13), Alpha, OverrideWorldGravity:Vector="0,0,0", ExternalForce) -> (Pose) {
    bOverrideWorldGravity=true
    overridePhysicsAsset=&/F_MED_ToothGreek_FaceAcc_Physics
    simulationSpace=BaseBoneSpace
    baseBoneRef {boneName=spine_03}
    componentAppliedLinearAccClamp=<2000 2000 3000>
    simSpaceSettings {dampingAlpha=0, maxAngularAcceleration=50, maxAngularVelocity=10, maxLinearAcceleration=150, maxLinearVelocity=25, worldAlpha=1}
   }
   n15 = "Constraint - dyn_hairStrandA_02_l (PPPP)"!(
    ComponentPose=@n14
    Alpha=1
    ConstraintWeights_0=0.3635
    ConstraintWeights_1:float=0.1751
    ConstraintWeights_2:float=0.047
    ConstraintWeights_3:float=0.4145
   ) -> (Pose) {
    boneToModify {boneName=dyn_hairStrandA_02_l}
    constraintSetup [{targetBone {boneName=dyn_hair_01_l}}, {targetBone {boneName=head}}, {targetBone {boneName=neck_02}}, {targetBone {boneName=dyn_hair_02_l}}]
    constraintWeights [1,1,1,1]
   }
   n16 = "Constraint - base_hairStrandE_r (PPPPPPPP)"!(
    ComponentPose=@n15
    Alpha=1
    ConstraintWeights_0=0.5695
    ConstraintWeights_1:float=0.2421
    ConstraintWeights_2:float=0.1147
    ConstraintWeights_3:float=0.0317
    ConstraintWeights_4:float=0.0184
    ConstraintWeights_5:float=0.0092
    ConstraintWeights_6:float=0.0076
    ConstraintWeights_7:float=0.0068
   ) -> (Pose) {
    boneToModify {boneName=base_hairStrandE_r}
    constraintSetup [
     {targetBone {boneName=dyn_hair_01_r}}
     {targetBone {boneName=dyn_hair_02_r}}
     {targetBone {boneName=dyn_hairStrandE_01_r}}
     {targetBone {boneName=dyn_hair_03_r}}
     {targetBone {boneName=spine_05}}
     {targetBone {boneName=head}}
     {targetBone {boneName=clavicle_r}}
     {targetBone {boneName=dyn_hair_04_r}}
    ]
    constraintWeights [1,1,1,1,1,1,1,1]
   }
   n17 = Anim Dynamics - Bone: dyn_hairStrandA_02_l!(ComponentPose=@n16, Alpha=1, GravityOverride:Vector="0,0,-980") -> (Pose) {
    angularDampingOverride=0.99
    angularSpringConstant=1500
    bAngularSpring=true
    bGravityOverrideInSimSpace=true
    bOverrideAngularDamping=true
    bUseGravityOverride=true
    alphaBoolBlend {blendOption=Linear}
    boundBone {boneName=dyn_hairStrandA_02_l}
    componentAppliedLinearAccClamp=<3500 3500 2500>
    componentLinearAccScale=<0.5 0.5 0.35>
    componentLinearVelScale=<0.55 0.55 0.4>
    gravityOverride {z=-980}
    physicsBodyDefinitions [{
      collisionType=CustomSphere
      sphereCollisionRadius=2
      boundBone {boneName=dyn_hairStrandA_02_l}
      boxExtents=<15 15 15>
      constraintSetup {angularLimitsMax=<2 7 2>, angularLimitsMin=<-2 -1 -15>, angularTarget=<0 0 0>}
      localJointOffset=<5 0 0>
     }]
    relativeSpaceBone {boneName=root}
   }
   n18 = Anim Dynamics - Bone: dyn_hairStrandA_02_r!(ComponentPose=@n17, Alpha=1, GravityOverride:Vector="0,0,-980") -> (Pose) {
    angularDampingOverride=0.99
    angularSpringConstant=1500
    bAngularSpring=true
    bGravityOverrideInSimSpace=true
    bOverrideAngularDamping=true
    bUseGravityOverride=true
    alphaBoolBlend {blendOption=Linear}
    boundBone {boneName=dyn_hairStrandA_02_r}
    componentAppliedLinearAccClamp=<3500 3500 2500>
    componentLinearAccScale=<0.5 0.5 0.35>
    componentLinearVelScale=<0.55 0.55 0.4>
    gravityOverride {z=-980}
    physicsBodyDefinitions [{
      collisionType=CustomSphere
      sphereCollisionRadius=2
      boundBone {boneName=dyn_hairStrandA_02_r}
      boxExtents=<15 15 15>
      constraintSetup {angularLimitsMax=<2 1 15>, angularLimitsMin=<-2 -7 -15>, angularTarget=<0 0 0>}
      localJointOffset=<5 0 0>
     }]
    relativeSpaceBone {boneName=root}
   }
   n19 = Anim Dynamics - Bone: dyn_hairStrandD_01_r!(ComponentPose=@n18, Alpha=1, GravityOverride:Vector="0,0,-980") -> (Pose) {
    angularDampingOverride=0.99
    angularSpringConstant=1500
    bAngularSpring=true
    bGravityOverrideInSimSpace=true
    bOverrideAngularDamping=true
    bUseGravityOverride=true
    alphaBoolBlend {blendOption=Linear}
    boundBone {boneName=dyn_hairStrandD_01_r}
    componentAppliedLinearAccClamp=<3500 3500 2500>
    componentLinearAccScale=<0.5 0.5 0.35>
    componentLinearVelScale=<0.55 0.55 0.4>
    gravityOverride {z=-980}
    physicsBodyDefinitions [{
      collisionType=CustomSphere
      sphereCollisionRadius=2
      boundBone {boneName=dyn_hairStrandD_01_r}
      boxExtents=<15 15 15>
      constraintSetup {angularLimitsMax=<2 7 5>, angularLimitsMin=<-2 -1 -15>, angularTarget=<0 0 0>}
      localJointOffset=<5 0 0>
     }]
    relativeSpaceBone {boneName=root}
   }
   n20 = Anim Dynamics - Bone: dyn_hairStrandE_01_r!(ComponentPose=@n19, Alpha=1, GravityOverride:Vector="0,0,-980") -> (Pose) {
    angularDampingOverride=0.99
    angularSpringConstant=1500
    bAngularSpring=true
    bGravityOverrideInSimSpace=true
    bOverrideAngularDamping=true
    bUseGravityOverride=true
    alphaBoolBlend {blendOption=Linear}
    boundBone {boneName=dyn_hairStrandE_01_r}
    componentAppliedLinearAccClamp=<3500 3500 2500>
    componentLinearAccScale=<0.5 0.5 0.35>
    componentLinearVelScale=<0.55 0.55 0.4>
    gravityOverride {z=-980}
    physicsBodyDefinitions [{
      collisionType=CustomSphere
      sphereCollisionRadius=2
      boundBone {boneName=dyn_hairStrandE_01_r}
      boxExtents=<15 15 15>
      constraintSetup {angularLimitsMax=<2 1 25>, angularLimitsMin=<-2 -7 -15>, angularTarget=<0 0 0>}
      localJointOffset=<5 0 0>
     }]
    relativeSpaceBone {boneName=root}
   }
   n21 = Anim Dynamics - Bone: dyn_hairStrandF_01_r!(ComponentPose=@n20, Alpha=1, GravityOverride:Vector="0,0,-980") -> (Pose) {
    angularDampingOverride=0.99
    angularSpringConstant=1500
    bAngularSpring=true
    bGravityOverrideInSimSpace=true
    bOverrideAngularDamping=true
    bUseGravityOverride=true
    alphaBoolBlend {blendOption=Linear}
    boundBone {boneName=dyn_hairStrandF_01_r}
    componentAppliedLinearAccClamp=<3500 3500 2500>
    componentLinearAccScale=<0.5 0.5 0.35>
    componentLinearVelScale=<0.55 0.55 0.4>
    gravityOverride {z=-980}
    physicsBodyDefinitions [{
      collisionType=CustomSphere
      sphereCollisionRadius=2
      boundBone {boneName=dyn_hairStrandF_01_r}
      boxExtents=<15 15 15>
      constraintSetup {angularLimitsMax=<2 1 2>, angularLimitsMin=<-2 -7 -15>, angularTarget=<0 0 0>}
      localJointOffset=<5 0 0>
     }]
    relativeSpaceBone {boneName=root}
   }
   n22 = "Constraint - pos_hair_a_01_l (P)"(ComponentPose=@n21, Alpha=1, ConstraintWeights_0=1) {boneToModify {boneName=pos_hair_a_01_l}, constraintSetup [{targetBone {boneName=dyn_hair_01_l}}], constraintWeights [1]}
   n23 = "Constraint - pos_hair_a_02_l (P)"(ComponentPose=@n22, Alpha=1, ConstraintWeights_0=1) {boneToModify {boneName=pos_hair_a_02_l}, constraintSetup [{targetBone {boneName=dyn_hair_02_l}}], constraintWeights [1]}
   n24 = "Constraint - pos_hair_a_03_l (P)"(ComponentPose=@n23, Alpha=1, ConstraintWeights_0=1) {boneToModify {boneName=pos_hair_a_03_l}, constraintSetup [{targetBone {boneName=dyn_hair_03_l}}], constraintWeights [1]}
   n25 = "Constraint - pos_hair_a_04_l (P)"(ComponentPose=@n24, Alpha=1, ConstraintWeights_0=1) {boneToModify {boneName=pos_hair_a_04_l}, constraintSetup [{targetBone {boneName=dyn_hair_04_l}}], constraintWeights [1]}
   n26 = "Transform (Modify) Bone - Bone: base_hairStrandE_r"!(
    ComponentPose=@n25
    bAlphaBoolEnabled:bool=Get FrontEndeNotGoldenGlory() -> (FrontEndeNotGoldenGlory:bool)
    Translation="-0.927,-1.828,-0.735"
    Rotation="0.843,-3.201,14.377"
    Scale="1,1,1"
   ) -> (Pose) {
    alphaInputType=Bool
    rotationMode=BMM_Additive
    rotationSpace=BCS_BoneSpace
    translationMode=BMM_Additive
    translationSpace=BCS_BoneSpace
    boneToModify {boneName=base_hairStrandE_r}
    rotation=<rot 0.843 -3.2011 14.377>
    translation=<-0.927 -1.828 -0.735>
   }
   Output Pose(Result=Component To Local(ComponentPose=@n26)) {layerGroup=None}
  ]}
 EventGraph {nodes [
   Event BlueprintUpdateAnimation()
   > n1 = Sequence()
   > n2 = "Gravity Override State Function (Parameter Asset) (Nativized)"(
    CharacterPartAnimInstance=Self-Reference()
    GameTimeInSeconds=GetGameTimeInSeconds()
    DebugJointName=spine_05
    Parameters=/BRCosmetics/Characters/Player/Female/Medium/Bodies/F_MED_ToothGreek/Meshes/F_MED_ToothGreek_Parameters
   )
   > Set Gravity Override(Gravity Override=@n2)
   > Set Skeletal_Mesh(Skeletal_Mesh:Skeletal Mesh Component Object Reference=GetOwnerSkeletalMeshComponent()) -> (Output_Get:Skeletal Mesh Component Object Reference)
   > Set isEmotingOrSkydiving(isEmotingOrSkydiving:bool=OR Boolean(A=Get bIsPlayingEmote(), B=Get bIsSkyDiving())) -> (Output_Get:bool)
   n3 = GetAnimInstance(self=GetOwnerSkeletalMeshComponent())
   n4 = OR Boolean(
    A=Montage_IsPlaying(self=@n3, Montage=/BRCosmetics/Animation/Game/MainPlayer/Emotes/GoldenGlory/CMF/Emote_GoldenGlory_CMF_M)
    B=Montage_IsPlaying(self=@n3, Montage=/BRCosmetics/Animation/Game/MainPlayer/Emotes/GoldenGlory/CMF/Emote_GoldenGlory_CMF_FrontEnd_M)
   )
   > Set GoldenGloryEmote(GoldenGloryEmote:bool=@n4) -> (Output_Get:bool)
   > Set FrontEndeNotGoldenGlory(FrontEndeNotGoldenGlory:bool=AND Boolean(A=NOT Boolean(A=@n4), B=Get bIsInFrontEnd())) -> (Output_Get:bool)
   n5 = Get Skeletal_Mesh() -> (Skeletal_Mesh:Skeletal Mesh Component Object Reference)
   n6 = BreakTransform(InTransform=GetSocketTransform(self=@n5, InSocketName=lowerarm_r, TransformSpace=RTS_World))
   Set lowerarm_right_loc(execute=@n1.then_1, lowerarm_right_loc:Vector=@n6) -> (Output_Get:Vector)
   > Set lowerarm_right_rot(lowerarm_right_rot:Rotator=@n6.Rotation) -> (Output_Get:Rotator)
   n7 = BreakTransform(InTransform=GetSocketTransform(self=@n5, InSocketName=lowerarm_l, TransformSpace=RTS_World))
   > Set lowerarm_left_loc(lowerarm_left_loc:Vector=@n7) -> (Output_Get:Vector)
   > Set lowerarm_left_rot(lowerarm_left_rot:Rotator=@n7.Rotation) -> (Output_Get:Rotator)
   n8 = BreakTransform(InTransform=GetSocketTransform(self=@n5, InSocketName=thigh_l, TransformSpace=RTS_World))
   > Set thigh_left_loc(thigh_left_loc:Vector=@n8) -> (Output_Get:Vector)
   > Set thigh_left_rot(thigh_left_rot:Rotator=@n8.Rotation) -> (Output_Get:Rotator)
   n9 = BreakTransform(InTransform=GetSocketTransform(self=@n5, InSocketName=thigh_r, TransformSpace=RTS_World))
   > Set thigh_right_loc(thigh_right_loc:Vector=@n9) -> (Output_Get:Vector)
   > Set thigh_right_rot(thigh_right_rot:Rotator=@n9.Rotation) -> (Output_Get:Rotator)
  ]}
]
variables [
 FrontEndeNotGoldenGlory {type=boolean}
 GoldenGloryEmote {type=boolean}
 Gravity Override {type=Vector}
 isEmotingOrSkydiving {type=boolean}
 lowerarm_left_loc {type=Vector}
 lowerarm_left_rot {type=Rotator}
 lowerarm_right_loc {type=Vector}
 lowerarm_right_rot {type=Rotator}
 Skeletal_Mesh {type=SkeletalMeshComponent}
 thigh_left_loc {type=Vector}
 thigh_left_rot {type=Rotator}
 thigh_right_loc {type=Vector}
 thigh_right_rot {type=Rotator}
]
